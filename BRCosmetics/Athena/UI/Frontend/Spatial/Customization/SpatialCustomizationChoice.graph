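# Fortnite frontend card widget for one spatial customization choice, showing an item icon with a lock, timer, bang or checkmark status.
# PreConstruct applies size, status-icon visibility, status-bar use and locked-state opacity from its variables.
# OnSetupChoice passes bar color to the icon's dynamic material, then maps unlock state to material opacity, unlocked flag and switcher index.
# Two hover animations drive material parameters, playing forward and reversing at higher speed.
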WidgetBlueprint /BRCosmetics/Athena/UI/Frontend/Spatial/Customization/SpatialCustomizationChoice
animations [
 Hover {
  binding_count=3
  length_seconds=0.25
  object=&self:Hover
  sequence {
   bindings [
    Image_Locked {
     id=#1
     possessed_class=FortLazyImage
     tracks [
      {
       type=MovieSceneColorTrack
       display_name=Color and Opacity
       sections [{
         type=MovieSceneColorSection
         channels [
          Color.R {type=MovieSceneScriptingFloatChannel, keys [[0,0.2159],[18,1]]}
          Color.G {type=MovieSceneScriptingFloatChannel, keys [[0,0.3515],[18,1]]}
          Color.B {type=MovieSceneScriptingFloatChannel, const [1,1]}
          Color.A {type=MovieSceneScriptingFloatChannel, const [1,1]}
         ]
        }]
      }
      {
       type=MovieScene2DTransformTrack
       sections [{
         type=MovieScene2DTransformSection
         end_frame=30
         start_frame=0
         channels [
          Translation.X {type=MovieSceneScriptingFloatChannel, keys [[0,0],[5,3],[11,-2],[18,1],[25,0.5],[30,0]]}
          Translation.Y {type=MovieSceneScriptingFloatChannel}
          Angle {type=MovieSceneScriptingFloatChannel}
          Scale.X {type=MovieSceneScriptingFloatChannel}
          Scale.Y {type=MovieSceneScriptingFloatChannel}
          Shear.X {type=MovieSceneScriptingFloatChannel}
          Shear.Y {type=MovieSceneScriptingFloatChannel}
         ]
        }]
      }
     ]
    }
    Image_ItemIcon {
     id=#2
     possessed_class=Image
     tracks [
      {
       type=MovieSceneWidgetMaterialTrack
       sections [{
         type=MovieSceneParameterSection
         channels [
          Icon Scale {type=MovieSceneScriptingFloatChannel, keys [[0,0.95],[30,1.1]]}
          AnimateIconOpacity {type=MovieSceneScriptingFloatChannel, keys [[0,0],[30,1]]}
          Bar_Right_Rotation {type=MovieSceneScriptingFloatChannel, keys [[0,-0.4],[30,-0.8]]}
          Border_Left_Rotation {type=MovieSceneScriptingFloatChannel, keys [[0,-0.6],[30,-0.7]]}
          Border_Right_Offset {type=MovieSceneScriptingFloatChannel, keys [[0,-7.5],[30,-8]]}
          Border_Right_Rotation {type=MovieSceneScriptingFloatChannel, keys [[0,-0.3],[30,-0.6]]}
         ]
        }]
      }
      {
       type=MovieScene2DTransformTrack
       sections [{
         type=MovieScene2DTransformSection
         end_frame=30
         start_frame=0
         channels [
          Translation.X {type=MovieSceneScriptingFloatChannel}
          Translation.Y {type=MovieSceneScriptingFloatChannel}
          Angle {type=MovieSceneScriptingFloatChannel}
          Scale.X {type=MovieSceneScriptingFloatChannel, keys [[0,1],[30,1.05]]}
          Scale.Y {type=MovieSceneScriptingFloatChannel, keys [[0,1],[30,1.05]]}
          Shear.X {type=MovieSceneScriptingFloatChannel}
          Shear.Y {type=MovieSceneScriptingFloatChannel}
         ]
        }]
      }
     ]
    }
    Image_timer {
     id=#3
     possessed_class=Image
     tracks [
      {
       type=MovieSceneColorTrack
       display_name=Color and Opacity
       sections [{
         type=MovieSceneColorSection
         channels [
          Color.R {type=MovieSceneScriptingFloatChannel, keys [[0,0.2159],[18,1]]}
          Color.G {type=MovieSceneScriptingFloatChannel, keys [[0,0.3515],[18,1]]}
          Color.B {type=MovieSceneScriptingFloatChannel, const [1,1]}
          Color.A {type=MovieSceneScriptingFloatChannel, const [1,1]}
         ]
        }]
      }
      {
       type=MovieScene2DTransformTrack
       sections [{
         type=MovieScene2DTransformSection
         channels [
          Translation.X {type=MovieSceneScriptingFloatChannel, keys [[0,0],[5,3],[11,-2],[18,1],[25,0.5],[30,0]]}
          Translation.Y {type=MovieSceneScriptingFloatChannel}
          Angle {type=MovieSceneScriptingFloatChannel}
          Scale.X {type=MovieSceneScriptingFloatChannel}
          Scale.Y {type=MovieSceneScriptingFloatChannel}
          Shear.X {type=MovieSceneScriptingFloatChannel}
          Shear.Y {type=MovieSceneScriptingFloatChannel}
         ]
        }]
      }
     ]
    }
   ]
   display_rate {numerator=120}
  }
 }
 Hover_alternate {
  binding_count=1
  length_seconds=0.25
  object=&self:Hover_alternate
  sequence {
   bindings [Image_ItemIcon {
     id=#2
     possessed_class=Image
     tracks [{
       type=MovieSceneWidgetMaterialTrack
       sections [{
         type=MovieSceneParameterSection
         channels [
          AnimateIconOpacity {type=MovieSceneScriptingFloatChannel, keys [[0,0],[30,1]]}
          Bar_Right_Rotation {type=MovieSceneScriptingFloatChannel, keys [[0,-0.6],[30,-0.8]]}
          Border_Left_Rotation {type=MovieSceneScriptingFloatChannel, keys [[0,-0.6],[30,-0.8]]}
          AnimateBorder {type=MovieSceneScriptingFloatChannel, keys [[0,0],[30,1]]}
          Border_Right_Rotation {type=MovieSceneScriptingFloatChannel, keys [[0,-0.3],[30,-0.4]]}
          Border_Bottom_Rotation {type=MovieSceneScriptingFloatChannel, keys [[0,0.2],[30,0.3]]}
          Border_Top_Rotation {type=MovieSceneScriptingFloatChannel, keys [[0,0.3],[30,0.4]]}
          AnimateIconScale {type=MovieSceneScriptingFloatChannel, keys [[0,0],[30,1]]}
         ]
        }]
      }]
    }]
   display_rate {numerator=120}
  }
 }
]
graphs [
 EventGraph {nodes [
   n1 = Get OverrideSize() -> (OverrideSize:double)
   n2 = SetPadding!(
    execute
    self:Overlay Slot Object Reference=SlotAsOverlaySlot(Widget=Get WidgetSwitcher_StatusIcon() -> (WidgetSwitcher_StatusIcon:Widget Switcher Object Reference))
    InPadding:Margin Structure=Select(
     Option 0:Margin Structure=Make Margin(Left=16, Top=16, Right=16, Bottom=16)
     Option 1:Margin Structure=Make Margin(Left=24, Top=24, Right=24, Bottom=24)
     Index="float > float"(A=@n1, B=180) {is=title K2Node_CallFunction}
    ) -> (ReturnValue:Margin Structure)
   ) -> (then)
   SetVisibility(
    execute=[@n2.then, @n13.then, @n12.then]
    self=Get WidgetSwitcher_StatusIcon()
    InVisibility=Select(
     Option 0:ESlateVisibility Enum=Collapsed
     Option 1:ESlateVisibility Enum=SelfHitTestInvisible
     Index=Get ShowStatusIcon() -> (ShowStatusIcon:bool)
    ) -> (ReturnValue:ESlateVisibility Enum)
   )
   > n3 = GetDynamicMaterial(self=Get Image_ItemIcon() -> (Image_ItemIcon:Fort Lazy Image Object Reference))
   > SetScalarParameterValue(self=@n3, ParameterName=UseBar, Value="To Float (Boolean)"(InBool=Get ShowStatusBar() -> (ShowStatusBar:bool)))
   > n4 = Branch(Condition=Get FullOpacity() -> (FullOpacity:bool))
   > Set OverallOpacity(OverallOpacity:double=1) -> (Output_Get:double)
   > Set IconOpacity(IconOpacity:double=1) -> (Output_Get:double)
   Set OverallOpacity(execute=@n4.else, OverallOpacity=0.8)
   > Set IconOpacity(IconOpacity=0.8)
   Play hover alternate() {is=title K2Node_CustomEvent}
   > PlayAnimationForward(InAnimation=Get Hover_alternate() -> (Hover_alternate:Widget Animation Object Reference))
   OnUnhovered() {is=title K2Node_CustomEvent}
   > PlayAnimationReverse(InAnimation=Get Hover() -> (Hover:Widget Animation Object Reference), PlaybackSpeed=3)
   OnHovered() {is=title K2Node_CustomEvent}
   > PlayAnimationForward(InAnimation=Get Hover())
   Play reversed hover alternate() {is=title K2Node_CustomEvent}
   > PlayAnimationReverse(InAnimation=Get Hover_alternate(), PlaybackSpeed=3)
   n5 = Event OnSetupChoice!() -> (OutputDelegate, then, State:ESpatialCustomizationCategoryState Enum=LockedByBattlePass, Color:Linear Color Structure) {is=title K2Node_Event}
   > n6 = GetDynamicMaterial(self=Get Image_ItemIcon())
   > SetVectorParameterValue(self=@n6, ParameterName=BarColor, Value=@n5.Color)
   > n7 = Sequence()
   > n8 = GetDynamicMaterial(self=Get Image_ItemIcon())
   > n9 = Switch on ESpatialCustomizationCategoryState!(execute, Selection:ESpatialCustomizationCategoryState Enum=@n5.State) -> (LockedByBattlePass:Exec, LockedByChallenge:Exec, LockedByChallengeCompletion:Exec, UnlockAvailable:Exec, UnlockUsed:Exec) {is=title K2Node_SwitchEnum}
   SetScalarParameterValue(execute=@n9.LockedByChallengeCompletion, self=@n8, ParameterName=OverallOpacity, Value=1)
   > n10 = SetScalarParameterValue(self=@n8, ParameterName=IconOpacity, Value=1)
   SetScalarParameterValue(execute=[@n10.then, @n11.then], self=@n8, ParameterName=IsUnlocked, Value=0)
   SetScalarParameterValue(execute=[@n9.UnlockAvailable, @n9.UnlockUsed], self=@n8, ParameterName=OverallOpacity, Value=1)
   > SetScalarParameterValue(self=@n8, ParameterName=IconOpacity, Value=1)
   > SetScalarParameterValue(self=@n8, ParameterName=IsUnlocked, Value=1)
   SetScalarParameterValue(
    execute=[@n9.LockedByBattlePass, @n9.LockedByChallenge]
    self=@n8
    ParameterName=OverallOpacity
    Value=Get OverallOpacity() -> (OverallOpacity:double)
   )
   > n11 = SetScalarParameterValue(self=@n8, ParameterName=IconOpacity, Value=Get IconOpacity() -> (IconOpacity:double))
   SetActiveWidgetIndex(
    execute=@n7.then_1
    self=Get WidgetSwitcher_StatusIcon()
    Index=Select!(
     LockedByBattlePass:int=0
     LockedByChallenge:int=1
     LockedByChallengeCompletion:int=0
     UnlockAvailable:int=2
     UnlockUsed:int=3
     Index:ESpatialCustomizationCategoryState Enum=@n5.State
    ) -> (ReturnValue:int)
   )
   Event PreConstruct()
   > SetHeightOverride(self=Get SizeBox_Whole() -> (SizeBox_Whole:Size Box Object Reference), InHeightOverride=@n1)
   > n12 = Branch(Condition=Get ShowStatusBar())
   n13 = SetWidthOverride(execute=@n12.else, self=Get SizeBox_Whole(), InWidthOverride=@n1)
   Event Construct()
  ]}
 UseSmallerIcon {nodes [
   n1 = UseSmallerIcon!() -> (then, bUseSmallIcon:bool=~) {is=title K2Node_FunctionEntry}
   > n2 = GetDynamicMaterial(self=Get Image_ItemIcon())
   > SetScalarParameterValue(self=@n2, ParameterName=UseSmallerIcon, Value=Select!(Option 0:double=0, Option 1:double=1, Index:bool=@n1) -> (ReturnValue:double))
  ]}
]
variables [
 FullOpacity {type=boolean}
 IconOpacity {type=number, default=1}
 OverallOpacity {type=number, default=1}
 OverrideSize {type=number, default=128}
 ShowStatusBar {type=boolean}
 ShowStatusIcon {type=boolean, default=true}
]
widgets {
 info {inheritedWidgetCount=3, parentClass=/Script/FortniteUI.AthenaSpatialCustomizationChoice, rootWidgetClass=&SizeBox, widgetCount=11}
 widgets [
  SizeBox_Whole {
   widgetClassPath=&SizeBox
   bIsVariable=true
   widget=SizeBox_Whole
   properties {bOverride_HeightOverride=true, bOverride_WidthOverride=true, heightOverride=200, widthOverride=152}
  }
  Overlay_0 {widgetClassPath=&Overlay, parent=SizeBox_Whole, widget=Overlay_0, slot {class=SizeBoxSlot}}
  Image_ItemIcon {
   widgetClassPath=&FortLazyImage
   bInherited=true
   bIsVariable=true
   parent=Overlay_0
   widget=Image_ItemIcon
   properties {
    materialTextureParamName=ItemIcon
    brush {
     bIsDynamicallyLoaded=false
     drawAs=Image
     imageType=NoImage
     mirroring=NoMirror
     resourceName=None
     resourceObject=/BRCosmetics/Athena/UI/Frontend/Spatial/Art/M_Ui_Spatial_itemCard_Umbrella_Base
     tiling=NoTile
     imageSize=<0 0>
     margin=<ltrb 0 0 0 0>
     outlineSettings {
      bUseBrushTransparency=false
      roundingType=HalfHeightRadius
      width=0
      color {colorUseRule=UseColor_Specified, specifiedColor=<rgba 0 0 0 0>}
      cornerRadii=<0 0 0 0>
     }
     tintColor {colorUseRule=UseColor_Specified, specifiedColor=<rgba 1 1 1 1>}
     uVRegion {bIsValid=false, max=<0 0>, min=<0 0>}
    }
   }
   slot {class=OverlaySlot, properties {horizontalAlignment=HAlign_Fill, verticalAlignment=VAlign_Fill}}
  }
  WidgetSwitcher_StatusIcon {
   widgetClassPath=&WidgetSwitcher
   bIsVariable=true
   parent=Overlay_0
   widget=WidgetSwitcher_StatusIcon
   slot {class=OverlaySlot, properties {horizontalAlignment=HAlign_Right, verticalAlignment=VAlign_Bottom}}
  }
  Image_Locked {
   widgetClassPath=&FortLazyImage
   bInherited=true
   bIsVariable=true
   parent=WidgetSwitcher_StatusIcon
   widget=Image_Locked
   properties {brush {
     bIsDynamicallyLoaded=false
     drawAs=Image
     imageType=NoImage
     mirroring=NoMirror
     resourceName=None
     resourceObject=/BRCosmetics/UI/Art_ToolKit/AngledBoxes/M_UI_AngledLockedBox_Umbrella
     tiling=NoTile
     imageSize=<40 40>
     margin=<ltrb 0 0 0 0>
     outlineSettings {
      bUseBrushTransparency=false
      roundingType=HalfHeightRadius
      width=0
      color {colorUseRule=UseColor_Specified, specifiedColor=<rgba 0 0 0 0>}
      cornerRadii=<0 0 0 0>
     }
     tintColor {colorUseRule=UseColor_Specified, specifiedColor=<rgba 1 1 1 1>}
     uVRegion {bIsValid=false, max=<0 0>, min=<0 0>}
    }}
   slot {class=WidgetSwitcherSlot, properties {verticalAlignment=VAlign_Bottom, padding=<ltrb 16 16 16 16>}}
  }
  Image_timer {
   widgetClassPath=&Image
   bIsVariable=true
   parent=WidgetSwitcher_StatusIcon
   widget=Image_timer
   properties {brush {
     bIsDynamicallyLoaded=false
     drawAs=Image
     imageType=NoImage
     mirroring=NoMirror
     resourceName=None
     resourceObject=/BRCosmetics/Athena/UI/Frontend/ActivityBrowser/Art/M_UI_Timer_Icon_Umbrella
     tiling=NoTile
     imageSize=<36 36>
     margin=<ltrb 0 0 0 0>
     outlineSettings {
      bUseBrushTransparency=false
      roundingType=HalfHeightRadius
      width=0
      color {colorUseRule=UseColor_Specified, specifiedColor=<rgba 0 0 0 0>}
      cornerRadii=<0 0 0 0>
     }
     tintColor {colorUseRule=UseColor_Specified, specifiedColor=<rgba 1 1 1 1>}
     uVRegion {bIsValid=false, max=<0 0>, min=<0 0>}
    }}
   slot {class=WidgetSwitcherSlot, properties {horizontalAlignment=HAlign_Right, verticalAlignment=VAlign_Bottom, padding=<ltrb 14 14 14 14>}}
  }
  Image_bang {
   widgetClassPath=&Overlay
   parent=WidgetSwitcher_StatusIcon
   widget=Image_bang
   slot {class=WidgetSwitcherSlot, properties {verticalAlignment=VAlign_Bottom, padding=<ltrb 16 16 16 16>}}
  }
  Image_179 {
   widgetClassPath=&Image
   bIsVariable=true
   parent=Image_bang
   widget=Image_179
   properties {brush {
     bIsDynamicallyLoaded=false
     drawAs=Image
     imageType=NoImage
     mirroring=NoMirror
     resourceName=None
     resourceObject=/BRCosmetics/UI/Art_ToolKit/AngledBoxes/M_UI_Bang_Umbrella
     tiling=NoTile
     imageSize=<36 36>
     margin=<ltrb 0 0 0 0>
     outlineSettings {
      bUseBrushTransparency=false
      roundingType=HalfHeightRadius
      width=0
      color {colorUseRule=UseColor_Specified, specifiedColor=<rgba 0 0 0 0>}
      cornerRadii=<0 0 0 0>
     }
     tintColor {colorUseRule=UseColor_Specified, specifiedColor=<rgba 1 1 1 1>}
     uVRegion {bIsValid=false, max=<0 0>, min=<0 0>}
    }}
   slot {class=OverlaySlot, properties {horizontalAlignment=HAlign_Fill, verticalAlignment=VAlign_Fill}}
  }
  Text_ExclamationMark {
   widgetClassPath=&CommonTextBlock
   bIsVariable=true
   parent=Image_bang
   widget=Text_ExclamationMark
   properties {
    justification=Center
    text="!"
    textTransformPolicy=ToUpper
    colorAndOpacity {colorUseRule=UseColor_Specified, specifiedColor=<rgba 0 0 0 1>}
    font {
     bForceMonospaced=false
     bMaterialIsStencil=false
     fontMaterial=null
     fontObject=&BurbankSmall
     letterSpacing=0
     monospacedWidth=1
     size=18
     skewAmount=0
     typefaceFontName=Bold
     outlineSettings {
      bApplyOutlineToDropShadows=false
      bMiteredCorners=false
      bSeparateFillAlpha=false
      outlineMaterial=null
      outlineSize=0
      outlineColor=<rgba 0 0 0 1>
     }
    }
    renderTransform {angle=0, scale=<1 1>, shear=<0 0>, translation=<0 3>}
    shadowOffset=<0 0>
   }
   slot {class=OverlaySlot, properties {horizontalAlignment=HAlign_Center}}
  }
  Image_complete {
   widgetClassPath=&Image
   parent=WidgetSwitcher_StatusIcon
   widget=Image_complete
   properties {
    visibility=SelfHitTestInvisible
    brush {
     bIsDynamicallyLoaded=false
     drawAs=Image
     imageType=NoImage
     mirroring=NoMirror
     resourceName=None
     resourceObject=/FortUILibrary/Foundation/Art_ToolKit/AngledBoxes/M_UI_AngledCheckBox
     tiling=NoTile
     imageSize=<40 40>
     margin=<ltrb 0 0 0 0>
     outlineSettings {
      bUseBrushTransparency=false
      roundingType=HalfHeightRadius
      width=0
      color {colorUseRule=UseColor_Specified, specifiedColor=<rgba 0 0 0 0>}
      cornerRadii=<0 0 0 0>
     }
     tintColor {colorUseRule=UseColor_Specified, specifiedColor=<rgba 1 1 1 1>}
     uVRegion {bIsValid=false, max=<0 0>, min=<0 0>}
    }
   }
   slot {class=WidgetSwitcherSlot, properties {horizontalAlignment=HAlign_Right, verticalAlignment=VAlign_Bottom, padding=<ltrb 16 16 16 16>}}
  }
  Image_Owned {
   widgetClassPath=&Image
   bInherited=true
   bIsVariable=true
   parent=WidgetSwitcher_StatusIcon
   widget=Image_Owned
   properties {
    visibility=Collapsed
    brush {
     bIsDynamicallyLoaded=false
     drawAs=Image
     imageType=NoImage
     mirroring=NoMirror
     resourceName=None
     resourceObject=&M_UI_AngledCheckBox
     tiling=NoTile
     imageSize=<40 40>
     margin=<ltrb 0 0 0 0>
     outlineSettings {
      bUseBrushTransparency=false
      roundingType=HalfHeightRadius
      width=0
      color {colorUseRule=UseColor_Specified, specifiedColor=<rgba 0 0 0 0>}
      cornerRadii=<0 0 0 0>
     }
     tintColor {colorUseRule=UseColor_Specified, specifiedColor=<rgba 1 1 1 1>}
     uVRegion {bIsValid=false, max=<0 0>, min=<0 0>}
    }
   }
   slot {class=WidgetSwitcherSlot, properties {horizontalAlignment=HAlign_Right, verticalAlignment=VAlign_Bottom, padding=<ltrb 16 16 16 16>}}
  }
 ]
}
