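# Category entry widget for the spatial customization frontend, pairing a choice icon with name, state description and input prompt.
# PreConstruct applies icon size and, with bLeftText, mirrors the layout via child reordering, negative render scale, right alignment and a FlipPortrait material parameter.
# Hover plays name, description and directional slide animations plus the choice's own hover, reversing them faster on unhover.
# Category state changes recolor text and the input prompt's rim brush.
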WidgetBlueprint /BRCosmetics/Athena/UI/Frontend/Spatial/Customization/SpatialCustomizationCategoryEntry
animations [
 Hover {
  binding_count=4
  length_seconds=0.2
  object=&self:Hover
  sequence {
   bindings [
    Text_EntryName_Big {
     id=#1
     possessed_class=CommonTextBlock
     tracks [
      {
       type=MovieSceneColorTrack
       display_name=Color and Opacity
       sections [{
         type=MovieSceneColorSection
         channels [
          Color.R {type=MovieSceneScriptingFloatChannel}
          Color.G {type=MovieSceneScriptingFloatChannel}
          Color.B {type=MovieSceneScriptingFloatChannel}
          Color.A {type=MovieSceneScriptingFloatChannel, keys [[0,0],[1,1]]}
         ]
        }]
      }
      {
       type=MovieScene2DTransformTrack
       sections [{
         type=MovieScene2DTransformSection
         channels [
          Scale.X {type=MovieSceneScriptingFloatChannel, keys [[0,0.75],[24,1]]}
          Scale.Y {type=MovieSceneScriptingFloatChannel, keys [[0,0.75],[24,1]]}
         ]
        }]
      }
     ]
    }
    Text_EntryName {
     id=#2
     possessed_class=CommonTextBlock
     tracks [{
       type=MovieSceneColorTrack
       display_name=Color and Opacity
       sections [{
         type=MovieSceneColorSection
         channels [
          Color.R {type=MovieSceneScriptingFloatChannel}
          Color.G {type=MovieSceneScriptingFloatChannel}
          Color.B {type=MovieSceneScriptingFloatChannel}
          Color.A {type=MovieSceneScriptingFloatChannel, keys [[0,1],[1,0]]}
         ]
        }]
      }]
    }
    Border_Desription {
     id=#3
     possessed_class=Border
     tracks [{
       type=MovieSceneColorTrack
       display_name=Content Color and Opacity
       sections [{
         type=MovieSceneColorSection
         channels [
          Color.R {type=MovieSceneScriptingFloatChannel}
          Color.G {type=MovieSceneScriptingFloatChannel}
          Color.B {type=MovieSceneScriptingFloatChannel}
          Color.A {type=MovieSceneScriptingFloatChannel, keys [[6,0],[18,1]]}
         ]
        }]
      }]
    }
    SizeBox_3 {
     id=#4
     possessed_class=SizeBox
     tracks [{
       type=MovieScene2DTransformTrack
       sections [{type=MovieScene2DTransformSection, channels [Translation.Y {type=MovieSceneScriptingFloatChannel, keys [[0,48],[24.002,0]]}]}]
      }]
    }
   ]
   display_rate {numerator=120}
  }
 }
 HoverRight {
  binding_count=2
  length_seconds=0.3
  object=&self:HoverRight
  sequence {bindings [
    Overlay_1 {
     id=#5
     possessed_class=Overlay
     tracks [{
       type=MovieScene2DTransformTrack
       sections [{
         type=MovieScene2DTransformSection
         end_frame=3
         start_frame=0
         channels [
          Translation.X {type=MovieSceneScriptingFloatChannel, keys [[0,0],[3,-16]]}
          Translation.Y {type=MovieSceneScriptingFloatChannel}
          Angle {type=MovieSceneScriptingFloatChannel}
          Scale.X {type=MovieSceneScriptingFloatChannel}
          Scale.Y {type=MovieSceneScriptingFloatChannel}
          Shear.X {type=MovieSceneScriptingFloatChannel}
          Shear.Y {type=MovieSceneScriptingFloatChannel}
         ]
        }]
      }]
    }
    Border_Desription {
     id=#6
     possessed_class=Border
     tracks [{
       type=MovieScene2DTransformTrack
       sections [{type=MovieScene2DTransformSection, channels [Translation.X {type=MovieSceneScriptingFloatChannel, keys [[1,-48],[6,0]]}]}]
      }]
    }
   ]}
 }
 HoverLeft {
  binding_count=2
  length_seconds=0.3
  object=&self:HoverLeft
  sequence {bindings [
    Border_Desription {
     id=#7
     possessed_class=Border
     tracks [{
       type=MovieScene2DTransformTrack
       sections [{type=MovieScene2DTransformSection, channels [Translation.X {type=MovieSceneScriptingFloatChannel, keys [[1,48],[6,0]]}]}]
      }]
    }
    Overlay_1 {
     id=#8
     possessed_class=Overlay
     tracks [{
       type=MovieScene2DTransformTrack
       sections [{
         type=MovieScene2DTransformSection
         end_frame=3
         start_frame=0
         channels [
          Translation.X {type=MovieSceneScriptingFloatChannel, keys [[0,0],[3,16]]}
          Translation.Y {type=MovieSceneScriptingFloatChannel}
          Angle {type=MovieSceneScriptingFloatChannel}
          Scale.X {type=MovieSceneScriptingFloatChannel}
          Scale.Y {type=MovieSceneScriptingFloatChannel}
          Shear.X {type=MovieSceneScriptingFloatChannel}
          Shear.Y {type=MovieSceneScriptingFloatChannel}
         ]
        }]
      }]
    }
   ]}
 }
 UnHover {
  binding_count=5
  length_seconds=0.125
  object=&self:UnHover
  sequence {
   bindings [
    Text_EntryName_Big {
     id=#1
     possessed_class=CommonTextBlock
     tracks [
      {
       type=MovieSceneColorTrack
       display_name=Color and Opacity
       sections [{
         type=MovieSceneColorSection
         channels [
          Color.R {type=MovieSceneScriptingFloatChannel}
          Color.G {type=MovieSceneScriptingFloatChannel}
          Color.B {type=MovieSceneScriptingFloatChannel}
          Color.A {type=MovieSceneScriptingFloatChannel, keys [[10,1],[15,0]]}
         ]
        }]
      }
      {
       type=MovieSceneWidgetMaterialTrack
       display_name=Font.Outline Settings.OutlineMaterial
       sections [{type=MovieSceneParameterSection, channels [Alpha {type=MovieSceneScriptingFloatChannel, const [0,2]}]}]
      }
      {
       type=MovieScene2DTransformTrack
       sections [{
         type=MovieScene2DTransformSection
         channels [
          Translation.X {type=MovieSceneScriptingFloatChannel}
          Translation.Y {type=MovieSceneScriptingFloatChannel}
          Angle {type=MovieSceneScriptingFloatChannel}
          Scale.X {type=MovieSceneScriptingFloatChannel, keys [[0,1],[15.002,0.75]]}
          Scale.Y {type=MovieSceneScriptingFloatChannel, keys [[0,1],[15.002,0.75]]}
          Shear.X {type=MovieSceneScriptingFloatChannel}
          Shear.Y {type=MovieSceneScriptingFloatChannel}
         ]
        }]
      }
      {
       type=MovieSceneColorTrack
       display_name=Shadow Color
       sections [{
         type=MovieSceneColorSection
         channels [
          Color.R {type=MovieSceneScriptingFloatChannel}
          Color.G {type=MovieSceneScriptingFloatChannel}
          Color.B {type=MovieSceneScriptingFloatChannel}
          Color.A {type=MovieSceneScriptingFloatChannel, const [0,2]}
         ]
        }]
      }
      {
       type=MovieSceneFloatVectorTrack
       display_name=Shadow Offset
       sections [{
         type=MovieSceneFloatVectorSection
         channels [
          Vector.X {type=MovieSceneScriptingFloatChannel, const [0,2]}
          Vector.Y {type=MovieSceneScriptingFloatChannel, const [0,2]}
         ]
        }]
      }
      {
       type=MovieSceneWidgetMaterialTrack
       display_name=Font.FontMaterial
       sections [{type=MovieSceneParameterSection, channels [FIllWhite {type=MovieSceneScriptingFloatChannel, const [1,2]}]}]
      }
     ]
    }
    VerticalBox_0 {
     id=#9
     possessed_class=VerticalBox
     tracks [{
       type=MovieSceneFloatTrack
       display_name=Render Opacity
       sections [{type=MovieSceneFloatSection, channels [None {type=MovieSceneScriptingFloatChannel, keys [[0,0],[15.002,1]]}]}]
      }]
    }
    Text_EntryName {
     id=#2
     possessed_class=CommonTextBlock
     tracks [{
       type=MovieSceneColorTrack
       display_name=Color and Opacity
       sections [{
         type=MovieSceneColorSection
         channels [
          Color.R {type=MovieSceneScriptingFloatChannel}
          Color.G {type=MovieSceneScriptingFloatChannel}
          Color.B {type=MovieSceneScriptingFloatChannel}
          Color.A {type=MovieSceneScriptingFloatChannel, keys [[13,0],[15.002,1]]}
         ]
        }]
      }]
    }
    Border_Desription {
     id=#3
     possessed_class=Border
     tracks [{
       type=MovieSceneColorTrack
       display_name=Content Color and Opacity
       sections [{
         type=MovieSceneColorSection
         channels [
          Color.R {type=MovieSceneScriptingFloatChannel}
          Color.G {type=MovieSceneScriptingFloatChannel}
          Color.B {type=MovieSceneScriptingFloatChannel}
          Color.A {type=MovieSceneScriptingFloatChannel, keys [[0,1],[15.002,0]]}
         ]
        }]
      }]
    }
    SizeBox_3 {
     id=#4
     possessed_class=SizeBox
     tracks [{
       type=MovieSceneFloatTrack
       display_name=Height Override
       sections [{type=MovieSceneFloatSection, channels [None {type=MovieSceneScriptingFloatChannel, keys [[0,50],[15,32]]}]}]
      }]
    }
   ]
   display_rate {numerator=120}
  }
 }
]
graphs [
 EventGraph {nodes [
   n1 = Event OnCategoryStateSet!() -> (OutputDelegate, then, State:ESpatialCustomizationCategoryState Enum=LockedByBattlePass, StateColor:Linear Color Structure) {is=title K2Node_Event}
   > HandleCategoryStateSet!(execute, self, State:ESpatialCustomizationCategoryState Enum=@n1.State, StateColor:Linear Color Structure=@n1.StateColor) -> (then)
   Event PreConstruct()
   > n2 = Cast To SpatialCustomizationChoice!(execute, Object=Get Choice_SelectedChoice() -> (Choice_SelectedChoice:Athena Spatial Customization Choice Object Reference)) -> (then, CastFailed, AsSpatial Customization Choice:Spatial Customization Choice Object Reference) {is=title K2Node_DynamicCast}
   > UseSmallerIcon!(execute, self:Spatial Customization Choice Object Reference=@n2, bUseSmallIcon:bool=Get bUseSmallerIcon() -> (bUseSmallerIcon:bool)) -> (then)
   > Branch(Condition=Get bLeftText() -> (bLeftText:bool))
   > AddChild(self=Get HBox_EntryContents() -> (HBox_EntryContents:Horizontal Box Object Reference), Content=Get Choice_SelectedChoice())
   > SetRenderTransformPivot!(
    execute
    self:Widget Object Reference=Get Text_EntryName_Big() -> (Text_EntryName_Big:Common Text Object Reference)
    Pivot:Vector 2D Structure=(X=1,Y=0)
   ) -> (then)
   n3 = Get Choice_SelectedChoice()
   > SetRenderScale!(execute, self:Widget Object Reference=@n3, Scale:Vector 2D Structure=(X=-1,Y=1)) -> (then)
   > n4 = Cast To SpatialCustomizationChoice(Object=@n3)
   > SetRenderScale(
    self=Get WidgetSwitcher_StatusIcon!(self:Spatial Customization Choice Object Reference=@n4) -> (WidgetSwitcher_StatusIcon:Widget Switcher Object Reference)
    Scale=(X=-1,Y=1)
   )
   > SetHorizontalAlignment(
    self=SlotAsOverlaySlot(Widget=Get ControllerPrompt() -> (ControllerPrompt:Vertical Box Object Reference))
    InHorizontalAlignment=HAlign_Right
   )
   > SetHorizontalAlignment(
    self:Border Slot Object Reference=SlotAsBorderSlot!(Widget:Widget Object Reference=Get ChannelName_Color() -> (ChannelName_Color:Border Object Reference)) -> (ReturnValue:Border Slot Object Reference)
    InHorizontalAlignment=HAlign_Right
   )
   > Set Justification!(
    execute
    Justification:ETextJustify Enum=Right
    self:Text Layout Widget Object Reference=Get Text_EntryName() -> (Text_EntryName:Common Text Object Reference)
   ) -> (then, Output_Get:ETextJustify Enum=Left)
   > Set Justification(Justification=Right, self=Get Text_EntryName_Big()) -> (Output_Get=Left)
   > Set Justification(Justification=Right, self=Get RichText_StateDescription() -> (RichText_StateDescription:Common Rich Text Block Object Reference)) -> (Output_Get=Left)
   > n5 = Cast To SpatialCustomizationChoice(Object=Get Choice_SelectedChoice())
   > n6 = GetDynamicMaterial(self=Get Image_ItemIcon!(self:Athena Spatial Customization Choice Object Reference=@n5) -> (Image_ItemIcon:Fort Lazy Image Object Reference))
   > SetScalarParameterValue(self=@n6, ParameterName=FlipPortrait, Value=1)
   Event On Hovered()
   > PlayAnimationForward(InAnimation=Get Hover() -> (Hover:Widget Animation Object Reference))
   > n7 = Cast To SpatialCustomizationChoice(Object=Get Choice_SelectedChoice())
   > n8 = Play hover alternate(self:Spatial Customization Choice Object Reference=@n7)
   n9 = Branch(execute=[@n8.then, @n7.CastFailed], Condition=Get bLeftText())
   > PlayAnimationForward(InAnimation=Get HoverLeft() -> (HoverLeft:Widget Animation Object Reference))
   PlayAnimationForward(execute=@n9.else, InAnimation=Get HoverRight() -> (HoverRight:Widget Animation Object Reference))
   Event On Unhovered()
   > PlayAnimationReverse(InAnimation=Get Hover(), PlaybackSpeed=3)
   > n10 = Cast To SpatialCustomizationChoice(Object=Get Choice_SelectedChoice())
   > n11 = Play reversed hover alternate(self:Spatial Customization Choice Object Reference=@n10)
   n12 = Branch(execute=[@n11.then, @n10.CastFailed], Condition=Get bLeftText())
   > PlayAnimationReverse(InAnimation=Get HoverLeft(), PlaybackSpeed=2)
   PlayAnimationReverse(execute=@n12.else, InAnimation=Get HoverRight(), PlaybackSpeed=2)
  ]}
 HandleCategoryStateSet {nodes [
   n1 = HandleCategoryStateSet!() -> (then, State:ESpatialCustomizationCategoryState Enum=LockedByBattlePass, StateColor:Linear Color Structure) {is=title K2Node_FunctionEntry}
   > "SetText (Text)"(self=Get Text_EntryName_Big(), InText="GetText (Text)"!(self:Text Object Reference=Get Text_EntryName()) -> (ReturnValue:Text))
   n2 = Make Slate Color(SpecifiedColor=(R=0.1,G=0.4,B=1,A=1))
   n3 = BreakColor(InColor=Get StateColor() -> (StateColor:Linear Color Structure))
   n4 = Make Linear Color!(R:float=@n3, G:float=@n3.G, B:float=@n3.B, A:float=1) -> (LinearColor:Linear Color Structure) {is=title K2Node_MakeStruct}
   n5 = Make Slate Color(SpecifiedColor=@n4)
   n6 = Make Slate Color(SpecifiedColor=(R=1,G=1,B=1,A=1))
   > SetColorAndOpacity(
    self=Get Text_EntryName()
    InColorAndOpacity=Select!(
     LockedByBattlePass:Slate Color Structure=@n2
     LockedByChallenge:Slate Color Structure=@n2
     LockedByChallengeCompletion:Slate Color Structure=@n5
     UnlockAvailable:Slate Color Structure=@n6
     UnlockUsed:Slate Color Structure=@n6
     Index:ESpatialCustomizationCategoryState Enum=@n1
    ) -> (ReturnValue:Slate Color Structure)
   )
   > SetDefaultColorAndOpacity!(
    execute
    self:Rich Text Block Object Reference=Get RichText_StateDescription()
    InColorAndOpacity:Slate Color Structure=Select(LockedByBattlePass=@n2, LockedByChallenge=@n2, LockedByChallengeCompletion=@n5, UnlockAvailable=@n5, UnlockUsed=@n5, Index=@n1)
   ) -> (then)
   > Set rim brush color!(
    execute
    self
    target:Common Action Widget Object Reference=Get InputActionWidget() -> (InputActionWidget:Common Action Widget Object Reference)
    color:Linear Color Structure=@n4
   ) -> (then) {is=title K2Node_CallFunction}
  ]}
 Set rim brush color {nodes [
   n1 = Set rim brush color!() -> (then, target:Common Action Widget Object Reference, color:Linear Color Structure) {is=title K2Node_FunctionEntry}
   n2 = Break Slate Brush!(SlateBrush:"Slate Brush Structure (by ref)"=Get IconRimBrush!(self:Common Action Widget Object Reference=@n1) -> (IconRimBrush:Slate Brush Structure)) -> (
    TintColor:Slate Color Structure
    DrawAs:ESlateBrushDrawType Enum
    Tiling:ESlateBrushTileType Enum
    Mirroring:ESlateBrushMirrorType Enum
    ImageSize:Vector 2D Structure
    Margin:Margin Structure
    ResourceObject:Object Reference
    OutlineSettings:Slate Brush Outline Settings Structure
   ) {is=title K2Node_BreakStruct}
   n3 = Break Slate Color!(SlateColor:"Slate Color Structure (by ref)"=@n2) -> (SpecifiedColor:Linear Color Structure, ColorUseRule:ESlateColorStylingMode Enum) {is=title K2Node_BreakStruct}
   > SetIconRimBrush!(
    execute
    self:Common Action Widget Object Reference=@n1
    InIconRimBrush:Slate Brush Structure=Make Slate Brush!(
     TintColor:Slate Color Structure=Make Slate Color(SpecifiedColor=@n1.color, ColorUseRule=@n3.ColorUseRule)
     DrawAs:ESlateBrushDrawType Enum=@n2.DrawAs
     Tiling:ESlateBrushTileType Enum=@n2.Tiling
     Mirroring:ESlateBrushMirrorType Enum=@n2.Mirroring
     ImageSize:Vector 2D Structure=@n2.ImageSize
     Margin:Margin Structure=@n2.Margin
     ResourceObject:Object Reference=@n2.ResourceObject
     OutlineSettings:Slate Brush Outline Settings Structure=(CornerRadii=(X=0,Y=0,Z=0,W=0),Color=(SpecifiedColor=(R=0,G=0,B=0,A=0),ColorUseRule=UseColor_Specified),Width=0,RoundingType=HalfHeightRadius,bUseBrushTransparency=False)
    ) -> (SlateBrush:Slate Brush Structure) {is=title K2Node_MakeStruct}
   ) -> (then)
  ]}
]
variables [
 bLeftText {type=boolean}
 bUseSmallerIcon {type=boolean}
]
widgets {
 info {
  inheritedWidgetCount=5
  parentClass=/Script/FortniteUI.AthenaSpatialCustomizationCategoryEntry
  rootWidgetClass=&Overlay
  widgetCount=15
 }
 widgets [
  Overlay_1 {widgetClassPath=&Overlay, widget=Overlay_1}
  SizeBox_0 {
   widgetClassPath=&SizeBox
   parent=Overlay_1
   widget=SizeBox_0
   properties {bOverride_HeightOverride=true, bOverride_WidthOverride=true, heightOverride=128, widthOverride=512}
   slot {class=OverlaySlot, properties {horizontalAlignment=HAlign_Fill, verticalAlignment=VAlign_Fill}}
  }
  HBox_EntryContents {widgetClassPath=&HorizontalBox, bIsVariable=true, parent=SizeBox_0, widget=HBox_EntryContents, slot {class=SizeBoxSlot}}
  Choice_SelectedChoice {
   widgetClassPath=&/SpatialCustomizationChoice
   bInherited=true
   bIsVariable=true
   parent=HBox_EntryContents
   widget=Choice_SelectedChoice
   properties {showStatusBar=true}
   slot {class=HorizontalBoxSlot}
  }
  VerticalBox_2 {
   widgetClassPath=&VerticalBox
   bIsVariable=true
   parent=HBox_EntryContents
   widget=VerticalBox_2
   slot {class=HorizontalBoxSlot, properties {padding=<ltrb 10 0 10 0>, size {sizeRule=Fill, value=1}}}
  }
  SizeBox_3 {
   widgetClassPath=&SizeBox
   parent=VerticalBox_2
   widget=SizeBox_3
   properties {bOverride_HeightOverride=true, heightOverride=50, renderTransform {angle=0, scale=<1 1>, shear=<0 0>, translation=<0 48>}}
   slot {class=VerticalBoxSlot, properties {verticalAlignment=VAlign_Top}}
  }
  ChannelName_Color {
   widgetClassPath=&Border
   bIsVariable=true
   parent=SizeBox_3
   widget=ChannelName_Color
   properties {
    background {
     bIsDynamicallyLoaded=false
     drawAs=NoDrawType
     imageType=NoImage
     mirroring=NoMirror
     resourceName=None
     resourceObject=null
     tiling=NoTile
     imageSize=<32 32>
     margin=<ltrb 0 0 0 0>
     outlineSettings {
      bUseBrushTransparency=false
      roundingType=HalfHeightRadius
      width=0
      color {colorUseRule=UseColor_Specified, specifiedColor=<rgba 0 0 0 0>}
      cornerRadii=<0 0 0 0>
     }
     tintColor {colorUseRule=UseColor_Specified, specifiedColor=<rgba 1 1 1 1>}
     uVRegion {bIsValid=false, max=<0 0>, min=<0 0>}
    }
    padding=<ltrb 0 0 0 0>
   }
   slot {class=SizeBoxSlot}
  }
  Overlay_0 {
   widgetClassPath=&Overlay
   parent=ChannelName_Color
   widget=Overlay_0
   slot {class=BorderSlot, properties {padding=<ltrb 0 0 0 0>}}
  }
  Text_EntryName {
   widgetClassPath=&CommonTextBlock
   bInherited=true
   bIsVariable=true
   parent=Overlay_0
   widget=Text_EntryName
   properties {
    mobileFontSizeMultiplier=1.25
    text=Channel Name
    textTransformPolicy=ToUpper
    font {
     bForceMonospaced=false
     bMaterialIsStencil=false
     fontMaterial=null
     fontObject=&Burbank
     letterSpacing=0
     monospacedWidth=1
     size=20
     skewAmount=0
     typefaceFontName=BurbankRegular-Black
     outlineSettings {
      bApplyOutlineToDropShadows=false
      bMiteredCorners=false
      bSeparateFillAlpha=false
      outlineMaterial=null
      outlineSize=3
      outlineColor=<rgba 0 0 0 0.0375>
     }
    }
    renderTransformPivot=<0 0>
    shadowOffset=<0 0>
   }
   slot {class=OverlaySlot, properties {horizontalAlignment=HAlign_Fill}}
  }
  Text_EntryName_Big {
   widgetClassPath=&CommonTextBlock
   bInherited=true
   bIsVariable=true
   parent=Overlay_0
   widget=Text_EntryName_Big
   properties {
    mobileFontSizeMultiplier=1.25
    scrollStyle=&TextScrollStyle-Base_C
    text=Channel Name
    textTransformPolicy=ToUpper
    colorAndOpacity {colorUseRule=UseColor_Specified, specifiedColor=<rgba 1 1 1 0>}
    font {
     bForceMonospaced=false
     bMaterialIsStencil=false
     fontMaterial=null
     fontObject=&Burbank
     letterSpacing=0
     monospacedWidth=1
     size=28
     skewAmount=0
     typefaceFontName=BurbankRegular-Black
     outlineSettings {
      bApplyOutlineToDropShadows=false
      bMiteredCorners=false
      bSeparateFillAlpha=false
      outlineMaterial=null
      outlineSize=0
      outlineColor=<rgba 1 1 1 1>
     }
    }
    renderTransform {angle=0, scale=<0.65 0.65>, shear=<0 0>, translation=<0 0>}
    renderTransformPivot=<0 0>
    shadowColorAndOpacity=<rgba 0.5 0.5 0.5 0>
    shadowOffset=<0 0>
   }
   slot {class=OverlaySlot, properties {horizontalAlignment=HAlign_Fill}}
  }
  Overlay_3 {
   widgetClassPath=&Overlay
   parent=VerticalBox_2
   widget=Overlay_3
   slot {class=VerticalBoxSlot, properties {padding=<ltrb 0 4 0 0>, size {sizeRule=Fill, value=1}}}
  }
  Border_Desription {
   widgetClassPath=&Border
   bIsVariable=true
   parent=Overlay_3
   widget=Border_Desription
   properties {
    background {
     bIsDynamicallyLoaded=false
     drawAs=NoDrawType
     imageType=NoImage
     mirroring=NoMirror
     resourceName=None
     resourceObject=null
     tiling=NoTile
     imageSize=<32 32>
     margin=<ltrb 0 0 0 0>
     outlineSettings {
      bUseBrushTransparency=false
      roundingType=HalfHeightRadius
      width=0
      color {colorUseRule=UseColor_Specified, specifiedColor=<rgba 0 0 0 0>}
      cornerRadii=<0 0 0 0>
     }
     tintColor {colorUseRule=UseColor_Specified, specifiedColor=<rgba 1 1 1 1>}
     uVRegion {bIsValid=false, max=<0 0>, min=<0 0>}
    }
    contentColorAndOpacity=<rgba 1 1 1 0>
    padding=<ltrb 0 0 0 0>
   }
   slot {class=OverlaySlot, properties {horizontalAlignment=HAlign_Fill}}
  }
  RichText_StateDescription {
   widgetClassPath=&CommonRichTextBlock
   bInherited=true
   bIsVariable=true
   parent=Border_Desription
   widget=RichText_StateDescription
   properties {
    autoWrapText=true
    bOverrideDefaultStyle=true
    lineHeightPercentage=0.85
    text="AVAILABLE WITH <text color=\"FFFFFFFF\">WEEK 6</> CHALLENGES"
    textTransformPolicy=ToUpper
    defaultTextStyleOverride {
     overflowPolicy=Clip
     transformPolicy=None
     colorAndOpacity {colorUseRule=UseColor_Specified, specifiedColor=<rgba 1 1 1 0.5>}
     font {
      bForceMonospaced=false
      bMaterialIsStencil=false
      fontMaterial=null
      fontObject=&BurbankSmall
      letterSpacing=0
      monospacedWidth=1
      size=15
      skewAmount=0
      typefaceFontName=Black
      outlineSettings {
       bApplyOutlineToDropShadows=false
       bMiteredCorners=false
       bSeparateFillAlpha=false
       outlineMaterial=null
       outlineSize=0
       outlineColor=<rgba 0 0 0 1>
      }
     }
     highlightColor {colorUseRule=UseColor_Specified, specifiedColor=<rgba 0 0 0 1>}
     highlightShape {
      bIsDynamicallyLoaded=false
      drawAs=Image
      imageType=NoImage
      mirroring=NoMirror
      resourceName=None
      resourceObject=null
      tiling=NoTile
      imageSize=<32 32>
      margin=<ltrb 0 0 0 0>
      outlineSettings {
       bUseBrushTransparency=false
       roundingType=HalfHeightRadius
       width=0
       color {colorUseRule=UseColor_Specified, specifiedColor=<rgba 0 0 0 0>}
       cornerRadii=<0 0 0 0>
      }
      tintColor {colorUseRule=UseColor_Specified, specifiedColor=<rgba 1 1 1 1>}
      uVRegion {bIsValid=false, max=<0 0>, min=<0 0>}
     }
     selectedBackgroundColor {colorUseRule=UseColor_Foreground, specifiedColor=<rgba 1 0 1 1>}
     shadowColorAndOpacity=<rgba 0 0 0 1>
     shadowOffset=<0 0>
     strikeBrush {
      bIsDynamicallyLoaded=false
      drawAs=Image
      imageType=NoImage
      mirroring=NoMirror
      resourceName=None
      resourceObject=null
      tiling=NoTile
      imageSize=<32 32>
      margin=<ltrb 0 0 0 0>
      outlineSettings {
       bUseBrushTransparency=false
       roundingType=HalfHeightRadius
       width=0
       color {colorUseRule=UseColor_Specified, specifiedColor=<rgba 0 0 0 0>}
       cornerRadii=<0 0 0 0>
      }
      tintColor {colorUseRule=UseColor_Specified, specifiedColor=<rgba 1 1 1 1>}
      uVRegion {bIsValid=false, max=<0 0>, min=<0 0>}
     }
     underlineBrush {
      bIsDynamicallyLoaded=false
      drawAs=Image
      imageType=NoImage
      mirroring=NoMirror
      resourceName=None
      resourceObject=null
      tiling=NoTile
      imageSize=<32 32>
      margin=<ltrb 0 0 0 0>
      outlineSettings {
       bUseBrushTransparency=false
       roundingType=HalfHeightRadius
       width=0
       color {colorUseRule=UseColor_Specified, specifiedColor=<rgba 0 0 0 0>}
       cornerRadii=<0 0 0 0>
      }
      tintColor {colorUseRule=UseColor_Specified, specifiedColor=<rgba 1 1 1 1>}
      uVRegion {bIsValid=false, max=<0 0>, min=<0 0>}
     }
    }
    margin=<ltrb 0 -6 0 0>
   }
   slot {class=BorderSlot, properties {padding=<ltrb 0 0 0 0>}}
  }
  ControllerPrompt {
   widgetClassPath=&VerticalBox
   bIsVariable=true
   parent=Overlay_1
   widget=ControllerPrompt
   slot {class=OverlaySlot, properties {verticalAlignment=VAlign_Center, padding=<ltrb -22 0 -22 0>}}
  }
  InputActionWidget {
   widgetClassPath=&CommonActionWidget
   bInherited=true
   bIsVariable=true
   parent=ControllerPrompt
   widget=InputActionWidget
   properties {iconRimBrush {
     bIsDynamicallyLoaded=false
     drawAs=Image
     imageType=NoImage
     mirroring=NoMirror
     resourceName=None
     resourceObject=/FortUILibrary/Foundation/Buttons/Materials/M-Button-Console-WhiteOutline
     tiling=NoTile
     imageSize=<40 40>
     margin=<ltrb 0 0 0 0>
     outlineSettings {
      bUseBrushTransparency=false
      roundingType=HalfHeightRadius
      width=0
      color {colorUseRule=UseColor_Specified, specifiedColor=<rgba 0 0 0 0>}
      cornerRadii=<0 0 0 0>
     }
     tintColor {colorUseRule=UseColor_Specified, specifiedColor=<rgba 1 1 1 1>}
     uVRegion {bIsValid=false, max=<0 0>, min=<0 0>}
    }}
   slot {class=VerticalBoxSlot}
  }
 ]
}
