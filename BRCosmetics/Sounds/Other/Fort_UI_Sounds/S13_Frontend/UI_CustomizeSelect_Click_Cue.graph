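# UI click sound cue for a customize-select action in the Season 13 frontend, assigned to the Fort_UI sound class.
# A mixer layers the main frontend button click with a challenge-table hover sound, which a short fixed delay offsets behind the click.
# A concurrency override limits the cue to one instance and stops the oldest, taking precedence over the assigned frontend menu concurrency set.
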SoundCue /BRCosmetics/Sounds/Other/Fort_UI_Sounds/S13_Frontend/UI_CustomizeSelect_Click_Cue
root=Mixer_0
nodes [
 {id=Delay_0, children [WavePlayer_6], properties {delayMax=0.1, delayMin=0.1}}
 {id=Mixer_0, children [WavePlayer_1,Delay_0], properties {inputVolume [1,1]}}
 {id=WavePlayer_1, properties {soundWaveAssetPtr=/Game/Sounds/Fort_UI_Sounds/FrontEnd/New_FrontEnd/UI_FN_Button_Click_Main_01}}
 {id=WavePlayer_6, properties {soundWaveAssetPtr=/Game/Sounds/Fort_UI_Sounds/S13_Frontend/S13_Challenge_Table_Hover}}
]
properties {
 bOverrideConcurrency=true
 maxDistance=2097152
 soundClassObject=&Fort_UI
 volumeMultiplier=1
 concurrencyOverrides {maxCount=1, resolutionRule=StopOldest}
 concurrencySet [/Game/Sounds/Concurrency_Rules/FrontEnd_UI_Menu_Max_03_Stop_Oldest]
}
