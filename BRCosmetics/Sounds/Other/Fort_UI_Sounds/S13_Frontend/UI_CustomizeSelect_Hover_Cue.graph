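# UI sound cue that, by its name, plays on hover over a customization selection in the frontend.
# A mixer layers a button hover click, a table-back sound pitched up through its own modulator, and a map pop-up sound at staggered volumes.
# A root modulator leaves pitch at its default random range, and the cue raises overall pitch.
# It uses the Fort_UI sound class and a shared frontend menu stop-oldest concurrency rule.
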SoundCue /BRCosmetics/Sounds/Other/Fort_UI_Sounds/S13_Frontend/UI_CustomizeSelect_Hover_Cue
root=Modulator_0
nodes [
 {id=Mixer_0, children [WavePlayer_2,Modulator_2,WavePlayer_6], properties {inputVolume [0.75,0.5,1]}}
 {id=Modulator_0, children [Mixer_0], properties {volumeMax=1, volumeMin=1}}
 {id=Modulator_2, children [WavePlayer_5], properties {pitchMax=2, pitchMin=2}}
 {id=WavePlayer_2, properties {soundWaveAssetPtr=/Game/Sounds/Fort_UI_Sounds/FrontEnd/New_FrontEnd/UI_FN_Button_Hover_Click_02}}
 {id=WavePlayer_5, properties {soundWaveAssetPtr=/BRCosmetics/Sounds/Other/Fort_UI_Sounds/S12_Frontend/UI_S12_BP_Table_Back}}
 {id=WavePlayer_6, properties {soundWaveAssetPtr=/BRCosmetics/Sounds/Other/Fort_UI_Sounds/S12_Frontend/UI_S12_Challenge_Map_PopUps}}
]
properties {
 maxDistance=2097152
 pitchMultiplier=1.5
 priority=0.9
 soundClassObject=&Fort_UI
 volumeMultiplier=1
 concurrencyOverrides {maxCount=1, resolutionRule=StopOldest}
 concurrencySet [/Game/Sounds/Concurrency_Rules/FrontEnd_UI_Menu_Max_03_Stop_Oldest]
}
