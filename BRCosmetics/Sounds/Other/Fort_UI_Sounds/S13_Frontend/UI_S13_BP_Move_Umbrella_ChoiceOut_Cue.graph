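# Sound cue for a Season 13 frontend battle pass UI transition, apparently when the build-an-umbrella choice screen wipes out.
# A single wave player plays the Builldabrella choice wipe-out sound with no randomization or modulation.
# It routes to the Fort_UI sound class with a slight volume boost.
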SoundCue /BRCosmetics/Sounds/Other/Fort_UI_Sounds/S13_Frontend/UI_S13_BP_Move_Umbrella_ChoiceOut_Cue
root=WavePlayer_0
nodes [{id=WavePlayer_0, properties {soundWaveAssetPtr=&/S13_Builldabrella_Choice_Wipe_Out}}]
properties {maxDistance=2097152, soundClassObject=&Fort_UI, volumeMultiplier=1.25}
