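# Season 13 umbrella customization screen widget showing ten category entries in Battle Pass (left) and Quests (right) columns.
# On initialization it registers both groups and maps their headers and entry counts, so progress events hide completed groups' unlock headers.
# Setup events drive percent-complete text and a material-based progress bar.
# Activation plays a staggered slide-in intro with direction-dependent sound; capture mode hides title and headers.
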WidgetBlueprint /BRCosmetics/Athena/UI/Frontend/Spatial/Customization/SpatialCustomizationCategories
animations [IntroCategories {
  binding_count=16
  length_seconds=0.4
  object=&self:IntroCategories
  sequence {
   bindings [
    CategoryEntry_4 {
     id=#1
     possessed_class=SpatialCustomizationCategoryEntry_C
     tracks [{
       type=MovieScene2DTransformTrack
       sections [{
         type=MovieScene2DTransformSection
         channels [
          Translation.X {type=MovieSceneScriptingFloatChannel, keys [[1,-1024],[16,0]]}
          Translation.Y {type=MovieSceneScriptingFloatChannel}
          Angle {type=MovieSceneScriptingFloatChannel}
          Scale.X {type=MovieSceneScriptingFloatChannel}
          Scale.Y {type=MovieSceneScriptingFloatChannel}
          Shear.X {type=MovieSceneScriptingFloatChannel}
          Shear.Y {type=MovieSceneScriptingFloatChannel}
         ]
        }]
      }]
    }
    Image_149 {
     id=#2
     possessed_class=Image
     tracks [{
       type=MovieSceneColorTrack
       display_name=Color and Opacity
       sections [{
         type=MovieSceneColorSection
         channels [
          Color.R {type=MovieSceneScriptingFloatChannel}
          Color.G {type=MovieSceneScriptingFloatChannel}
          Color.B {type=MovieSceneScriptingFloatChannel}
          Color.A {type=MovieSceneScriptingFloatChannel, keys [[0,0],[24.001,1]]}
         ]
        }]
      }]
    }
    CommonTextBlock {
     id=#3
     possessed_class=CommonTextBlock
     tracks [{
       type=MovieSceneFloatTrack
       display_name=Render Opacity
       sections [{type=MovieSceneFloatSection, end_frame=20, start_frame=0, channels [None {type=MovieSceneScriptingFloatChannel, keys [[12,0],[20,1]]}]}]
      }]
    }
    CategoryEntry_1 {
     id=#4
     possessed_class=SpatialCustomizationCategoryEntry_C
     tracks [{
       type=MovieScene2DTransformTrack
       sections [{
         type=MovieScene2DTransformSection
         channels [
          Translation.X {type=MovieSceneScriptingFloatChannel, keys [[4,-1024],[19,0]]}
          Translation.Y {type=MovieSceneScriptingFloatChannel}
          Angle {type=MovieSceneScriptingFloatChannel}
          Scale.X {type=MovieSceneScriptingFloatChannel}
          Scale.Y {type=MovieSceneScriptingFloatChannel}
          Shear.X {type=MovieSceneScriptingFloatChannel}
          Shear.Y {type=MovieSceneScriptingFloatChannel}
         ]
        }]
      }]
    }
    CommonTextBlock_76 {
     id=#5
     possessed_class=CommonTextBlock
     tracks [{
       type=MovieSceneFloatTrack
       display_name=Render Opacity
       sections [{type=MovieSceneFloatSection, end_frame=20, start_frame=0, channels [None {type=MovieSceneScriptingFloatChannel, keys [[12,0],[20,1]]}]}]
      }]
    }
    Text_title {
     id=#6
     possessed_class=CommonTextBlock
     tracks [
      {
       type=MovieSceneWidgetMaterialTrack
       display_name=Font.FontMaterial
       sections [{
         type=MovieSceneParameterSection
         channels [
          DistortionAmount {type=MovieSceneScriptingFloatChannel, keys [[10,0.05],[20,0.001]]}
          ScanLines_Opacity {type=MovieSceneScriptingFloatChannel, keys [[0,0.2],[10,1],[20,0.25]]}
          ScanLines_Speed {type=MovieSceneScriptingFloatChannel, keys [[0,0.8],[20,0.5]]}
          Squares_OpacityMin {type=MovieSceneScriptingFloatChannel, keys [[0,0],[5,1],[10,0],[20,1]]}
          ForceAlphaChange {type=MovieSceneScriptingFloatChannel, const [0.5,2]}
         ]
        }]
      }
      {
       type=MovieSceneWidgetMaterialTrack
       display_name=Font.Outline Settings.OutlineMaterial
       sections [{
         type=MovieSceneParameterSection
         channels [
          DistortionAmount {type=MovieSceneScriptingFloatChannel, keys [[0,0.05],[10,0.29],[20,0.001]]}
          ForceAlphaChange {type=MovieSceneScriptingFloatChannel, keys [[10,0.5],[20,0.15]]}
         ]
        }]
      }
     ]
    }
    CategoryEntry_3 {
     id=#7
     possessed_class=SpatialCustomizationCategoryEntry_C
     tracks [{
       type=MovieScene2DTransformTrack
       sections [{
         type=MovieScene2DTransformSection
         channels [
          Translation.X {type=MovieSceneScriptingFloatChannel, keys [[2,-1024],[17,0]]}
          Translation.Y {type=MovieSceneScriptingFloatChannel}
          Angle {type=MovieSceneScriptingFloatChannel}
          Scale.X {type=MovieSceneScriptingFloatChannel}
          Scale.Y {type=MovieSceneScriptingFloatChannel}
          Shear.X {type=MovieSceneScriptingFloatChannel}
          Shear.Y {type=MovieSceneScriptingFloatChannel}
         ]
        }]
      }]
    }
    CategoryEntry_5 {
     id=#8
     possessed_class=SpatialCustomizationCategoryEntry_C
     tracks [{
       type=MovieScene2DTransformTrack
       sections [{
         type=MovieScene2DTransformSection
         channels [
          Translation.X {type=MovieSceneScriptingFloatChannel, keys [[0,-1024],[15,0]]}
          Translation.Y {type=MovieSceneScriptingFloatChannel}
          Angle {type=MovieSceneScriptingFloatChannel}
          Scale.X {type=MovieSceneScriptingFloatChannel}
          Scale.Y {type=MovieSceneScriptingFloatChannel}
          Shear.X {type=MovieSceneScriptingFloatChannel}
          Shear.Y {type=MovieSceneScriptingFloatChannel}
         ]
        }]
      }]
    }
    Overlay_Title {
     id=#9
     possessed_class=Overlay
     tracks [{
       type=MovieScene2DTransformTrack
       sections [{
         type=MovieScene2DTransformSection
         channels [
          Translation.X {type=MovieSceneScriptingFloatChannel}
          Translation.Y {type=MovieSceneScriptingFloatChannel}
          Angle {type=MovieSceneScriptingFloatChannel}
          Scale.X {type=MovieSceneScriptingFloatChannel, keys [[0,0],[7,1]]}
          Scale.Y {type=MovieSceneScriptingFloatChannel, keys [[0,0],[7,1]]}
         ]
        }]
      }]
    }
    CategoryEntry_2 {
     id=#10
     possessed_class=SpatialCustomizationCategoryEntry_C
     tracks [{
       type=MovieScene2DTransformTrack
       sections [{
         type=MovieScene2DTransformSection
         channels [
          Translation.X {type=MovieSceneScriptingFloatChannel, keys [[3,-1024],[18,0]]}
          Translation.Y {type=MovieSceneScriptingFloatChannel}
          Angle {type=MovieSceneScriptingFloatChannel}
          Scale.X {type=MovieSceneScriptingFloatChannel}
          Scale.Y {type=MovieSceneScriptingFloatChannel}
          Shear.X {type=MovieSceneScriptingFloatChannel}
          Shear.Y {type=MovieSceneScriptingFloatChannel}
         ]
        }]
      }]
    }
    Image {
     id=#11
     possessed_class=Image
     tracks [
      {
       type=MovieSceneColorTrack
       display_name=Color and Opacity
       sections [{
         type=MovieSceneColorSection
         channels [
          Color.R {type=MovieSceneScriptingFloatChannel}
          Color.G {type=MovieSceneScriptingFloatChannel}
          Color.B {type=MovieSceneScriptingFloatChannel}
          Color.A {type=MovieSceneScriptingFloatChannel, keys [[0,0],[24,1]]}
         ]
        }]
      }
      {type=MovieSceneWidgetMaterialTrack, sections [{type=MovieSceneParameterSection, end_frame=20, start_frame=0}]}
     ]
    }
    CategoryEntry_9 {
     id=#12
     possessed_class=SpatialCustomizationCategoryEntry_C
     tracks [{
       type=MovieScene2DTransformTrack
       sections [{
         type=MovieScene2DTransformSection
         channels [
          Translation.X {type=MovieSceneScriptingFloatChannel, keys [[1,1004],[16,0]]}
          Translation.Y {type=MovieSceneScriptingFloatChannel}
          Angle {type=MovieSceneScriptingFloatChannel}
          Scale.X {type=MovieSceneScriptingFloatChannel}
          Scale.Y {type=MovieSceneScriptingFloatChannel}
          Shear.X {type=MovieSceneScriptingFloatChannel}
          Shear.Y {type=MovieSceneScriptingFloatChannel}
         ]
        }]
      }]
    }
    CategoryEntry_8 {
     id=#13
     possessed_class=SpatialCustomizationCategoryEntry_C
     tracks [{
       type=MovieScene2DTransformTrack
       sections [{
         type=MovieScene2DTransformSection
         channels [
          Translation.X {type=MovieSceneScriptingFloatChannel, keys [[2,1024],[17,0]]}
          Translation.Y {type=MovieSceneScriptingFloatChannel}
          Angle {type=MovieSceneScriptingFloatChannel}
          Scale.X {type=MovieSceneScriptingFloatChannel}
          Scale.Y {type=MovieSceneScriptingFloatChannel}
          Shear.X {type=MovieSceneScriptingFloatChannel}
          Shear.Y {type=MovieSceneScriptingFloatChannel}
         ]
        }]
      }]
    }
    CategoryEntry_10 {
     id=#14
     possessed_class=SpatialCustomizationCategoryEntry_C
     tracks [{
       type=MovieScene2DTransformTrack
       sections [{
         type=MovieScene2DTransformSection
         channels [
          Translation.X {type=MovieSceneScriptingFloatChannel, keys [[0,1024],[15,0]]}
          Translation.Y {type=MovieSceneScriptingFloatChannel}
          Angle {type=MovieSceneScriptingFloatChannel}
          Scale.X {type=MovieSceneScriptingFloatChannel}
          Scale.Y {type=MovieSceneScriptingFloatChannel}
          Shear.X {type=MovieSceneScriptingFloatChannel}
          Shear.Y {type=MovieSceneScriptingFloatChannel}
         ]
        }]
      }]
    }
    CategoryEntry_7 {
     id=#15
     possessed_class=SpatialCustomizationCategoryEntry_C
     tracks [{
       type=MovieScene2DTransformTrack
       sections [{
         type=MovieScene2DTransformSection
         channels [
          Translation.X {type=MovieSceneScriptingFloatChannel, keys [[3,1024],[18,0]]}
          Translation.Y {type=MovieSceneScriptingFloatChannel}
          Angle {type=MovieSceneScriptingFloatChannel}
          Scale.X {type=MovieSceneScriptingFloatChannel}
          Scale.Y {type=MovieSceneScriptingFloatChannel}
          Shear.X {type=MovieSceneScriptingFloatChannel}
          Shear.Y {type=MovieSceneScriptingFloatChannel}
         ]
        }]
      }]
    }
    CategoryEntry_6 {
     id=#16
     possessed_class=SpatialCustomizationCategoryEntry_C
     tracks [{
       type=MovieScene2DTransformTrack
       sections [{
         type=MovieScene2DTransformSection
         channels [
          Translation.X {type=MovieSceneScriptingFloatChannel, keys [[4,1024],[19,0]]}
          Translation.Y {type=MovieSceneScriptingFloatChannel}
          Angle {type=MovieSceneScriptingFloatChannel}
          Scale.X {type=MovieSceneScriptingFloatChannel}
          Scale.Y {type=MovieSceneScriptingFloatChannel}
          Shear.X {type=MovieSceneScriptingFloatChannel}
          Shear.Y {type=MovieSceneScriptingFloatChannel}
         ]
        }]
      }]
    }
   ]
   display_rate {numerator=60}
   tracks [{
     type=MovieSceneEventTrack
     sections [{
       type=MovieSceneEventTriggerSection
       channels [None {type=MovieSceneScriptingEventChannel, const ["<Struct 'MovieSceneEvent' (0x46666713480) {}>", 1]}]
      }]
    }]
  }
 }]
graphs [
 EventGraph {nodes [
   n1 = Animation Playing Event() {is=title K2Node_CustomEvent}
   > n2 = IsAnimationPlayingForward!(
    execute
    self:User Widget Object Reference
    InAnimation:Widget Animation Object Reference=Get IntroCategories() -> (IntroCategories:Widget Animation Object Reference)
   ) -> (then, ReturnValue:bool)
   > n3 = Branch(Condition=@n2)
   > PlaySound2D(Sound=/BRCosmetics/Sounds/Other/Fort_UI_Sounds/S13_Frontend/UI_S13_BP_Move_Umbrella_ChoiceIn_Cue)
   PlaySound2D(execute=@n3.else, Sound=/BRCosmetics/Sounds/Other/Fort_UI_Sounds/S13_Frontend/UI_S13_BP_Move_Umbrella_ChoiceOut_Cue)
   Event OnInitialized()
   > InitializeCategories()
   > BindToAnimationStarted!(execute, self:User Widget Object Reference, Animation:Widget Animation Object Reference=Get IntroCategories(), Delegate:Delegate=@n1) -> (then)
   n4 = Event OnSetCaptureMode!() -> (OutputDelegate, then, bCaptureMode:bool) {is=title K2Node_Event}
   n5 = Select(Option 0=HitTestInvisible, Option 1=Hidden, Index=@n4.bCaptureMode)
   > SetVisibility(self=Get Overlay_Title() -> (Overlay_Title:Overlay Object Reference), InVisibility=@n5)
   > SetVisibility(self=Get RightHeader() -> (RightHeader:Horizontal Box Object Reference), InVisibility=@n5)
   > SetVisibility(self=Get LeftHeader() -> (LeftHeader:Horizontal Box Object Reference), InVisibility=@n5)
   n6 = Event BP_OnCategoryGroupProgress!() -> (OutputDelegate, then, CategoryGroup:Name=None, NumCategoriesFinalized:int=0) {is=title K2Node_Event}
   > HandleCategoryGroupProgress!(execute, self, CategoryGroup:Name=@n6.CategoryGroup, NumCategoriesFinalized:int=@n6.NumCategoriesFinalized) -> (then)
   n7 = Event BP_OnCategoriesSetup!() -> (OutputDelegate, then, TotalNumCategoriesFinalized:int=0) {is=title K2Node_Event}
   > UpdateCompletionText!(execute, self, FinalizedCount:int=@n7.TotalNumCategoriesFinalized) -> (then)
   Event On Activated()
   > PlayAnimationForward(InAnimation=Get IntroCategories())
  ]}
 HandleCategoryGroupProgress {nodes [
   HandleCategoryGroupProgress!() -> (then, CategoryGroup:Name, NumCategoriesFinalized:int) {is=title K2Node_FunctionEntry}
   n1 = Find(
    TargetMap:Map of Names to Panel Widget Object References=Get CategoryGroupMap() -> (CategoryGroupMap:Map of Names to Panel Widget Object References)
    Key:"Name (by ref)"=Get CategoryGroup() -> (CategoryGroup:Name=None)
   ) -> (Value:Panel Widget Object Reference)
   > Set HeaderWidget(HeaderWidget:Panel Widget Object Reference=@n1) -> (Output_Get:Panel Widget Object Reference)
   n2 = Find(
    TargetMap:Map of Names to Integers=Get CategoryCompletionMap() -> (CategoryCompletionMap:Map of Names to Integers)
    Key=Get CategoryGroup() -> (CategoryGroup=None)
   ) -> (Value:int)
   > Set CompletionCount(CompletionCount:int=@n2) -> (Output_Get:int=0)
   > Branch(Condition=AND Boolean(A=@n2.ReturnValue, B=@n1.ReturnValue))
   n3 = Get NumCategoriesFinalized() -> (NumCategoriesFinalized:int=0)
   n4 = Get CompletionCount() -> (CompletionCount:int=0)
   > SetVisibility(
    self=Get HeaderWidget() -> (HeaderWidget:Panel Widget Object Reference)
    InVisibility=Select(
     Option 0=Hidden
     Option 1=SelfHitTestInvisible
     Index="integer < integer"!(A:int=@n3, B:int=@n4) -> (ReturnValue:bool) {is=title K2Node_CallFunction}
    )
   )
  ]}
 InitializeCategories {nodes [
   InitializeCategories() {is=title K2Node_FunctionEntry}
   > AddCategoryGroup!(
    execute
    self:Athena Spatial Customization Categories Object Reference
    CategoryGroup:Name=Get BattlePass() -> (BattlePass:Name=None)
    CategoryEntries:Array of Athena Spatial Customization Category Entry Object References=Make Array!(
     "[0]":Athena Spatial Customization Category Entry Object Reference=Get CategoryEntry_1() -> (CategoryEntry_1:Spatial Customization Category Entry Object Reference)
     "[1]":Athena Spatial Customization Category Entry Object Reference=Get CategoryEntry_2() -> (CategoryEntry_2:Spatial Customization Category Entry Object Reference)
     "[2]":Athena Spatial Customization Category Entry Object Reference=Get CategoryEntry_3() -> (CategoryEntry_3:Spatial Customization Category Entry Object Reference)
     "[3]":Athena Spatial Customization Category Entry Object Reference=Get CategoryEntry_4() -> (CategoryEntry_4:Spatial Customization Category Entry Object Reference)
     "[4]":Athena Spatial Customization Category Entry Object Reference=Get CategoryEntry_5() -> (CategoryEntry_5:Spatial Customization Category Entry Object Reference)
    ) -> (Array:Array of Athena Spatial Customization Category Entry Object References)
   ) -> (then)
   > AddCategoryGroup(
    CategoryGroup=Get Quests() -> (Quests:Name=None)
    CategoryEntries=Make Array(
     "[0]"=Get CategoryEntry_6() -> (CategoryEntry_6:Spatial Customization Category Entry Object Reference)
     "[1]"=Get CategoryEntry_7() -> (CategoryEntry_7:Spatial Customization Category Entry Object Reference)
     "[2]"=Get CategoryEntry_8() -> (CategoryEntry_8:Spatial Customization Category Entry Object Reference)
     "[3]"=Get CategoryEntry_9() -> (CategoryEntry_9:Spatial Customization Category Entry Object Reference)
     "[4]"=Get CategoryEntry_10() -> (CategoryEntry_10:Spatial Customization Category Entry Object Reference)
    )
   )
   n1 = Get BattlePass() -> (BattlePass=None)
   > Add!(
    execute
    TargetMap:Map of Names to Panel Widget Object References=Get CategoryGroupMap()
    Key:"Name (by ref)"=@n1
    Value:"Panel Widget Object Reference (by ref)"=Get LeftHeader()
   ) -> (then) {is=title K2Node_CallFunction}
   n2 = Get Vbox_LeftCategories() -> (Vbox_LeftCategories:Vertical Box Object Reference)
   > Add(
    TargetMap:Map of Names to Integers=Get CategoryCompletionMap()
    Key=@n1
    Value:"Integer (by ref)"=int - int(A=GetChildrenCount!(self:Panel Widget Object Reference=@n2) -> (ReturnValue:int=0), B=1) {is=title K2Node_CallFunction}
   )
   n3 = Get Quests() -> (Quests=None)
   > Add(
    TargetMap:Map of Names to Panel Widget Object References=Get CategoryGroupMap()
    Key=@n3
    Value:"Panel Widget Object Reference (by ref)"=Get RightHeader()
   )
   n4 = Get VBox_RightCategories() -> (VBox_RightCategories:Vertical Box Object Reference)
   > Add(
    TargetMap:Map of Names to Integers=Get CategoryCompletionMap()
    Key=@n3
    Value:"Integer (by ref)"=int - int(A=GetChildrenCount(self=@n4) -> (ReturnValue=0), B=1)
   )
  ]}
 UpdateCompletionText {nodes [
   UpdateCompletionText!() -> (then, FinalizedCount:int) {is=title K2Node_FunctionEntry}
   n1 = int * int(A:int=Get FinalizedCount() -> (FinalizedCount:int=0), B:int=10) -> (ReturnValue:int=0) {is=title K2Node_CommutativeAssociativeBinaryOperator}
   > SetText(
    self=Get Text_PercentComplete() -> (Text_PercentComplete:Common Rich Text Block Object Reference)
    InText=Format Text!(
     Format="LOCTEXT(\"{finalizedPercent}<text fontface=\\\"black\\\" size=\\\"18\\\">%</> <text color=\\\"BFFFFF\\\" case=\\\"upper\\\" fontface=\\\"black\\\" size=\\\"22\\\">Complete</>\")"
     finalizedPercent:Text="To Text (Integer)"(Value=@n1, MaximumIntegralDigits=3)
    ) -> (Result)
   )
   n2 = Get FinalizedCount() -> (FinalizedCount=0)
   > SetVisibility(
    self=Get ProgressBar() -> (ProgressBar:Image Object Reference)
    InVisibility=Select(Option 0=SelfHitTestInvisible, Option 1=Hidden, Index="Equal (Integer)"(A=@n2, B=0) {is=title K2Node_CallFunction})
   )
   > n3 = GetDynamicMaterial(self=Get ProgressBar())
   n4 = Get FinalizedCount() -> (FinalizedCount=0)
   > SetScalarParameterValue(self=@n3, ParameterName=LevelProgress, Value=float * float(A="To Float (Integer)"(InInt=@n4), B=0.1))
  ]}
]
variables [
 BattlePass {type=string, default=Battlepass}
 CategoryCompletionMap {type="map<integer>", default {}}
 CategoryGroupMap {type="map<PanelWidget>", default {}}
 Quests {type=string, default=Quests}
]
widgets {
 info {
  inheritedWidgetCount=0
  parentClass=/Script/FortniteUI.AthenaSpatialCustomizationCategories
  rootWidgetClass=&Overlay
  widgetCount=42
 }
 widgets [
  Overlay_Categories {widgetClassPath=&Overlay, widget=Overlay_Categories}
  SafeZone_1 {
   widgetClassPath=&SafeZone
   bIsVariable=true
   parent=Overlay_Categories
   widget=SafeZone_1
   properties {padLeft=false, padRight=false}
   slot {class=OverlaySlot, properties {horizontalAlignment=HAlign_Center, verticalAlignment=VAlign_Fill, padding=<ltrb 0 32 0 0>}}
  }
  HBox_Categories {
   widgetClassPath=&HorizontalBox
   bIsVariable=true
   parent=SafeZone_1
   widget=HBox_Categories
   properties {navigation=&self:WidgetTree.HBox_Categories.WidgetNavigation_0}
   slot {class=SafeZoneSlot}
  }
  Overlay_0 {
   widgetClassPath=&Overlay
   parent=HBox_Categories
   widget=Overlay_0
   properties {renderTransform {angle=0, scale=<1 1>, shear=<4 0>, translation=<0 0>}}
   slot {class=HorizontalBoxSlot}
  }
  Image_149 {
   widgetClassPath=&Image
   bIsVariable=true
   parent=Overlay_0
   widget=Image_149
   properties {brush {
     bIsDynamicallyLoaded=false
     drawAs=Box
     imageType=NoImage
     mirroring=NoMirror
     resourceName=None
     resourceObject=/BRCosmetics/Athena/UI/Frontend/Spatial/Art/M_Ui_UmbrellaVignette_S13
     tiling=NoTile
     imageSize=<512 512>
     margin=<ltrb 0 0 0.25 0>
     outlineSettings {
      bUseBrushTransparency=false
      roundingType=HalfHeightRadius
      width=0
      color {colorUseRule=UseColor_Specified, specifiedColor=<rgba 0 0 0 0>}
      cornerRadii=<0 0 0 0>
     }
     tintColor {colorUseRule=UseColor_Specified, specifiedColor=<rgba 1 1 1 1>}
     uVRegion {bIsValid=false, max=<0 0>, min=<0 0>}
    }}
   slot {class=OverlaySlot, properties {horizontalAlignment=HAlign_Fill, verticalAlignment=VAlign_Fill, padding=<ltrb -1024 -512 -128 -512>}}
  }
  Vbox_LeftCategories {
   widgetClassPath=&VerticalBox
   bIsVariable=true
   parent=Overlay_0
   widget=Vbox_LeftCategories
   slot {class=OverlaySlot, properties {verticalAlignment=VAlign_Center}}
  }
  LeftHeader {
   widgetClassPath=&HorizontalBox
   bIsVariable=true
   parent=Vbox_LeftCategories
   widget=LeftHeader
   slot {class=VerticalBoxSlot, properties {horizontalAlignment=HAlign_Right, padding=<ltrb 0 0 0 12>}}
  }
  SizeBox_1 {
   widgetClassPath=&SizeBox
   parent=LeftHeader
   widget=SizeBox_1
   properties {bOverride_MaxDesiredWidth=true, maxDesiredWidth=512}
   slot {class=HorizontalBoxSlot, properties {verticalAlignment=VAlign_Center, padding=<ltrb 0 0 4 0>}}
  }
  ScaleBox_0 {
   widgetClassPath=&ScaleBox
   parent=SizeBox_1
   widget=ScaleBox_0
   properties {stretch=ScaleToFitX, stretchDirection=DownOnly}
   slot {class=SizeBoxSlot}
  }
  CommonTextBlock_76 {
   widgetClassPath=&CommonTextBlock
   parent=ScaleBox_0
   widget=CommonTextBlock_76
   properties {
    justification=Right
    renderOpacity=0.75
    text=Level Up Battle Pass to Unlock
    textTransformPolicy=ToUpper
    colorAndOpacity {colorUseRule=UseColor_Specified, specifiedColor=<rgba 0.5 1 1 1>}
    font {
     bForceMonospaced=false
     bMaterialIsStencil=false
     fontMaterial=null
     fontObject=&BurbankSmall
     letterSpacing=0
     monospacedWidth=1
     size=15
     skewAmount=0
     typefaceFontName=Black
     outlineSettings {
      bApplyOutlineToDropShadows=false
      bMiteredCorners=false
      bSeparateFillAlpha=false
      outlineMaterial=null
      outlineSize=3
      outlineColor=<rgba 0 0 0 0.05>
     }
    }
    margin=<ltrb 4 0 4 0>
    renderTransform {angle=0, scale=<1 1>, shear=<-4 0>, translation=<0 0>}
    shadowOffset=<0 0>
   }
   slot {class=ScaleBoxSlot}
  }
  Overlay_111 {
   widgetClassPath=&Overlay
   parent=LeftHeader
   widget=Overlay_111
   properties {visibility=Collapsed, renderTransform {angle=0, scale=<1 1>, shear=<-4 0>, translation=<0 0>}}
   slot {class=HorizontalBoxSlot}
  }
  Image_72 {
   widgetClassPath=&Image
   bIsVariable=true
   parent=Overlay_111
   widget=Image_72
   properties {brush {
     bIsDynamicallyLoaded=false
     drawAs=Image
     imageType=NoImage
     mirroring=NoMirror
     resourceName=None
     resourceObject=/BRCosmetics/UI/Art_ToolKit/Tools/M_UI_Circle_VariantCyle_Backing
     tiling=NoTile
     imageSize=<36 36>
     margin=<ltrb 0 0 0 0>
     outlineSettings {
      bUseBrushTransparency=false
      roundingType=HalfHeightRadius
      width=0
      color {colorUseRule=UseColor_Specified, specifiedColor=<rgba 0 0 0 0>}
      cornerRadii=<0 0 0 0>
     }
     tintColor {colorUseRule=UseColor_Specified, specifiedColor=<rgba 1 1 1 1>}
     uVRegion {bIsValid=false, max=<0 0>, min=<0 0>}
    }}
   slot {class=OverlaySlot}
  }
  Image_126 {
   widgetClassPath=&Image
   bIsVariable=true
   parent=Overlay_111
   widget=Image_126
   properties {
    brush {
     bIsDynamicallyLoaded=false
     drawAs=Image
     imageType=NoImage
     mirroring=NoMirror
     resourceName=None
     resourceObject=/Game/UI/Foundation/Textures/Icons/Locks/T-Icon-Lock-32
     tiling=NoTile
     imageSize=<24 24>
     margin=<ltrb 0 0 0 0>
     outlineSettings {
      bUseBrushTransparency=false
      roundingType=HalfHeightRadius
      width=0
      color {colorUseRule=UseColor_Specified, specifiedColor=<rgba 0 0 0 0>}
      cornerRadii=<0 0 0 0>
     }
     tintColor {colorUseRule=UseColor_Specified, specifiedColor=<rgba 1 1 1 1>}
     uVRegion {bIsValid=false, max=<0 0>, min=<0 0>}
    }
    colorAndOpacity=<rgba 0.25 0.75 1 1>
    renderTransform {angle=0, scale=<1 1>, shear=<0 0>, translation=<-0.5 -2>}
   }
   slot {class=OverlaySlot, properties {horizontalAlignment=HAlign_Center, verticalAlignment=VAlign_Center}}
  }
  CategoryEntry_1 {
   widgetClassPath=&/SpatialCustomizationCategoryEntry
   bIsVariable=true
   parent=Vbox_LeftCategories
   widget=CategoryEntry_1
   properties {
    bLeftText=true
    categoryKey=/BRCosmetics/Items/Tokens/AthenaToken_BRS13_UmbrellaSelection_01_Token
    onButtonBaseClicked=(&self:WidgetTree.CategoryEntry_1.HandleOnButtonClicked)
    onButtonBaseDoubleClicked=(&self:WidgetTree.CategoryEntry_1.HandleOnButtonDoubleClicked)
    onButtonBaseHovered=(&self:WidgetTree.CategoryEntry_1.HandleOnButtonHovered)
    onButtonBaseUnhovered=(&self:WidgetTree.CategoryEntry_1.HandleOnButtonUnhovered)
    onSelectedChangedBase=(&self:WidgetTree.CategoryEntry_1.HandleOnSelectedChanged)
    style=/FortUILibrary/Foundation/Buttons/ButtonStyles/ButtonStyle-MediumTransparentNoCues.ButtonStyle-MediumTransparentNoCues_C
    hoveredSlateSoundOverride {resourceObject=/BRCosmetics/Sounds/Other/Fort_UI_Sounds/S13_Frontend/UI_CustomizeSelect_Hover_Cue}
    pressedSlateSoundOverride {resourceObject=/BRCosmetics/Sounds/Other/Fort_UI_Sounds/S13_Frontend/UI_CustomizeSelect_Click_Cue}
   }
   slot {class=VerticalBoxSlot, properties {horizontalAlignment=HAlign_Right}}
  }
  CategoryEntry_2 {
   widgetClassPath=&SpatialCustomizationCategoryEntry
   bIsVariable=true
   parent=Vbox_LeftCategories
   widget=CategoryEntry_2
   properties {
    bLeftText=true
    categoryKey=/BRCosmetics/Items/Tokens/AthenaToken_BRS13_UmbrellaSelection_02_Token
    onButtonBaseClicked=(&self:WidgetTree.CategoryEntry_2.HandleOnButtonClicked)
    onButtonBaseDoubleClicked=(&self:WidgetTree.CategoryEntry_2.HandleOnButtonDoubleClicked)
    onButtonBaseHovered=(&self:WidgetTree.CategoryEntry_2.HandleOnButtonHovered)
    onButtonBaseUnhovered=(&self:WidgetTree.CategoryEntry_2.HandleOnButtonUnhovered)
    onSelectedChangedBase=(&self:WidgetTree.CategoryEntry_2.HandleOnSelectedChanged)
    style=&ButtonStyle-MediumTransparentNoCues_C
    cameraTagToNavigateTo {tagName=ERG.Camera.ChangingRoom.ChoiceLow}
    hoveredSlateSoundOverride {resourceObject=&UI_CustomizeSelect_Hover_Cue}
    pressedSlateSoundOverride {resourceObject=&UI_CustomizeSelect_Click_Cue}
   }
   slot {class=VerticalBoxSlot, properties {horizontalAlignment=HAlign_Right}}
  }
  CategoryEntry_3 {
   widgetClassPath=&SpatialCustomizationCategoryEntry
   bIsVariable=true
   parent=Vbox_LeftCategories
   widget=CategoryEntry_3
   properties {
    bLeftText=true
    categoryKey=/BRCosmetics/Items/Tokens/AthenaToken_BRS13_UmbrellaSelection_03_Token
    onButtonBaseClicked=(&self:WidgetTree.CategoryEntry_3.HandleOnButtonClicked)
    onButtonBaseDoubleClicked=(&self:WidgetTree.CategoryEntry_3.HandleOnButtonDoubleClicked)
    onButtonBaseHovered=(&self:WidgetTree.CategoryEntry_3.HandleOnButtonHovered)
    onButtonBaseUnhovered=(&self:WidgetTree.CategoryEntry_3.HandleOnButtonUnhovered)
    onSelectedChangedBase=(&self:WidgetTree.CategoryEntry_3.HandleOnSelectedChanged)
    style=&ButtonStyle-MediumTransparentNoCues_C
    hoveredSlateSoundOverride {resourceObject=&UI_CustomizeSelect_Hover_Cue}
    pressedSlateSoundOverride {resourceObject=&UI_CustomizeSelect_Click_Cue}
   }
   slot {class=VerticalBoxSlot, properties {horizontalAlignment=HAlign_Right}}
  }
  CategoryEntry_4 {
   widgetClassPath=&SpatialCustomizationCategoryEntry
   bIsVariable=true
   parent=Vbox_LeftCategories
   widget=CategoryEntry_4
   properties {
    bLeftText=true
    categoryKey=/BRCosmetics/Items/Tokens/AthenaToken_BRS13_UmbrellaSelection_04_Token
    onButtonBaseClicked=(&self:WidgetTree.CategoryEntry_4.HandleOnButtonClicked)
    onButtonBaseDoubleClicked=(&self:WidgetTree.CategoryEntry_4.HandleOnButtonDoubleClicked)
    onButtonBaseHovered=(&self:WidgetTree.CategoryEntry_4.HandleOnButtonHovered)
    onButtonBaseUnhovered=(&self:WidgetTree.CategoryEntry_4.HandleOnButtonUnhovered)
    onSelectedChangedBase=(&self:WidgetTree.CategoryEntry_4.HandleOnSelectedChanged)
    style=&ButtonStyle-MediumTransparentNoCues_C
    cameraTagToNavigateTo {tagName=ERG.Camera.ChangingRoom.ChoiceLow}
    hoveredSlateSoundOverride {resourceObject=&UI_CustomizeSelect_Hover_Cue}
    pressedSlateSoundOverride {resourceObject=&UI_CustomizeSelect_Click_Cue}
   }
   slot {class=VerticalBoxSlot, properties {horizontalAlignment=HAlign_Right}}
  }
  CategoryEntry_5 {
   widgetClassPath=&SpatialCustomizationCategoryEntry
   bIsVariable=true
   parent=Vbox_LeftCategories
   widget=CategoryEntry_5
   properties {
    bCategoryDefaultColorIsNull=true
    bLeftText=true
    bUseColorPicker=true
    bUseSmallerIcon=true
    categoryKey=/BRCosmetics/Items/Tokens/AthenaToken_BRS13_UmbrellaSelection_05_Token
    onButtonBaseClicked=(&self:WidgetTree.CategoryEntry_5.HandleOnButtonClicked)
    onButtonBaseDoubleClicked=(&self:WidgetTree.CategoryEntry_5.HandleOnButtonDoubleClicked)
    onButtonBaseHovered=(&self:WidgetTree.CategoryEntry_5.HandleOnButtonHovered)
    onButtonBaseUnhovered=(&self:WidgetTree.CategoryEntry_5.HandleOnButtonUnhovered)
    onSelectedChangedBase=(&self:WidgetTree.CategoryEntry_5.HandleOnSelectedChanged)
    style=&ButtonStyle-MediumTransparentNoCues_C
    hoveredSlateSoundOverride {resourceObject=&UI_CustomizeSelect_Hover_Cue}
    pressedSlateSoundOverride {resourceObject=&UI_CustomizeSelect_Click_Cue}
   }
   slot {class=VerticalBoxSlot, properties {horizontalAlignment=HAlign_Right}}
  }
  Spacer_149 {widgetClassPath=&Spacer, parent=HBox_Categories, widget=Spacer_149, properties {size=<575 0>}, slot {class=HorizontalBoxSlot}}
  Overlay_2 {
   widgetClassPath=&Overlay
   parent=HBox_Categories
   widget=Overlay_2
   properties {renderTransform {angle=0, scale=<1 1>, shear=<-4 0>, translation=<0 0>}}
   slot {class=HorizontalBoxSlot}
  }
  Image {
   widgetClassPath=&Image
   bIsVariable=true
   parent=Overlay_2
   widget=Image
   properties {
    brush {
     bIsDynamicallyLoaded=false
     drawAs=Box
     imageType=NoImage
     mirroring=NoMirror
     resourceName=None
     resourceObject=&M_Ui_UmbrellaVignette_S13
     tiling=NoTile
     imageSize=<512 512>
     margin=<ltrb 0 0 0.25 0>
     outlineSettings {
      bUseBrushTransparency=false
      roundingType=HalfHeightRadius
      width=0
      color {colorUseRule=UseColor_Specified, specifiedColor=<rgba 0 0 0 0>}
      cornerRadii=<0 0 0 0>
     }
     tintColor {colorUseRule=UseColor_Specified, specifiedColor=<rgba 1 1 1 1>}
     uVRegion {bIsValid=false, max=<0 0>, min=<0 0>}
    }
    renderTransform {angle=0, scale=<-1 1>, shear=<0 0>, translation=<0 0>}
   }
   slot {class=OverlaySlot, properties {horizontalAlignment=HAlign_Fill, verticalAlignment=VAlign_Fill, padding=<ltrb -128 -512 -1024 -512>}}
  }
  VBox_RightCategories {
   widgetClassPath=&VerticalBox
   bIsVariable=true
   parent=Overlay_2
   widget=VBox_RightCategories
   slot {class=OverlaySlot, properties {verticalAlignment=VAlign_Center}}
  }
  RightHeader {
   widgetClassPath=&HorizontalBox
   bIsVariable=true
   parent=VBox_RightCategories
   widget=RightHeader
   slot {class=VerticalBoxSlot, properties {horizontalAlignment=HAlign_Left, padding=<ltrb 0 0 0 12>}}
  }
  Overlay {
   widgetClassPath=&Overlay
   parent=RightHeader
   widget=Overlay
   properties {visibility=Collapsed, renderTransform {angle=0, scale=<1 1>, shear=<4 0>, translation=<0 0>}}
   slot {class=HorizontalBoxSlot}
  }
  Image_1 {
   widgetClassPath=&Image
   bIsVariable=true
   parent=Overlay
   widget=Image_1
   properties {brush {
     bIsDynamicallyLoaded=false
     drawAs=Image
     imageType=NoImage
     mirroring=NoMirror
     resourceName=None
     resourceObject=&M_UI_Circle_VariantCyle_Backing
     tiling=NoTile
     imageSize=<36 36>
     margin=<ltrb 0 0 0 0>
     outlineSettings {
      bUseBrushTransparency=false
      roundingType=HalfHeightRadius
      width=0
      color {colorUseRule=UseColor_Specified, specifiedColor=<rgba 0 0 0 0>}
      cornerRadii=<0 0 0 0>
     }
     tintColor {colorUseRule=UseColor_Specified, specifiedColor=<rgba 1 1 1 1>}
     uVRegion {bIsValid=false, max=<0 0>, min=<0 0>}
    }}
   slot {class=OverlaySlot}
  }
  Image_2 {
   widgetClassPath=&Image
   bIsVariable=true
   parent=Overlay
   widget=Image_2
   properties {
    brush {
     bIsDynamicallyLoaded=false
     drawAs=Image
     imageType=NoImage
     mirroring=NoMirror
     resourceName=None
     resourceObject=&T-Icon-Lock-32
     tiling=NoTile
     imageSize=<24 24>
     margin=<ltrb 0 0 0 0>
     outlineSettings {
      bUseBrushTransparency=false
      roundingType=HalfHeightRadius
      width=0
      color {colorUseRule=UseColor_Specified, specifiedColor=<rgba 0 0 0 0>}
      cornerRadii=<0 0 0 0>
     }
     tintColor {colorUseRule=UseColor_Specified, specifiedColor=<rgba 1 1 1 1>}
     uVRegion {bIsValid=false, max=<0 0>, min=<0 0>}
    }
    colorAndOpacity=<rgba 0.25 0.75 1 1>
    renderTransform {angle=0, scale=<1 1>, shear=<0 0>, translation=<-0.5 -2>}
   }
   slot {class=OverlaySlot, properties {horizontalAlignment=HAlign_Center, verticalAlignment=VAlign_Center}}
  }
  SizeBox_0 {
   widgetClassPath=&SizeBox
   parent=RightHeader
   widget=SizeBox_0
   properties {bOverride_MaxDesiredWidth=true, maxDesiredWidth=512}
   slot {class=HorizontalBoxSlot, properties {verticalAlignment=VAlign_Center, padding=<ltrb 4 0 0 0>}}
  }
  ScaleBox_1 {
   widgetClassPath=&ScaleBox
   parent=SizeBox_0
   widget=ScaleBox_1
   properties {stretch=ScaleToFitX, stretchDirection=DownOnly}
   slot {class=SizeBoxSlot}
  }
  CommonTextBlock {
   widgetClassPath=&CommonTextBlock
   parent=ScaleBox_1
   widget=CommonTextBlock
   properties {
    renderOpacity=0.75
    text=Complete Challenges to Unlock
    textTransformPolicy=ToUpper
    colorAndOpacity {colorUseRule=UseColor_Specified, specifiedColor=<rgba 0.5 1 1 1>}
    font {
     bForceMonospaced=false
     bMaterialIsStencil=false
     fontMaterial=null
     fontObject=&BurbankSmall
     letterSpacing=0
     monospacedWidth=1
     size=15
     skewAmount=0
     typefaceFontName=Black
     outlineSettings {
      bApplyOutlineToDropShadows=false
      bMiteredCorners=false
      bSeparateFillAlpha=false
      outlineMaterial=null
      outlineSize=3
      outlineColor=<rgba 0 0 0 0.05>
     }
    }
    margin=<ltrb 4 0 4 0>
    renderTransform {angle=0, scale=<1 1>, shear=<4 0>, translation=<0 0>}
    shadowOffset=<0 0>
   }
   slot {class=ScaleBoxSlot}
  }
  CategoryEntry_6 {
   widgetClassPath=&SpatialCustomizationCategoryEntry
   bIsVariable=true
   parent=VBox_RightCategories
   widget=CategoryEntry_6
   properties {
    bUseColorPicker=true
    bUseSmallerIcon=true
    categoryKey=/Game/Items/Tokens/AthenaToken_BRS13_UmbrellaSelection_06
    onButtonBaseClicked=(&self:WidgetTree.CategoryEntry_6.HandleOnButtonClicked)
    onButtonBaseDoubleClicked=(&self:WidgetTree.CategoryEntry_6.HandleOnButtonDoubleClicked)
    onButtonBaseHovered=(&self:WidgetTree.CategoryEntry_6.HandleOnButtonHovered)
    onButtonBaseUnhovered=(&self:WidgetTree.CategoryEntry_6.HandleOnButtonUnhovered)
    onSelectedChangedBase=(&self:WidgetTree.CategoryEntry_6.HandleOnSelectedChanged)
    style=&ButtonStyle-MediumTransparentNoCues_C
    hoveredSlateSoundOverride {resourceObject=&UI_CustomizeSelect_Hover_Cue}
    pressedSlateSoundOverride {resourceObject=&UI_CustomizeSelect_Click_Cue}
   }
   slot {class=VerticalBoxSlot}
  }
  CategoryEntry_7 {
   widgetClassPath=&SpatialCustomizationCategoryEntry
   bIsVariable=true
   parent=VBox_RightCategories
   widget=CategoryEntry_7
   properties {
    bUseColorPicker=true
    bUseSmallerIcon=true
    categoryKey=/Game/Items/Tokens/AthenaToken_BRS13_UmbrellaSelection_07
    onButtonBaseClicked=(&self:WidgetTree.CategoryEntry_7.HandleOnButtonClicked)
    onButtonBaseDoubleClicked=(&self:WidgetTree.CategoryEntry_7.HandleOnButtonDoubleClicked)
    onButtonBaseHovered=(&self:WidgetTree.CategoryEntry_7.HandleOnButtonHovered)
    onButtonBaseUnhovered=(&self:WidgetTree.CategoryEntry_7.HandleOnButtonUnhovered)
    onSelectedChangedBase=(&self:WidgetTree.CategoryEntry_7.HandleOnSelectedChanged)
    style=&ButtonStyle-MediumTransparentNoCues_C
    cameraTagToNavigateTo {tagName=ERG.Camera.ChangingRoom.ChoiceLow}
    hoveredSlateSoundOverride {resourceObject=&UI_CustomizeSelect_Hover_Cue}
    pressedSlateSoundOverride {resourceObject=&UI_CustomizeSelect_Click_Cue}
   }
   slot {class=VerticalBoxSlot}
  }
  CategoryEntry_8 {
   widgetClassPath=&SpatialCustomizationCategoryEntry
   bIsVariable=true
   parent=VBox_RightCategories
   widget=CategoryEntry_8
   properties {
    bUseColorPicker=true
    bUseSmallerIcon=true
    categoryKey=/Game/Items/Tokens/AthenaToken_BRS13_UmbrellaSelection_08
    onButtonBaseClicked=(&self:WidgetTree.CategoryEntry_8.HandleOnButtonClicked)
    onButtonBaseDoubleClicked=(&self:WidgetTree.CategoryEntry_8.HandleOnButtonDoubleClicked)
    onButtonBaseHovered=(&self:WidgetTree.CategoryEntry_8.HandleOnButtonHovered)
    onButtonBaseUnhovered=(&self:WidgetTree.CategoryEntry_8.HandleOnButtonUnhovered)
    onSelectedChangedBase=(&self:WidgetTree.CategoryEntry_8.HandleOnSelectedChanged)
    style=&ButtonStyle-MediumTransparentNoCues_C
    hoveredSlateSoundOverride {resourceObject=&UI_CustomizeSelect_Hover_Cue}
    pressedSlateSoundOverride {resourceObject=&UI_CustomizeSelect_Click_Cue}
   }
   slot {class=VerticalBoxSlot}
  }
  CategoryEntry_9 {
   widgetClassPath=&SpatialCustomizationCategoryEntry
   bIsVariable=true
   parent=VBox_RightCategories
   widget=CategoryEntry_9
   properties {
    bUseColorPicker=true
    bUseSmallerIcon=true
    categoryKey=/Game/Items/Tokens/AthenaToken_BRS13_UmbrellaSelection_09
    onButtonBaseClicked=(&self:WidgetTree.CategoryEntry_9.HandleOnButtonClicked)
    onButtonBaseDoubleClicked=(&self:WidgetTree.CategoryEntry_9.HandleOnButtonDoubleClicked)
    onButtonBaseHovered=(&self:WidgetTree.CategoryEntry_9.HandleOnButtonHovered)
    onButtonBaseUnhovered=(&self:WidgetTree.CategoryEntry_9.HandleOnButtonUnhovered)
    onSelectedChangedBase=(&self:WidgetTree.CategoryEntry_9.HandleOnSelectedChanged)
    style=&ButtonStyle-MediumTransparentNoCues_C
    cameraTagToNavigateTo {tagName=ERG.Camera.ChangingRoom.ChoiceLow}
    hoveredSlateSoundOverride {resourceObject=&UI_CustomizeSelect_Hover_Cue}
    pressedSlateSoundOverride {resourceObject=&UI_CustomizeSelect_Click_Cue}
   }
   slot {class=VerticalBoxSlot}
  }
  CategoryEntry_10 {
   widgetClassPath=&SpatialCustomizationCategoryEntry
   bIsVariable=true
   parent=VBox_RightCategories
   widget=CategoryEntry_10
   properties {
    bUseColorPicker=true
    bUseSmallerIcon=true
    categoryKey=/Game/Items/Tokens/AthenaToken_BRS13_UmbrellaSelection_10
    lowerClampingColorBound=0.03
    onButtonBaseClicked=(&self:WidgetTree.CategoryEntry_10.HandleOnButtonClicked)
    onButtonBaseDoubleClicked=(&self:WidgetTree.CategoryEntry_10.HandleOnButtonDoubleClicked)
    onButtonBaseHovered=(&self:WidgetTree.CategoryEntry_10.HandleOnButtonHovered)
    onButtonBaseUnhovered=(&self:WidgetTree.CategoryEntry_10.HandleOnButtonUnhovered)
    onSelectedChangedBase=(&self:WidgetTree.CategoryEntry_10.HandleOnSelectedChanged)
    style=&ButtonStyle-MediumTransparentNoCues_C
    hoveredSlateSoundOverride {resourceObject=&UI_CustomizeSelect_Hover_Cue}
    pressedSlateSoundOverride {resourceObject=&UI_CustomizeSelect_Click_Cue}
   }
   slot {class=VerticalBoxSlot}
  }
  SafeZone_0 {
   widgetClassPath=&SafeZone
   bIsVariable=true
   parent=Overlay_Categories
   widget=SafeZone_0
   properties {padBottom=false, padRight=false}
   slot {class=OverlaySlot, properties {padding=<ltrb 32 32 32 32>}}
  }
  Overlay_Title {
   widgetClassPath=&Overlay
   bIsVariable=true
   parent=SafeZone_0
   widget=Overlay_Title
   properties {renderTransformPivot=<0 0.5>}
   slot {class=SafeZoneSlot, properties {hAlign=HAlign_Left}}
  }
  HorizontalBox_1 {widgetClassPath=&HorizontalBox, parent=Overlay_Title, widget=HorizontalBox_1, slot {class=OverlaySlot}}
  Image_64 {
   widgetClassPath=&Image
   bIsVariable=true
   parent=HorizontalBox_1
   widget=Image_64
   properties {brush {
     bIsDynamicallyLoaded=false
     drawAs=Image
     imageType=NoImage
     mirroring=NoMirror
     resourceName=None
     resourceObject=/BRCosmetics/Athena/UI/Frontend/Spatial/Art/M_UI_UmbrellaIcon_Base
     tiling=NoTile
     imageSize=<88 88>
     margin=<ltrb 0 0 0 0>
     outlineSettings {
      bUseBrushTransparency=false
      roundingType=HalfHeightRadius
      width=0
      color {colorUseRule=UseColor_Specified, specifiedColor=<rgba 0 0 0 0>}
      cornerRadii=<0 0 0 0>
     }
     tintColor {colorUseRule=UseColor_Specified, specifiedColor=<rgba 1 1 1 1>}
     uVRegion {bIsValid=false, max=<0 0>, min=<0 0>}
    }}
   slot {class=HorizontalBoxSlot, properties {verticalAlignment=VAlign_Top}}
  }
  VerticalBox_0 {
   widgetClassPath=&VerticalBox
   parent=HorizontalBox_1
   widget=VerticalBox_0
   slot {class=HorizontalBoxSlot, properties {padding=<ltrb 0 8 0 0>}}
  }
  Text_title {
   widgetClassPath=&CommonTextBlock
   parent=VerticalBox_0
   widget=Text_title
   properties {
    lineHeightPercentage=0.9
    text=Build-A-Brella
    textTransformPolicy=ToUpper
    visibility=HitTestInvisible
    font {
     bForceMonospaced=false
     bMaterialIsStencil=false
     fontMaterial=/BRCosmetics/UI/Art_ToolKit/AngledBoxes/M_UI_SimpleFIll_color_Inst
     fontObject=&Burbank
     letterSpacing=0
     monospacedWidth=1
     size=40
     skewAmount=0
     typefaceFontName=BurbankRegular-Black
     outlineSettings {
      bApplyOutlineToDropShadows=false
      bMiteredCorners=false
      bSeparateFillAlpha=true
      outlineMaterial=null
      outlineSize=0
      outlineColor=<rgba 1 1 1 1>
     }
    }
    margin=<ltrb 0 -4 0 0>
    renderTransform {angle=0, scale=<1 1>, shear=<-12 0>, translation=<0 0>}
    renderTransformPivot=<0 0.5>
    shadowOffset=<0 0>
   }
   slot {class=VerticalBoxSlot}
  }
  Text_PercentComplete {
   widgetClassPath=&CommonRichTextBlock
   bIsVariable=true
   parent=VerticalBox_0
   widget=Text_PercentComplete
   properties {
    bOverrideDefaultStyle=true
    text=10
    defaultTextStyleOverride {
     overflowPolicy=Clip
     transformPolicy=None
     colorAndOpacity {colorUseRule=UseColor_Specified, specifiedColor=<rgba 1 1 1 1>}
     font {
      bForceMonospaced=false
      bMaterialIsStencil=false
      fontMaterial=null
      fontObject=&BurbankSmall
      letterSpacing=0
      monospacedWidth=1
      size=28
      skewAmount=0
      typefaceFontName=Black
      outlineSettings {
       bApplyOutlineToDropShadows=false
       bMiteredCorners=false
       bSeparateFillAlpha=false
       outlineMaterial=null
       outlineSize=0
       outlineColor=<rgba 0 0 0 1>
      }
     }
     highlightColor {colorUseRule=UseColor_Specified, specifiedColor=<rgba 0 0 0 1>}
     highlightShape {
      bIsDynamicallyLoaded=false
      drawAs=Image
      imageType=NoImage
      mirroring=NoMirror
      resourceName=None
      resourceObject=null
      tiling=NoTile
      imageSize=<32 32>
      margin=<ltrb 0 0 0 0>
      outlineSettings {
       bUseBrushTransparency=false
       roundingType=HalfHeightRadius
       width=0
       color {colorUseRule=UseColor_Specified, specifiedColor=<rgba 0 0 0 0>}
       cornerRadii=<0 0 0 0>
      }
      tintColor {colorUseRule=UseColor_Specified, specifiedColor=<rgba 1 1 1 1>}
      uVRegion {bIsValid=false, max=<0 0>, min=<0 0>}
     }
     selectedBackgroundColor {colorUseRule=UseColor_Foreground, specifiedColor=<rgba 1 0 1 1>}
     shadowColorAndOpacity=<rgba 0 0 0 1>
     shadowOffset=<0 0>
     strikeBrush {
      bIsDynamicallyLoaded=false
      drawAs=Image
      imageType=NoImage
      mirroring=NoMirror
      resourceName=None
      resourceObject=null
      tiling=NoTile
      imageSize=<32 32>
      margin=<ltrb 0 0 0 0>
      outlineSettings {
       bUseBrushTransparency=false
       roundingType=HalfHeightRadius
       width=0
       color {colorUseRule=UseColor_Specified, specifiedColor=<rgba 0 0 0 0>}
       cornerRadii=<0 0 0 0>
      }
      tintColor {colorUseRule=UseColor_Specified, specifiedColor=<rgba 1 1 1 1>}
      uVRegion {bIsValid=false, max=<0 0>, min=<0 0>}
     }
     underlineBrush {
      bIsDynamicallyLoaded=false
      drawAs=Image
      imageType=NoImage
      mirroring=NoMirror
      resourceName=None
      resourceObject=null
      tiling=NoTile
      imageSize=<32 32>
      margin=<ltrb 0 0 0 0>
      outlineSettings {
       bUseBrushTransparency=false
       roundingType=HalfHeightRadius
       width=0
       color {colorUseRule=UseColor_Specified, specifiedColor=<rgba 0 0 0 0>}
       cornerRadii=<0 0 0 0>
      }
      tintColor {colorUseRule=UseColor_Specified, specifiedColor=<rgba 1 1 1 1>}
      uVRegion {bIsValid=false, max=<0 0>, min=<0 0>}
     }
    }
    renderTransform {angle=0, scale=<1 1>, shear=<-12 0>, translation=<0 0>}
   }
   slot {class=VerticalBoxSlot, properties {horizontalAlignment=HAlign_Left, padding=<ltrb -12 -28 0 0>}}
  }
  ProgressBar {
   widgetClassPath=&Image
   bIsVariable=true
   parent=VerticalBox_0
   widget=ProgressBar
   properties {
    brush {
     bIsDynamicallyLoaded=false
     drawAs=Image
     imageType=NoImage
     mirroring=NoMirror
     resourceName=None
     resourceObject=/BRCosmetics/UI/Art_ToolKit/ProgressBars/M_UI_CustomizationScreen_Progress
     tiling=NoTile
     imageSize=<32 8>
     margin=<ltrb 0.5 0 0.5 0>
     outlineSettings {
      bUseBrushTransparency=false
      roundingType=HalfHeightRadius
      width=0
      color {colorUseRule=UseColor_Specified, specifiedColor=<rgba 0 0 0 0>}
      cornerRadii=<0 0 0 0>
     }
     tintColor {colorUseRule=UseColor_Specified, specifiedColor=<rgba 1 1 1 1>}
     uVRegion {bIsValid=false, max=<0 0>, min=<0 0>}
    }
    renderTransform {angle=0, scale=<1 1>, shear=<-12 0>, translation=<0 0>}
   }
   slot {class=VerticalBoxSlot, properties {padding=<ltrb -8 0 0 0>}}
  }
 ]
}
