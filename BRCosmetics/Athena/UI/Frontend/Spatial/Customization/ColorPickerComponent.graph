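# Widget Blueprint for one color component row in the Athena frontend customization color picker, built on native FortColorPickerComponent.
# A bordered layout shows the inherited name and value texts above a FortColorPickerSlider, flanked by collapsed minus and plus labels.
# Mouse enter and leave play the Selected animation forward and in reverse, fading in the background and brightening both texts.
# Minus_Hovered and Plus_Hovered both target the minus label and are not played by this graph.
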WidgetBlueprint /BRCosmetics/Athena/UI/Frontend/Spatial/Customization/ColorPickerComponent
animations [
 Selected {
  binding_count=3
  length_seconds=0.1
  object=&self:Selected
  sequence {bindings [
    CommonBorder_55 {
     id=#1
     possessed_class=CommonBorder
     tracks [{
       type=MovieSceneColorTrack
       display_name=Brush Color
       sections [{
         type=MovieSceneColorSection
         end_frame=2
         start_frame=0
         channels [
          Color.R {type=MovieSceneScriptingFloatChannel, const [1,1]}
          Color.G {type=MovieSceneScriptingFloatChannel, const [1,1]}
          Color.B {type=MovieSceneScriptingFloatChannel, const [1,1]}
          Color.A {type=MovieSceneScriptingFloatChannel, keys [[0,0],[2,0.5]]}
         ]
        }]
      }]
    }
    Text_ComponentName {
     id=#2
     possessed_class=CommonTextBlock
     tracks [{
       type=MovieSceneFloatTrack
       display_name=Render Opacity
       sections [{type=MovieSceneFloatSection, end_frame=2, start_frame=0, channels [None {type=MovieSceneScriptingFloatChannel, keys [[0,0.75],[2,1]]}]}]
      }]
    }
    Text_CurrentValue {
     id=#3
     possessed_class=CommonTextBlock
     tracks [{
       type=MovieSceneFloatTrack
       display_name=Render Opacity
       sections [{type=MovieSceneFloatSection, end_frame=2, start_frame=0, channels [None {type=MovieSceneScriptingFloatChannel, keys [[0,0.75],[2,1]]}]}]
      }]
    }
   ]}
 }
 Minus_Hovered {
  binding_count=1
  length_seconds=0.1
  object=&self:Minus_Hovered
  sequence {bindings [CommonTextBlock_139 {
     id=#4
     display_name=CommonTextBlock_53
     possessed_class=CommonTextBlock
     tracks [{
       type=MovieSceneColorTrack
       display_name=Color and Opacity
       sections [{
         type=MovieSceneColorSection
         end_frame=2
         start_frame=0
         channels [
          Color.R {type=MovieSceneScriptingFloatChannel, keys [[0,0],[2,1]]}
          Color.G {type=MovieSceneScriptingFloatChannel, keys [[0,0.1441],[2,1]]}
          Color.B {type=MovieSceneScriptingFloatChannel, const [1,2]}
          Color.A {type=MovieSceneScriptingFloatChannel}
         ]
        }]
      }]
    }]}
 }
 Plus_Hovered {
  binding_count=1
  length_seconds=0.1
  object=&self:Plus_Hovered
  sequence {bindings [CommonTextBlock_139 {
     id=#4
     possessed_class=CommonTextBlock
     tracks [{
       type=MovieSceneColorTrack
       display_name=Color and Opacity
       sections [{
         type=MovieSceneColorSection
         end_frame=2
         start_frame=0
         channels [
          Color.R {type=MovieSceneScriptingFloatChannel, keys [[0,0],[2,1]]}
          Color.G {type=MovieSceneScriptingFloatChannel, keys [[0,0.1441],[2,1]]}
          Color.B {type=MovieSceneScriptingFloatChannel, const [1,2]}
          Color.A {type=MovieSceneScriptingFloatChannel}
         ]
        }]
      }]
    }]}
 }
]
graphs [EventGraph {nodes [
   Event OnMouseEnter()
   n1 = Get Selected() -> (Selected:Widget Animation Object Reference)
   > PlayAnimationForward(InAnimation=@n1)
   Event OnMouseLeave!() -> (OutputDelegate, then, MouseEvent:"Pointer Event Structure (by ref)") {is=title K2Node_Event}
   > PlayAnimationReverse(InAnimation=@n1)
  ]}]
variables []
widgets {
 info {inheritedWidgetCount=3, parentClass=/Script/FortniteUI.FortColorPickerComponent, rootWidgetClass=&Overlay, widgetCount=12}
 widgets [
  Overlay_38 {widgetClassPath=&Overlay, widget=Overlay_38, properties {renderTransform {angle=0, scale=<1 1>, shear=<4 0>, translation=<0 0>}}}
  CommonBorder_55 {
   widgetClassPath=&CommonBorder
   parent=Overlay_38
   widget=CommonBorder_55
   properties {
    background {
     bIsDynamicallyLoaded=false
     drawAs=Box
     imageType=NoImage
     mirroring=NoMirror
     resourceName=None
     resourceObject=/FortniteCoreUI/Frontend/Spatial/Art/M_Ui_Spacial_Seaparator_Background_progress
     tiling=NoTile
     imageSize=<8 8>
     margin=<ltrb 0.25 0.25 0.25 0.25>
     outlineSettings {
      bUseBrushTransparency=false
      roundingType=HalfHeightRadius
      width=0
      color {colorUseRule=UseColor_Specified, specifiedColor=<rgba 0 0 0 0>}
      cornerRadii=<0 0 0 0>
     }
     tintColor {colorUseRule=UseColor_Specified, specifiedColor=<rgba 0 0.1667 1 1>}
     uVRegion {bIsValid=false, max=<0 0>, min=<0 0>}
    }
    brushColor=<rgba 1 1 1 0>
    padding=<ltrb 16 8 16 8>
   }
   slot {class=OverlaySlot, properties {horizontalAlignment=HAlign_Fill, verticalAlignment=VAlign_Fill}}
  }
  VerticalBox_75 {
   widgetClassPath=&VerticalBox
   parent=CommonBorder_55
   widget=VerticalBox_75
   slot {class=BorderSlot, properties {padding=<ltrb 16 8 16 8>}}
  }
  HorizontalBox_39 {widgetClassPath=&HorizontalBox, parent=VerticalBox_75, widget=HorizontalBox_39, slot {class=VerticalBoxSlot}}
  Text_ComponentName {
   widgetClassPath=&CommonTextBlock
   bInherited=true
   parent=HorizontalBox_39
   widget=Text_ComponentName
   properties {
    renderOpacity=0.75
    text=Text Block
    textTransformPolicy=ToUpper
    colorAndOpacity {colorUseRule=UseColor_Specified, specifiedColor=<rgba 0 0.3333 1 1>}
    font {
     bForceMonospaced=false
     bMaterialIsStencil=false
     fontMaterial=null
     fontObject=&BurbankSmall
     letterSpacing=0
     monospacedWidth=1
     size=18
     skewAmount=0
     typefaceFontName=Black
     outlineSettings {
      bApplyOutlineToDropShadows=false
      bMiteredCorners=false
      bSeparateFillAlpha=false
      outlineMaterial=null
      outlineSize=0
      outlineColor=<rgba 0 0 0 1>
     }
    }
    margin=<ltrb 4 0 4 0>
    shadowOffset=<0 0>
   }
   slot {class=HorizontalBoxSlot, properties {padding=<ltrb 0 0 8 0>, size {sizeRule=Fill, value=1}}}
  }
  Text_CurrentValue {
   widgetClassPath=&CommonTextBlock
   bInherited=true
   parent=HorizontalBox_39
   widget=Text_CurrentValue
   properties {
    renderOpacity=0.75
    text=10
    colorAndOpacity {colorUseRule=UseColor_Specified, specifiedColor=<rgba 0.25 0.875 1 1>}
    font {
     bForceMonospaced=false
     bMaterialIsStencil=false
     fontMaterial=null
     fontObject=&BurbankSmall
     letterSpacing=0
     monospacedWidth=1
     size=18
     skewAmount=0
     typefaceFontName=Bold
     outlineSettings {
      bApplyOutlineToDropShadows=false
      bMiteredCorners=false
      bSeparateFillAlpha=false
      outlineMaterial=null
      outlineSize=0
      outlineColor=<rgba 0 0 0 1>
     }
    }
    margin=<ltrb 0 0 0 2>
    shadowOffset=<0 0>
   }
   slot {class=HorizontalBoxSlot, properties {horizontalAlignment=HAlign_Right, verticalAlignment=VAlign_Bottom}}
  }
  HorizontalBox_41 {widgetClassPath=&HorizontalBox, parent=VerticalBox_75, widget=HorizontalBox_41, slot {class=VerticalBoxSlot}}
  SizeBox_0 {
   widgetClassPath=&SizeBox
   parent=HorizontalBox_41
   widget=SizeBox_0
   properties {bOverride_WidthOverride=true, visibility=Collapsed, widthOverride=32}
   slot {class=HorizontalBoxSlot, properties {horizontalAlignment=HAlign_Center, verticalAlignment=VAlign_Center}}
  }
  CommonTextBlock_139 {
   widgetClassPath=&CommonTextBlock
   parent=SizeBox_0
   widget=CommonTextBlock_139
   properties {
    text=-
    colorAndOpacity {colorUseRule=UseColor_Specified, specifiedColor=<rgba 0 0.1441 1 1>}
    font {
     bForceMonospaced=false
     bMaterialIsStencil=false
     fontMaterial=null
     fontObject=&BurbankSmall
     letterSpacing=0
     monospacedWidth=1
     size=36
     skewAmount=0
     typefaceFontName=Bold
     outlineSettings {
      bApplyOutlineToDropShadows=false
      bMiteredCorners=false
      bSeparateFillAlpha=false
      outlineMaterial=null
      outlineSize=0
      outlineColor=<rgba 0 0 0 1>
     }
    }
    margin=<ltrb 0 -16 0 -16>
    shadowOffset=<0 0>
   }
   slot {class=SizeBoxSlot, properties {horizontalAlignment=HAlign_Left}}
  }
  ColorSlider_Component {
   widgetClassPath=/Script/FortniteUI.FortColorPickerSlider
   bInherited=true
   bIsVariable=true
   parent=HorizontalBox_41
   widget=ColorSlider_Component
   properties {widgetStyle {
     barThickness=32
     disabledBarImage {
      bIsDynamicallyLoaded=false
      drawAs=Image
      imageType=NoImage
      mirroring=NoMirror
      resourceName=None
      resourceObject=null
      tiling=NoTile
      imageSize=<0 0>
      margin=<ltrb 0 0 0 0>
      outlineSettings {
       bUseBrushTransparency=false
       roundingType=HalfHeightRadius
       width=0
       color {colorUseRule=UseColor_Specified, specifiedColor=<rgba 0 0 0 0>}
       cornerRadii=<0 0 0 0>
      }
      tintColor {colorUseRule=UseColor_Specified, specifiedColor=<rgba 0.5 0.5 0.5 1>}
      uVRegion {bIsValid=false, max=<0 0>, min=<0 0>}
     }
     disabledThumbImage {
      bIsDynamicallyLoaded=false
      drawAs=Image
      imageType=FullColor
      mirroring=NoMirror
      resourceName=../../../Engine/Content/Slate/Common/Button_Disabled.png
      resourceObject=/BRCosmetics/Athena/UI/Frontend/Spatial/Art/M_Ui_Spacial_Seaparator_Background_progress_Inst
      tiling=NoTile
      imageSize=<8 42>
      margin=<ltrb 0 0 0 0>
      outlineSettings {
       bUseBrushTransparency=false
       roundingType=HalfHeightRadius
       width=0
       color {colorUseRule=UseColor_Specified, specifiedColor=<rgba 0 0 0 0>}
       cornerRadii=<0 0 0 0>
      }
      tintColor {colorUseRule=UseColor_Specified, specifiedColor=<rgba 0.5 0.5 0.5 1>}
      uVRegion {bIsValid=false, max=<0 0>, min=<0 0>}
     }
     hoveredBarImage {
      bIsDynamicallyLoaded=false
      drawAs=Image
      imageType=NoImage
      mirroring=NoMirror
      resourceName=None
      resourceObject=null
      tiling=NoTile
      imageSize=<0 0>
      margin=<ltrb 0 0 0 0>
      outlineSettings {
       bUseBrushTransparency=false
       roundingType=HalfHeightRadius
       width=0
       color {colorUseRule=UseColor_Specified, specifiedColor=<rgba 0 0 0 0>}
       cornerRadii=<0 0 0 0>
      }
      tintColor {colorUseRule=UseColor_Specified, specifiedColor=<rgba 1 1 1 1>}
      uVRegion {bIsValid=false, max=<0 0>, min=<0 0>}
     }
     hoveredThumbImage {
      bIsDynamicallyLoaded=false
      drawAs=Image
      imageType=FullColor
      mirroring=NoMirror
      resourceName=../../../Engine/Content/Slate/Common/Button.png
      resourceObject=&M_Ui_Spacial_Seaparator_Background_progress_Inst
      tiling=NoTile
      imageSize=<8 42>
      margin=<ltrb 0 0 0 0>
      outlineSettings {
       bUseBrushTransparency=false
       roundingType=HalfHeightRadius
       width=0
       color {colorUseRule=UseColor_Specified, specifiedColor=<rgba 0 0 0 0>}
       cornerRadii=<0 0 0 0>
      }
      tintColor {colorUseRule=UseColor_Specified, specifiedColor=<rgba 1 1 1 1>}
      uVRegion {bIsValid=false, max=<0 0>, min=<0 0>}
     }
     normalBarImage {
      bIsDynamicallyLoaded=false
      drawAs=Image
      imageType=NoImage
      mirroring=NoMirror
      resourceName=None
      resourceObject=null
      tiling=NoTile
      imageSize=<0 0>
      margin=<ltrb 0 0 0 0>
      outlineSettings {
       bUseBrushTransparency=false
       roundingType=HalfHeightRadius
       width=0
       color {colorUseRule=UseColor_Specified, specifiedColor=<rgba 0 0 0 0>}
       cornerRadii=<0 0 0 0>
      }
      tintColor {colorUseRule=UseColor_Specified, specifiedColor=<rgba 1 1 1 1>}
      uVRegion {bIsValid=false, max=<0 0>, min=<0 0>}
     }
     normalThumbImage {
      bIsDynamicallyLoaded=false
      drawAs=Image
      imageType=FullColor
      mirroring=NoMirror
      resourceName=../../../Engine/Content/Slate/Common/Button.png
      resourceObject=&M_Ui_Spacial_Seaparator_Background_progress_Inst
      tiling=NoTile
      imageSize=<8 42>
      margin=<ltrb 0.25 0.25 0.25 0.25>
      outlineSettings {
       bUseBrushTransparency=false
       roundingType=HalfHeightRadius
       width=0
       color {colorUseRule=UseColor_Specified, specifiedColor=<rgba 0 0 0 0>}
       cornerRadii=<0 0 0 0>
      }
      tintColor {colorUseRule=UseColor_Specified, specifiedColor=<rgba 0 0.1441 1 1>}
      uVRegion {bIsValid=false, max=<0 0>, min=<0 0>}
     }
    }}
   slot {class=HorizontalBoxSlot, properties {size {sizeRule=Fill, value=1}}}
  }
  SizeBox_1 {
   widgetClassPath=&SizeBox
   parent=HorizontalBox_41
   widget=SizeBox_1
   properties {bOverride_WidthOverride=true, visibility=Collapsed, widthOverride=32}
   slot {class=HorizontalBoxSlot}
  }
  CommonTextBlock_53 {
   widgetClassPath=&CommonTextBlock
   parent=SizeBox_1
   widget=CommonTextBlock_53
   properties {
    text=+
    colorAndOpacity {colorUseRule=UseColor_Specified, specifiedColor=<rgba 0 0.1441 1 1>}
    font {
     bForceMonospaced=false
     bMaterialIsStencil=false
     fontMaterial=null
     fontObject=&BurbankSmall
     letterSpacing=0
     monospacedWidth=1
     size=36
     skewAmount=0
     typefaceFontName=Bold
     outlineSettings {
      bApplyOutlineToDropShadows=false
      bMiteredCorners=false
      bSeparateFillAlpha=false
      outlineMaterial=null
      outlineSize=0
      outlineColor=<rgba 0 0 0 1>
     }
    }
    margin=<ltrb 0 -16 0 -16>
    shadowOffset=<0 0>
   }
   slot {class=SizeBoxSlot, properties {horizontalAlignment=HAlign_Right, verticalAlignment=VAlign_Center}}
  }
 ]
}
