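# ColorSwatch is a circular color tile widget for the cosmetics customization color picker, built on FortColorPickerSwatchTile.
# When the swatch color is set, a null color whitens the dynamic material, hides its inner fill and shows a rotated plus glyph; other colors restore the fill and hide the glyph.
# A short Hover animation plays forward on hover and in reverse on unhover, scaling the color and selection images, animating their material inner radius and brightening the glyph.
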WidgetBlueprint /BRCosmetics/Athena/UI/Frontend/Spatial/Customization/ColorSwatch
animations [Hover {
  binding_count=5
  length_seconds=0.15
  object=&self:Hover
  sequence {bindings [
    Image_Selected {
     id=#1
     possessed_class=Image
     tracks [
      {
       type=MovieSceneWidgetMaterialTrack
       sections [{
         type=MovieSceneParameterSection
         end_frame=3
         start_frame=0
         channels [AnimateInnerRadius {type=MovieSceneScriptingFloatChannel, keys [[0,0],[3,1]]}]
        }]
      }
      {
       type=MovieScene2DTransformTrack
       sections [{
         type=MovieScene2DTransformSection
         end_frame=3
         start_frame=0
         channels [
          Translation.X {type=MovieSceneScriptingFloatChannel}
          Translation.Y {type=MovieSceneScriptingFloatChannel}
          Angle {type=MovieSceneScriptingFloatChannel}
          Scale.X {type=MovieSceneScriptingFloatChannel, keys [[0,1],[3,1.1]]}
          Scale.Y {type=MovieSceneScriptingFloatChannel, keys [[0,1],[3,1.1]]}
          Shear.X {type=MovieSceneScriptingFloatChannel}
          Shear.Y {type=MovieSceneScriptingFloatChannel}
         ]
        }]
      }
     ]
    }
    Image_hoverbottom {
     id=#2
     possessed_class=Image
     tracks [{
       type=MovieSceneColorTrack
       display_name=Color and Opacity
       sections [{
         type=MovieSceneColorSection
         end_frame=3
         start_frame=0
         channels [
          Color.R {type=MovieSceneScriptingFloatChannel}
          Color.G {type=MovieSceneScriptingFloatChannel}
          Color.B {type=MovieSceneScriptingFloatChannel}
          Color.A {type=MovieSceneScriptingFloatChannel, keys [[0,0],[3,1]]}
         ]
        }]
      }]
    }
    text_null {
     id=#3
     possessed_class=CommonTextBlock
     tracks [{
       type=MovieSceneColorTrack
       display_name=Color and Opacity
       sections [{
         type=MovieSceneColorSection
         end_frame=3
         start_frame=0
         channels [
          Color.R {type=MovieSceneScriptingFloatChannel}
          Color.G {type=MovieSceneScriptingFloatChannel}
          Color.B {type=MovieSceneScriptingFloatChannel}
          Color.A {type=MovieSceneScriptingFloatChannel, keys [[0,0.5],[3,1]]}
         ]
        }]
      }]
    }
    Image_Color {
     id=#4
     possessed_class=Image
     tracks [
      {
       type=MovieScene2DTransformTrack
       sections [{
         type=MovieScene2DTransformSection
         end_frame=3
         start_frame=0
         channels [
          Angle {type=MovieSceneScriptingFloatChannel}
          Scale.X {type=MovieSceneScriptingFloatChannel, keys [[0,1],[3,1.1]]}
          Scale.Y {type=MovieSceneScriptingFloatChannel, keys [[0,1],[3,1.1]]}
          Shear.X {type=MovieSceneScriptingFloatChannel}
          Shear.Y {type=MovieSceneScriptingFloatChannel}
         ]
        }]
      }
      {
       type=MovieSceneWidgetMaterialTrack
       sections [{
         type=MovieSceneParameterSection
         end_frame=3
         start_frame=0
         channels [AnimateInnerRadius {type=MovieSceneScriptingFloatChannel, keys [[0,0],[3,1]]}]
        }]
      }
     ]
    }
    Image_hovertop {
     id=#5
     possessed_class=Image
     tracks [{
       type=MovieSceneColorTrack
       display_name=Color and Opacity
       sections [{
         type=MovieSceneColorSection
         end_frame=3
         start_frame=0
         channels [
          Color.R {type=MovieSceneScriptingFloatChannel}
          Color.G {type=MovieSceneScriptingFloatChannel}
          Color.B {type=MovieSceneScriptingFloatChannel}
          Color.A {type=MovieSceneScriptingFloatChannel, keys [[0,0],[3,1]]}
         ]
        }]
      }]
    }
   ]}
 }]
graphs [EventGraph {nodes [
   n1 = Event OnSwatchColorSet!() -> (OutputDelegate, then, bInNullColor:bool) {is=title K2Node_Event}
   > n2 = GetDynamicMaterial(self=Get Image_Color() -> (Image_Color:Image Object Reference))
   > n3 = Branch(Condition=@n1.bInNullColor)
   > SetVectorParameterValue(self=@n2, ParameterName=RGB1, Value=(R=1,G=1,B=1,A=1))
   > SetScalarParameterValue(self=@n2, ParameterName=Inner_Opacity, Value=0)
   n4 = Get text_null() -> (text_null:Common Text Object Reference)
   > SetVisibility(self=@n4, InVisibility=SelfHitTestInvisible)
   SetScalarParameterValue(execute=@n3.else, self=@n2, ParameterName=Inner_Opacity, Value=1)
   > SetVisibility(self=@n4, InVisibility=Collapsed)
   Event On Hovered()
   n5 = Get Hover() -> (Hover:Widget Animation Object Reference)
   > PlayAnimationForward(InAnimation=@n5)
   Event On Unhovered()
   > PlayAnimationReverse(InAnimation=@n5)
  ]}]
variables []
widgets {
 info {inheritedWidgetCount=4, parentClass=/Script/FortniteUI.FortColorPickerSwatchTile, rootWidgetClass=&Overlay, widgetCount=9}
 widgets [
  Overlay_0 {widgetClassPath=&Overlay, widget=Overlay_0}
  text_null {
   widgetClassPath=&CommonTextBlock
   bIsVariable=true
   parent=Overlay_0
   widget=text_null
   properties {
    text=+
    visibility=Collapsed
    colorAndOpacity {colorUseRule=UseColor_Specified, specifiedColor=<rgba 0 0.1667 1 0.5>}
    font {
     bForceMonospaced=false
     bMaterialIsStencil=false
     fontMaterial=null
     fontObject=&BurbankSmall
     letterSpacing=0
     monospacedWidth=1
     size=64
     skewAmount=0
     typefaceFontName=Medium
     outlineSettings {
      bApplyOutlineToDropShadows=false
      bMiteredCorners=false
      bSeparateFillAlpha=false
      outlineMaterial=null
      outlineSize=0
      outlineColor=<rgba 0 0 0 1>
     }
    }
    margin=<ltrb 0 -32 0 -32>
    renderTransform {angle=45, scale=<1 1>, shear=<0 0>, translation=<0 0>}
    shadowOffset=<0 0>
   }
   slot {class=OverlaySlot, properties {horizontalAlignment=HAlign_Center, verticalAlignment=VAlign_Center}}
  }
  Image_Color {
   widgetClassPath=&Image
   bInherited=true
   bIsVariable=true
   parent=Overlay_0
   widget=Image_Color
   properties {brush {
     bIsDynamicallyLoaded=false
     drawAs=Image
     imageType=NoImage
     mirroring=NoMirror
     resourceName=None
     resourceObject=/BRCosmetics/Athena/UI/Frontend/Spatial/Art/M_Ui_circular_Swatch_Inst
     tiling=NoTile
     imageSize=<90 90>
     margin=<ltrb 0 0 0 0>
     outlineSettings {
      bUseBrushTransparency=false
      roundingType=HalfHeightRadius
      width=0
      color {colorUseRule=UseColor_Specified, specifiedColor=<rgba 0 0 0 0>}
      cornerRadii=<0 0 0 0>
     }
     tintColor {colorUseRule=UseColor_Specified, specifiedColor=<rgba 1 1 1 1>}
     uVRegion {bIsValid=false, max=<0 0>, min=<0 0>}
    }}
   slot {class=OverlaySlot}
  }
  Image_hovertop {
   widgetClassPath=&Image
   bIsVariable=true
   parent=Overlay_0
   widget=Image_hovertop
   properties {
    visibility=Collapsed
    brush {
     bIsDynamicallyLoaded=false
     drawAs=Box
     imageType=NoImage
     mirroring=NoMirror
     resourceName=None
     resourceObject=/FortniteCoreUI/Frontend/Spatial/Art/M_Ui_Spacial_HoverItemIcon
     tiling=NoTile
     imageSize=<4 4>
     margin=<ltrb 0.25 0.25 0.25 0.25>
     outlineSettings {
      bUseBrushTransparency=false
      roundingType=HalfHeightRadius
      width=0
      color {colorUseRule=UseColor_Specified, specifiedColor=<rgba 0 0 0 0>}
      cornerRadii=<0 0 0 0>
     }
     tintColor {colorUseRule=UseColor_Specified, specifiedColor=<rgba 1 1 1 1>}
     uVRegion {bIsValid=false, max=<0 0>, min=<0 0>}
    }
    colorAndOpacity=<rgba 1 1 1 0>
   }
   slot {class=OverlaySlot, properties {horizontalAlignment=HAlign_Fill, padding=<ltrb 1 0 1 0>}}
  }
  Image_hoverbottom {
   widgetClassPath=&Image
   bIsVariable=true
   parent=Overlay_0
   widget=Image_hoverbottom
   properties {
    visibility=Collapsed
    brush {
     bIsDynamicallyLoaded=false
     drawAs=Box
     imageType=NoImage
     mirroring=NoMirror
     resourceName=None
     resourceObject=&M_Ui_Spacial_HoverItemIcon
     tiling=NoTile
     imageSize=<4 4>
     margin=<ltrb 0.25 0.25 0.25 0.25>
     outlineSettings {
      bUseBrushTransparency=false
      roundingType=HalfHeightRadius
      width=0
      color {colorUseRule=UseColor_Specified, specifiedColor=<rgba 0 0 0 0>}
      cornerRadii=<0 0 0 0>
     }
     tintColor {colorUseRule=UseColor_Specified, specifiedColor=<rgba 1 1 1 1>}
     uVRegion {bIsValid=false, max=<0 0>, min=<0 0>}
    }
    colorAndOpacity=<rgba 1 1 1 0>
   }
   slot {class=OverlaySlot, properties {horizontalAlignment=HAlign_Fill, verticalAlignment=VAlign_Bottom, padding=<ltrb 1 0 1 0>}}
  }
  Image_Selected {
   widgetClassPath=&Image
   bInherited=true
   bIsVariable=true
   parent=Overlay_0
   widget=Image_Selected
   properties {brush {
     bIsDynamicallyLoaded=false
     drawAs=Image
     imageType=NoImage
     mirroring=NoMirror
     resourceName=None
     resourceObject=/BRCosmetics/Athena/UI/Frontend/Spatial/Art/M_Ui_circular_Swatch_Inst_Inst
     tiling=NoTile
     imageSize=<90 90>
     margin=<ltrb 0 0 0 0>
     outlineSettings {
      bUseBrushTransparency=false
      roundingType=HalfHeightRadius
      width=0
      color {colorUseRule=UseColor_Specified, specifiedColor=<rgba 0 0 0 0>}
      cornerRadii=<0 0 0 0>
     }
     tintColor {colorUseRule=UseColor_Specified, specifiedColor=<rgba 0.9301 1 0.0097 1>}
     uVRegion {bIsValid=false, max=<0 0>, min=<0 0>}
    }}
   slot {class=OverlaySlot, properties {horizontalAlignment=HAlign_Fill, verticalAlignment=VAlign_Fill}}
  }
  Overlay_1 {
   widgetClassPath=&Overlay
   parent=Overlay_0
   widget=Overlay_1
   properties {visibility=Collapsed}
   slot {class=OverlaySlot, properties {horizontalAlignment=HAlign_Right, verticalAlignment=VAlign_Bottom}}
  }
  Image_Locked {
   widgetClassPath=&FortLazyImage
   bInherited=true
   bIsVariable=true
   parent=Overlay_1
   widget=Image_Locked
   properties {
    visibility=Collapsed
    brush {
     bIsDynamicallyLoaded=false
     drawAs=Image
     imageType=NoImage
     mirroring=NoMirror
     resourceName=None
     resourceObject=/Game/UI/Foundation/Textures/Icons/Locks/T-Icon-Lock-32
     tiling=NoTile
     imageSize=<32 32>
     margin=<ltrb 0 0 0 0>
     outlineSettings {
      bUseBrushTransparency=false
      roundingType=HalfHeightRadius
      width=0
      color {colorUseRule=UseColor_Specified, specifiedColor=<rgba 0 0 0 0>}
      cornerRadii=<0 0 0 0>
     }
     tintColor {colorUseRule=UseColor_Specified, specifiedColor=<rgba 1 1 1 1>}
     uVRegion {bIsValid=false, max=<0 0>, min=<0 0>}
    }
    colorAndOpacity=<rgba 0.2159 0.3515 1 1>
   }
   slot {class=OverlaySlot}
  }
  InputActionWidget {
   widgetClassPath=&CommonActionWidget
   bInherited=true
   bIsVariable=true
   parent=Overlay_0
   widget=InputActionWidget
   slot {class=OverlaySlot, properties {horizontalAlignment=HAlign_Center, verticalAlignment=VAlign_Bottom}}
  }
 ]
}
