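# Color picker widget for frontend cosmetic customization, built on FortColorPicker, with swatch and custom color pages in a widget switcher.
# The swatch page holds a horizontally scrolling grid and Custom Color button; the custom page has three component controls, a preview and Save Color.
# Setup selects a page by picker type, with Both opening on swatches, and the swap event toggles pages.
# The graph only switches pages; inherited bindings expose key widgets to the parent class.
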WidgetBlueprint /BRCosmetics/Athena/UI/Frontend/Spatial/Customization/ColorPicker
graphs [EventGraph {nodes [
   n1 = Event OnSetupColorPicker!() -> (OutputDelegate, then, InColorPickerType:EColorPickerType Enum=Swatches) {is=title K2Node_Event}
   > SetActiveWidget(
    self:Widget Switcher Object Reference=Get Switcher_PickerType() -> (Switcher_PickerType:Common Widget Switcher Legacy Object Reference)
    Widget=Select!(
     Swatches:Widget Object Reference=Get VBox_Swatches() -> (VBox_Swatches:Vertical Box Object Reference)
     CustomColor:Widget Object Reference=Get VBox_CustomColor() -> (VBox_CustomColor:Vertical Box Object Reference)
     Both:Widget Object Reference=Get VBox_Swatches()
     Index:EColorPickerType Enum=@n1.InColorPickerType
    ) -> (ReturnValue:Widget Object Reference)
   )
   n2 = Event BP_OnSwapColorPickerType!() -> (OutputDelegate, then, bToSwatch:bool) {is=title K2Node_Event}
   > SetActiveWidget(
    self=Get Switcher_PickerType()
    Widget=Select!(Option 0:Widget Object Reference=Get VBox_CustomColor(), Option 1:Widget Object Reference=Get VBox_Swatches(), Index:bool=@n2.bToSwatch) -> (ReturnValue)
   )
  ]}]
variables []
widgets {
 info {inheritedWidgetCount=8, parentClass=/Script/FortniteUI.FortColorPicker, rootWidgetClass=&SizeBox, widgetCount=22}
 widgets [
  SizeBox_103 {
   widgetClassPath=&SizeBox
   widget=SizeBox_103
   properties {bOverride_WidthOverride=true, maxDesiredHeight=700, maxDesiredWidth=500, widthOverride=510}
  }
  Switcher_PickerType {
   widgetClassPath=&CommonWidgetSwitcherLegacy
   bIsVariable=true
   parent=SizeBox_103
   widget=Switcher_PickerType
   properties {transitionDuration=0}
   slot {class=SizeBoxSlot}
  }
  VBox_Swatches {widgetClassPath=&VerticalBox, bIsVariable=true, parent=Switcher_PickerType, widget=VBox_Swatches, slot {class=WidgetSwitcherSlot}}
  ScrollBox_38 {
   widgetClassPath=&ScrollBox
   parent=VBox_Swatches
   widget=ScrollBox_38
   properties {
    navigation=&self:WidgetTree.ScrollBox_38.WidgetNavigation_0
    orientation=Orient_Horizontal
    scrollWhenFocusChanges=AnimatedScroll
   }
   slot {class=VerticalBoxSlot, properties {size {sizeRule=Fill, value=1}}}
  }
  GridPanel_Swatches {
   widgetClassPath=/Script/UMG.UniformGridPanel
   bInherited=true
   bIsVariable=true
   parent=ScrollBox_38
   widget=GridPanel_Swatches
   properties {
    navigation=&self:WidgetTree.GridPanel_Swatches.WidgetNavigation_0
    slotPadding=<ltrb 8 0 8 0>
   }
   slot {class=ScrollBoxSlot}
  }
  ColorSwatch {
   widgetClassPath=&/ColorSwatch
   bIsVariable=true
   parent=GridPanel_Swatches
   widget=ColorSwatch
   properties {
    onButtonBaseClicked=(&self:WidgetTree.ColorSwatch.HandleOnButtonClicked)
    onButtonBaseDoubleClicked=(&self:WidgetTree.ColorSwatch.HandleOnButtonDoubleClicked)
    onButtonBaseHovered=(&self:WidgetTree.ColorSwatch.HandleOnButtonHovered)
    onButtonBaseUnhovered=(&self:WidgetTree.ColorSwatch.HandleOnButtonUnhovered)
    onSelectedChangedBase=(&self:WidgetTree.ColorSwatch.HandleOnSelectedChanged)
   }
   slot {class=UniformGridSlot}
  }
  ColorSwatch_97 {
   widgetClassPath=&ColorSwatch
   bIsVariable=true
   parent=GridPanel_Swatches
   widget=ColorSwatch_97
   properties {
    onButtonBaseClicked=(&self:WidgetTree.ColorSwatch_97.HandleOnButtonClicked)
    onButtonBaseDoubleClicked=(&self:WidgetTree.ColorSwatch_97.HandleOnButtonDoubleClicked)
    onButtonBaseHovered=(&self:WidgetTree.ColorSwatch_97.HandleOnButtonHovered)
    onButtonBaseUnhovered=(&self:WidgetTree.ColorSwatch_97.HandleOnButtonUnhovered)
    onSelectedChangedBase=(&self:WidgetTree.ColorSwatch_97.HandleOnSelectedChanged)
   }
   slot {class=UniformGridSlot, properties {row=1}}
  }
  ColorSwatch_169 {
   widgetClassPath=&ColorSwatch
   bIsVariable=true
   parent=GridPanel_Swatches
   widget=ColorSwatch_169
   properties {
    onButtonBaseClicked=(&self:WidgetTree.ColorSwatch_169.HandleOnButtonClicked)
    onButtonBaseDoubleClicked=(&self:WidgetTree.ColorSwatch_169.HandleOnButtonDoubleClicked)
    onButtonBaseHovered=(&self:WidgetTree.ColorSwatch_169.HandleOnButtonHovered)
    onButtonBaseUnhovered=(&self:WidgetTree.ColorSwatch_169.HandleOnButtonUnhovered)
    onSelectedChangedBase=(&self:WidgetTree.ColorSwatch_169.HandleOnSelectedChanged)
   }
   slot {class=UniformGridSlot, properties {row=2}}
  }
  ColorSwatch_240 {
   widgetClassPath=&ColorSwatch
   bIsVariable=true
   parent=GridPanel_Swatches
   widget=ColorSwatch_240
   properties {
    onButtonBaseClicked=(&self:WidgetTree.ColorSwatch_240.HandleOnButtonClicked)
    onButtonBaseDoubleClicked=(&self:WidgetTree.ColorSwatch_240.HandleOnButtonDoubleClicked)
    onButtonBaseHovered=(&self:WidgetTree.ColorSwatch_240.HandleOnButtonHovered)
    onButtonBaseUnhovered=(&self:WidgetTree.ColorSwatch_240.HandleOnButtonUnhovered)
    onSelectedChangedBase=(&self:WidgetTree.ColorSwatch_240.HandleOnSelectedChanged)
   }
   slot {class=UniformGridSlot, properties {column=1}}
  }
  Image_Separator_Up {
   widgetClassPath=&Image
   bIsVariable=true
   parent=VBox_Swatches
   widget=Image_Separator_Up
   properties {
    brush {
     bIsDynamicallyLoaded=false
     drawAs=Box
     imageType=NoImage
     mirroring=NoMirror
     resourceName=None
     resourceObject=/Game/Athena/UI/ReplayBrowser/M_Ui_Skewed_Vfor9slices_Inst1
     tiling=NoTile
     imageSize=<16 12>
     margin=<ltrb 0.5 0.5 0.5 0.5>
     outlineSettings {
      bUseBrushTransparency=false
      roundingType=HalfHeightRadius
      width=0
      color {colorUseRule=UseColor_Specified, specifiedColor=<rgba 0 0 0 0>}
      cornerRadii=<0 0 0 0>
     }
     tintColor {colorUseRule=UseColor_Specified, specifiedColor=<rgba 1 1 1 1>}
     uVRegion {bIsValid=false, max=<0 0>, min=<0 0>}
    }
    colorAndOpacity=<rgba 0 0.0208 0.25 1>
   }
   slot {class=VerticalBoxSlot, properties {verticalAlignment=VAlign_Top}}
  }
  HorizontalBox_1 {
   widgetClassPath=&HorizontalBox
   parent=VBox_Swatches
   widget=HorizontalBox_1
   properties {renderTransform {angle=0, scale=<1 1>, shear=<4 0>, translation=<0 0>}}
   slot {class=VerticalBoxSlot, properties {padding=<ltrb 4 16 4 4>}}
  }
  Button_CustomColor {
   widgetClassPath=&IconTextButton
   bInherited=true
   bIsVariable=true
   parent=HorizontalBox_1
   widget=Button_CustomColor
   properties {
    buttonText=Custom Color
    onButtonBaseClicked=(&self:WidgetTree.Button_CustomColor.HandleOnButtonClicked)
    onButtonBaseDoubleClicked=(&self:WidgetTree.Button_CustomColor.HandleOnButtonDoubleClicked)
    onButtonBaseHovered=(&self:WidgetTree.Button_CustomColor.HandleOnButtonHovered)
    onButtonBaseUnhovered=(&self:WidgetTree.Button_CustomColor.HandleOnButtonUnhovered)
    onSelectedChangedBase=(&self:WidgetTree.Button_CustomColor.HandleOnSelectedChanged)
    style=/FortUILibrary/Foundation/Buttons/ButtonStyles/Athena/Athena_ButtonStyle_AngledBlueMenuButton.Athena_ButtonStyle_AngledBlueMenuButton_C
    triggeringInputAction {dataTable=&InputActionDataTable, rowName=GenericRightStickClick}
   }
   slot {class=HorizontalBoxSlot, properties {padding=<ltrb 8 8 8 8>, size {sizeRule=Fill, value=1}}}
  }
  Swatch_CustomColor {
   widgetClassPath=&ColorSwatch
   bInherited=true
   bIsVariable=true
   parent=HorizontalBox_1
   widget=Swatch_CustomColor
   properties {
    bApplyAlphaOnDisable=false
    onButtonBaseClicked=(&self:WidgetTree.Swatch_CustomColor.HandleOnButtonClicked)
    onButtonBaseDoubleClicked=(&self:WidgetTree.Swatch_CustomColor.HandleOnButtonDoubleClicked)
    onButtonBaseHovered=(&self:WidgetTree.Swatch_CustomColor.HandleOnButtonHovered)
    onButtonBaseUnhovered=(&self:WidgetTree.Swatch_CustomColor.HandleOnButtonUnhovered)
    onSelectedChangedBase=(&self:WidgetTree.Swatch_CustomColor.HandleOnSelectedChanged)
    visibility=HitTestInvisible
    renderTransform {angle=45, scale=<1 1>, shear=<0 0>, translation=<0 0>}
   }
   slot {class=HorizontalBoxSlot, properties {padding=<ltrb 24 0 24 0>}}
  }
  VBox_CustomColor {
   widgetClassPath=&VerticalBox
   bIsVariable=true
   parent=Switcher_PickerType
   widget=VBox_CustomColor
   slot {class=WidgetSwitcherSlot}
  }
  VerticalBox_0 {
   widgetClassPath=&VerticalBox
   parent=VBox_CustomColor
   widget=VerticalBox_0
   properties {navigation=&self:WidgetTree.VerticalBox_0.WidgetNavigation_0}
   slot {class=VerticalBoxSlot, properties {padding=<ltrb 8 4 8 4>, size {sizeRule=Fill, value=1}}}
  }
  ColorComponent_Component1 {
   widgetClassPath=&/ColorPickerComponent
   bInherited=true
   bIsVariable=true
   parent=VerticalBox_0
   widget=ColorComponent_Component1
   slot {class=VerticalBoxSlot, properties {size {sizeRule=Fill, value=1}}}
  }
  ColorComponent_Component2 {
   widgetClassPath=&ColorPickerComponent
   bInherited=true
   bIsVariable=true
   parent=VerticalBox_0
   widget=ColorComponent_Component2
   slot {class=VerticalBoxSlot, properties {size {sizeRule=Fill, value=1}}}
  }
  ColorComponent_Component3 {
   widgetClassPath=&ColorPickerComponent
   bInherited=true
   bIsVariable=true
   parent=VerticalBox_0
   widget=ColorComponent_Component3
   slot {class=VerticalBoxSlot, properties {size {sizeRule=Fill, value=1}}}
  }
  Image {
   widgetClassPath=&Image
   bIsVariable=true
   parent=VBox_CustomColor
   widget=Image
   properties {
    brush {
     bIsDynamicallyLoaded=false
     drawAs=Box
     imageType=NoImage
     mirroring=NoMirror
     resourceName=None
     resourceObject=&M_Ui_Skewed_Vfor9slices_Inst1
     tiling=NoTile
     imageSize=<16 12>
     margin=<ltrb 0.5 0.5 0.5 0.5>
     outlineSettings {
      bUseBrushTransparency=false
      roundingType=HalfHeightRadius
      width=0
      color {colorUseRule=UseColor_Specified, specifiedColor=<rgba 0 0 0 0>}
      cornerRadii=<0 0 0 0>
     }
     tintColor {colorUseRule=UseColor_Specified, specifiedColor=<rgba 1 1 1 1>}
     uVRegion {bIsValid=false, max=<0 0>, min=<0 0>}
    }
    colorAndOpacity=<rgba 0 0.0208 0.25 1>
   }
   slot {class=VerticalBoxSlot, properties {verticalAlignment=VAlign_Top, padding=<ltrb 0 0 8 0>}}
  }
  HorizontalBox_0 {
   widgetClassPath=&HorizontalBox
   parent=VBox_CustomColor
   widget=HorizontalBox_0
   properties {renderTransform {angle=0, scale=<1 1>, shear=<4 0>, translation=<0 0>}}
   slot {class=VerticalBoxSlot, properties {padding=<ltrb 4 16 4 4>}}
  }
  Button_SaveColor {
   widgetClassPath=&IconTextButton
   bInherited=true
   bIsVariable=true
   parent=HorizontalBox_0
   widget=Button_SaveColor
   properties {
    bIgnoreInputActionWidgetText=true
    buttonText=Save Color
    onButtonBaseClicked=(&self:WidgetTree.Button_SaveColor.HandleOnButtonClicked)
    onButtonBaseDoubleClicked=(&self:WidgetTree.Button_SaveColor.HandleOnButtonDoubleClicked)
    onButtonBaseHovered=(&self:WidgetTree.Button_SaveColor.HandleOnButtonHovered)
    onButtonBaseUnhovered=(&self:WidgetTree.Button_SaveColor.HandleOnButtonUnhovered)
    onSelectedChangedBase=(&self:WidgetTree.Button_SaveColor.HandleOnSelectedChanged)
    style=&Athena_ButtonStyle_AngledBlueMenuButton_C
    triggeringInputAction {dataTable=&InputActionDataTable, rowName=GenericFaceButtonBottom}
   }
   slot {class=HorizontalBoxSlot, properties {padding=<ltrb 8 8 8 8>, size {sizeRule=Fill, value=1}}}
  }
  Swatch_CustomColorPreview {
   widgetClassPath=&ColorSwatch
   bInherited=true
   bIsVariable=true
   parent=HorizontalBox_0
   widget=Swatch_CustomColorPreview
   properties {
    bIsFocusable=false
    bShouldUseFallbackDefaultInputAction=false
    onButtonBaseClicked=(&self:WidgetTree.Swatch_CustomColorPreview.HandleOnButtonClicked)
    onButtonBaseDoubleClicked=(&self:WidgetTree.Swatch_CustomColorPreview.HandleOnButtonDoubleClicked)
    onButtonBaseHovered=(&self:WidgetTree.Swatch_CustomColorPreview.HandleOnButtonHovered)
    onButtonBaseUnhovered=(&self:WidgetTree.Swatch_CustomColorPreview.HandleOnButtonUnhovered)
    onSelectedChangedBase=(&self:WidgetTree.Swatch_CustomColorPreview.HandleOnSelectedChanged)
    visibility=HitTestInvisible
    renderTransform {angle=45, scale=<1 1>, shear=<0 0>, translation=<0 0>}
   }
   slot {class=HorizontalBoxSlot, properties {padding=<ltrb 24 0 24 0>}}
  }
 ]
}
