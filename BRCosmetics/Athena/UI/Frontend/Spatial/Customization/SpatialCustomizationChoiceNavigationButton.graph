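# Arrow button for stepping through choices in the spatial cosmetic customization frontend, built on CommonButtonLegacy.
# PreConstruct swaps the default reversed arrow for the RightDirection texture when bRight is set.
# Construct enlarges the arrow on mobile, then binds to input method changes, hiding the arrow on gamepad while the inherited input action widget overlays it.
# Hover plays a short animation that scales and brightens the arrow, reversing on unhover.
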WidgetBlueprint /BRCosmetics/Athena/UI/Frontend/Spatial/Customization/SpatialCustomizationChoiceNavigationButton
animations [hover {
  binding_count=1
  length_seconds=0.1
  object=&self:hover
  sequence {bindings [Image_Direction {
     id=#1
     possessed_class=FortLazyImage
     tracks [
      {
       type=MovieScene2DTransformTrack
       sections [{
         type=MovieScene2DTransformSection
         end_frame=2
         start_frame=0
         channels [
          Translation.X {type=MovieSceneScriptingFloatChannel}
          Translation.Y {type=MovieSceneScriptingFloatChannel}
          Angle {type=MovieSceneScriptingFloatChannel}
          Scale.X {type=MovieSceneScriptingFloatChannel, keys [[0,1],[2,1.1]]}
          Scale.Y {type=MovieSceneScriptingFloatChannel, keys [[0,1],[2,1.1]]}
          Shear.X {type=MovieSceneScriptingFloatChannel}
          Shear.Y {type=MovieSceneScriptingFloatChannel}
         ]
        }]
      }
      {
       type=MovieSceneColorTrack
       display_name=Color and Opacity
       sections [{
         type=MovieSceneColorSection
         end_frame=2
         start_frame=0
         channels [
          Color.R {type=MovieSceneScriptingFloatChannel, const [0,2]}
          Color.G {type=MovieSceneScriptingFloatChannel, keys [[0,0.125],[2,0.175]]}
          Color.B {type=MovieSceneScriptingFloatChannel, keys [[0,0.75],[2,1]]}
          Color.A {type=MovieSceneScriptingFloatChannel, const [1,1]}
         ]
        }]
      }
     ]
    }]}
 }]
graphs [EventGraph {nodes [
   n1 = InputSourceChanged!() -> (OutputDelegate, then, bNewInputType:ECommonInputType Enum) {is=title K2Node_CustomEvent}
   n2 = GetCurrentInputType(self=Get CommonInputSubsystem!(PlayerController:Player Controller Object Reference=GetOwningPlayer()) -> (ReturnValue:Common Input Object Reference) {is=title K2Node_GetSubsystemFromPC})
   SetVisibility(
    execute=[@n5.then, @n1.then]
    self=Get Image_Direction() -> (Image_Direction:Fort Lazy Image Object Reference)
    InVisibility=Select(Option 0=Visible, Option 1=Hidden, Index="Equal (Enum)"(A=@n2, B=Gamepad))
   )
   Event PreConstruct()
   > Branch(Condition=Get bRight() -> (bRight:bool))
   > SetBrushFromLazyTexture!(
    execute
    self:Common Lazy Image Object Reference=Get Image_Direction()
    LazyTexture:"Texture 2D Soft Object Reference (by ref)"=Get RightDirection() -> (RightDirection:Texture 2D Soft Object Reference)
    bMatchSize:bool=false
   ) -> (then)
   Event Construct()
   > n3 = Branch(Condition=IsMobileGame())
   > n4 = SetDesiredSizeOverride!(execute, self:Image Object Reference=Get Image_Direction(), DesiredSize:Vector 2D Structure=(X=75,Y=75)) -> (then)
   n5 = Bind Event to On Input Method Changed(execute=[@n4.then, @n3.else], self=Get CommonInputSubsystem(PlayerController=GetOwningPlayer()), Delegate=Create Event())
   Event On Hovered()
   n6 = Get hover() -> (hover:Widget Animation Object Reference)
   > PlayAnimationForward(InAnimation=@n6)
   Event On Unhovered()
   > PlayAnimationReverse(InAnimation=@n6)
  ]}]
variables [
 bRight {type=boolean}
 RightDirection {type=Texture2D, default=/Game/UI/SocialPanel/Art/Textures/T_UI_Icon_TreeArrow_64}
]
widgets {
 info {inheritedWidgetCount=1, parentClass=&CommonButtonLegacy, rootWidgetClass=&Overlay, widgetCount=3}
 widgets [
  Overlay_0 {widgetClassPath=&Overlay, widget=Overlay_0}
  Image_Direction {
   widgetClassPath=&FortLazyImage
   bIsVariable=true
   parent=Overlay_0
   widget=Image_Direction
   properties {
    brush {
     bIsDynamicallyLoaded=false
     drawAs=Image
     imageType=NoImage
     mirroring=NoMirror
     resourceName=None
     resourceObject=/Game/UI/SocialPanel/Art/Textures/T_UI_Icon_TreeArrow_64_reversed
     tiling=NoTile
     imageSize=<64 64>
     margin=<ltrb 0 0 0 0>
     outlineSettings {
      bUseBrushTransparency=false
      roundingType=HalfHeightRadius
      width=0
      color {colorUseRule=UseColor_Specified, specifiedColor=<rgba 0 0 0 0>}
      cornerRadii=<0 0 0 0>
     }
     tintColor {colorUseRule=UseColor_Specified, specifiedColor=<rgba 1 1 1 1>}
     uVRegion {bIsValid=false, max=<0 0>, min=<0 0>}
    }
    colorAndOpacity=<rgba 0 0.125 0.75 1>
   }
   slot {class=OverlaySlot, properties {horizontalAlignment=HAlign_Fill, verticalAlignment=VAlign_Fill}}
  }
  InputActionWidget {
   widgetClassPath=&CommonActionWidget
   bInherited=true
   bIsVariable=true
   parent=Overlay_0
   widget=InputActionWidget
   properties {visibility=SelfHitTestInvisible}
   slot {class=OverlaySlot, properties {horizontalAlignment=HAlign_Center, verticalAlignment=VAlign_Center}}
  }
 ]
}
