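# Spatial customization widget that shows progress on the challenge required to select a cosmetic upgrade.
# A header pairs a tilted unlock icon with fixed "Complete Challenge to select upgrade" text, above a description, two overlaid progress bars and a sheared count label.
# The event graph is empty; description, bars and count are widgets inherited from native AthenaQuestProgression, so this asset supplies only layout and styling.
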WidgetBlueprint /BRCosmetics/Athena/UI/Frontend/Spatial/Customization/SpatialQuestProgress
graphs [EventGraph {links [], nodes []}]
variables []
widgets {
 info {inheritedWidgetCount=4, parentClass=/Script/FortniteUI.AthenaQuestProgression, rootWidgetClass=&Border, widgetCount=15}
 widgets [
  Border_Challenge {
   widgetClassPath=&Border
   bIsVariable=true
   widget=Border_Challenge
   properties {
    background {
     bIsDynamicallyLoaded=false
     drawAs=NoDrawType
     imageType=NoImage
     mirroring=NoMirror
     resourceName=None
     resourceObject=null
     tiling=NoTile
     imageSize=<128 128>
     margin=<ltrb 0.5 0.25 0.25 0.25>
     outlineSettings {
      bUseBrushTransparency=false
      roundingType=HalfHeightRadius
      width=0
      color {colorUseRule=UseColor_Specified, specifiedColor=<rgba 0 0 0 0>}
      cornerRadii=<0 0 0 0>
     }
     tintColor {colorUseRule=UseColor_Specified, specifiedColor=<rgba 1 1 1 1>}
     uVRegion {bIsValid=false, max=<0 0>, min=<0 0>}
    }
    padding=<ltrb 0 0 0 0>
   }
  }
  VerticalBox_1 {
   widgetClassPath=&VerticalBox
   parent=Border_Challenge
   widget=VerticalBox_1
   slot {class=BorderSlot, properties {padding=<ltrb 0 0 0 0>}}
  }
  Border_71 {
   widgetClassPath=&Border
   parent=VerticalBox_1
   widget=Border_71
   properties {
    background {
     bIsDynamicallyLoaded=false
     drawAs=NoDrawType
     imageType=NoImage
     mirroring=NoMirror
     resourceName=None
     resourceObject=null
     tiling=NoTile
     imageSize=<128 128>
     margin=<ltrb 0.5 0.25 0.25 0.25>
     outlineSettings {
      bUseBrushTransparency=false
      roundingType=HalfHeightRadius
      width=0
      color {colorUseRule=UseColor_Specified, specifiedColor=<rgba 0 0 0 0>}
      cornerRadii=<0 0 0 0>
     }
     tintColor {colorUseRule=UseColor_Specified, specifiedColor=<rgba 1 1 1 1>}
     uVRegion {bIsValid=false, max=<0 0>, min=<0 0>}
    }
    brushColor=<rgba 0.0125 0.0938 0.5 0.5>
    padding=<ltrb 0 0 0 0>
   }
   slot {class=VerticalBoxSlot}
  }
  HorizontalBox_0 {
   widgetClassPath=&HorizontalBox
   parent=Border_71
   widget=HorizontalBox_0
   slot {class=BorderSlot, properties {padding=<ltrb 0 0 0 0>}}
  }
  Image_107 {
   widgetClassPath=&Image
   bIsVariable=true
   parent=HorizontalBox_0
   widget=Image_107
   properties {
    brush {
     bIsDynamicallyLoaded=false
     drawAs=Image
     imageType=NoImage
     mirroring=NoMirror
     resourceName=None
     resourceObject=/Game/UI/Foundation/Textures/Icons/Locks/T-Icon-Unlocked-64
     tiling=NoTile
     imageSize=<40 40>
     margin=<ltrb 0 0 0 0>
     outlineSettings {
      bUseBrushTransparency=false
      roundingType=HalfHeightRadius
      width=0
      color {colorUseRule=UseColor_Specified, specifiedColor=<rgba 0 0 0 0>}
      cornerRadii=<0 0 0 0>
     }
     tintColor {colorUseRule=UseColor_Specified, specifiedColor=<rgba 1 1 1 1>}
     uVRegion {bIsValid=false, max=<0 0>, min=<0 0>}
    }
    colorAndOpacity=<rgba 0.25 0.875 1 1>
    renderTransform {angle=-15, scale=<1 1>, shear=<0 0>, translation=<0 0>}
   }
   slot {class=HorizontalBoxSlot, properties {horizontalAlignment=HAlign_Center, verticalAlignment=VAlign_Center, padding=<ltrb -2 0 6 0>}}
  }
  CommonRichTextBlock_54 {
   widgetClassPath=&CommonRichTextBlock
   bIsVariable=true
   parent=HorizontalBox_0
   widget=CommonRichTextBlock_54
   properties {
    bOverrideDefaultStyle=true
    lineHeightPercentage=0.95
    scrollStyle=&TextScrollStyle-Base_C
    text="<text color=\"ffffffff\" fontface=\"black\">Complete Challenge</> to select upgrade"
    textTransformPolicy=ToUpper
    defaultTextStyleOverride {
     overflowPolicy=Clip
     transformPolicy=None
     colorAndOpacity {colorUseRule=UseColor_Specified, specifiedColor=<rgba 0.25 0.875 1 1>}
     font {
      bForceMonospaced=false
      bMaterialIsStencil=false
      fontMaterial=null
      fontObject=&BurbankSmall
      letterSpacing=0
      monospacedWidth=1
      size=20
      skewAmount=0
      typefaceFontName=Bold
      outlineSettings {
       bApplyOutlineToDropShadows=false
       bMiteredCorners=false
       bSeparateFillAlpha=false
       outlineMaterial=null
       outlineSize=0
       outlineColor=<rgba 0 0 0 1>
      }
     }
     highlightColor {colorUseRule=UseColor_Specified, specifiedColor=<rgba 0 0 0 1>}
     highlightShape {
      bIsDynamicallyLoaded=false
      drawAs=Image
      imageType=NoImage
      mirroring=NoMirror
      resourceName=None
      resourceObject=null
      tiling=NoTile
      imageSize=<32 32>
      margin=<ltrb 0 0 0 0>
      outlineSettings {
       bUseBrushTransparency=false
       roundingType=HalfHeightRadius
       width=0
       color {colorUseRule=UseColor_Specified, specifiedColor=<rgba 0 0 0 0>}
       cornerRadii=<0 0 0 0>
      }
      tintColor {colorUseRule=UseColor_Specified, specifiedColor=<rgba 1 1 1 1>}
      uVRegion {bIsValid=false, max=<0 0>, min=<0 0>}
     }
     selectedBackgroundColor {colorUseRule=UseColor_Foreground, specifiedColor=<rgba 1 0 1 1>}
     shadowColorAndOpacity=<rgba 0 0 0 1>
     shadowOffset=<0 0>
     strikeBrush {
      bIsDynamicallyLoaded=false
      drawAs=Image
      imageType=NoImage
      mirroring=NoMirror
      resourceName=None
      resourceObject=null
      tiling=NoTile
      imageSize=<32 32>
      margin=<ltrb 0 0 0 0>
      outlineSettings {
       bUseBrushTransparency=false
       roundingType=HalfHeightRadius
       width=0
       color {colorUseRule=UseColor_Specified, specifiedColor=<rgba 0 0 0 0>}
       cornerRadii=<0 0 0 0>
      }
      tintColor {colorUseRule=UseColor_Specified, specifiedColor=<rgba 1 1 1 1>}
      uVRegion {bIsValid=false, max=<0 0>, min=<0 0>}
     }
     underlineBrush {
      bIsDynamicallyLoaded=false
      drawAs=Image
      imageType=NoImage
      mirroring=NoMirror
      resourceName=None
      resourceObject=null
      tiling=NoTile
      imageSize=<32 32>
      margin=<ltrb 0 0 0 0>
      outlineSettings {
       bUseBrushTransparency=false
       roundingType=HalfHeightRadius
       width=0
       color {colorUseRule=UseColor_Specified, specifiedColor=<rgba 0 0 0 0>}
       cornerRadii=<0 0 0 0>
      }
      tintColor {colorUseRule=UseColor_Specified, specifiedColor=<rgba 1 1 1 1>}
      uVRegion {bIsValid=false, max=<0 0>, min=<0 0>}
     }
    }
   }
   slot {class=HorizontalBoxSlot, properties {horizontalAlignment=HAlign_Left, verticalAlignment=VAlign_Center, size {sizeRule=Fill, value=1}}}
  }
  VerticalBox_2 {
   widgetClassPath=&VerticalBox
   parent=VerticalBox_1
   widget=VerticalBox_2
   slot {class=VerticalBoxSlot, properties {padding=<ltrb 40 8 0 20>, size {sizeRule=Fill, value=1}}}
  }
  Text_ChallengeDescription {
   widgetClassPath=&CommonTextBlock
   bInherited=true
   bIsVariable=true
   parent=VerticalBox_2
   widget=Text_ChallengeDescription
   properties {
    text="{Mission description}"
    wrapTextAt=512
    colorAndOpacity {colorUseRule=UseColor_Specified, specifiedColor=<rgba 0 0.2502 1 1>}
    font {
     bForceMonospaced=false
     bMaterialIsStencil=false
     fontMaterial=null
     fontObject=&BurbankSmall
     letterSpacing=0
     monospacedWidth=1
     size=18
     skewAmount=0
     typefaceFontName=Bold
     outlineSettings {
      bApplyOutlineToDropShadows=false
      bMiteredCorners=false
      bSeparateFillAlpha=false
      outlineMaterial=null
      outlineSize=0
      outlineColor=<rgba 0 0 0 1>
     }
    }
    shadowOffset=<0 0>
   }
   slot {class=VerticalBoxSlot, properties {verticalAlignment=VAlign_Top, padding=<ltrb 0 0 88 0>, size {sizeRule=Fill, value=1}}}
  }
  HorizontalBox_1 {
   widgetClassPath=&HorizontalBox
   parent=VerticalBox_2
   widget=HorizontalBox_1
   slot {class=VerticalBoxSlot, properties {horizontalAlignment=HAlign_Left, verticalAlignment=VAlign_Bottom, padding=<ltrb 0 8 0 0>}}
  }
  SizeBox_2 {
   widgetClassPath=&SizeBox
   parent=HorizontalBox_1
   widget=SizeBox_2
   properties {bOverride_HeightOverride=true, bOverride_WidthOverride=true, heightOverride=16, widthOverride=512}
   slot {
    class=HorizontalBoxSlot
    properties {horizontalAlignment=HAlign_Left, verticalAlignment=VAlign_Center, padding=<ltrb 4 0 0 0>, size {sizeRule=Fill, value=1}}
   }
  }
  Overlay_3 {widgetClassPath=&Overlay, parent=SizeBox_2, widget=Overlay_3, slot {class=SizeBoxSlot}}
  ProgressBar_Challenge_Back {
   widgetClassPath=/Script/UMG.ProgressBar
   bInherited=true
   bIsVariable=true
   parent=Overlay_3
   widget=ProgressBar_Challenge_Back
   properties {
    percent=0.5524
    widgetStyle {
     enableFillAnimation=false
     backgroundImage {
      bIsDynamicallyLoaded=false
      drawAs=Box
      imageType=FullColor
      mirroring=NoMirror
      resourceName=../../../Engine/Content/Slate/Common/ProgressBar_Background.png
      resourceObject=/FortniteCoreUI/Frontend/Spatial/Art/M_Ui_Spacial_Seaparator_Background_progress
      tiling=NoTile
      imageSize=<8 8>
      margin=<ltrb 0.25 0.25 0.25 0.25>
      outlineSettings {
       bUseBrushTransparency=false
       roundingType=HalfHeightRadius
       width=0
       color {colorUseRule=UseColor_Specified, specifiedColor=<rgba 0 0 0 0>}
       cornerRadii=<0 0 0 0>
      }
      tintColor {colorUseRule=UseColor_Specified, specifiedColor=<rgba 0 0.0417 0.25 1>}
      uVRegion {bIsValid=false, max=<0 0>, min=<0 0>}
     }
     fillImage {
      bIsDynamicallyLoaded=false
      drawAs=Box
      imageType=FullColor
      mirroring=NoMirror
      resourceName=../../../Engine/Content/Slate/Common/ProgressBar_Fill.png
      resourceObject=&M_Ui_Spacial_Seaparator_Background_progress
      tiling=NoTile
      imageSize=<8 8>
      margin=<ltrb 0.25 0.25 0.25 0.25>
      outlineSettings {
       bUseBrushTransparency=false
       roundingType=HalfHeightRadius
       width=0
       color {colorUseRule=UseColor_Specified, specifiedColor=<rgba 0 0 0 0>}
       cornerRadii=<0 0 0 0>
      }
      tintColor {colorUseRule=UseColor_Specified, specifiedColor=<rgba 0 0.0417 0.25 1>}
      uVRegion {bIsValid=false, max=<0 0>, min=<0 0>}
     }
     marqueeImage {
      bIsDynamicallyLoaded=false
      drawAs=Image
      imageType=FullColor
      mirroring=NoMirror
      resourceName=../../../Engine/Content/Slate/Common/ProgressBar_Marquee.png
      resourceObject=null
      tiling=Horizontal
      imageSize=<20 12>
      margin=<ltrb 0 0 0 0>
      outlineSettings {
       bUseBrushTransparency=false
       roundingType=HalfHeightRadius
       width=0
       color {colorUseRule=UseColor_Specified, specifiedColor=<rgba 0 0 0 0>}
       cornerRadii=<0 0 0 0>
      }
      tintColor {colorUseRule=UseColor_Specified, specifiedColor=<rgba 1 1 1 1>}
      uVRegion {bIsValid=false, max=<0 0>, min=<0 0>}
     }
    }
   }
   slot {class=OverlaySlot, properties {horizontalAlignment=HAlign_Fill, verticalAlignment=VAlign_Fill}}
  }
  ProgressBar_Challenge_Front {
   widgetClassPath=&ProgressBar
   bInherited=true
   bIsVariable=true
   parent=Overlay_3
   widget=ProgressBar_Challenge_Front
   properties {
    percent=0.25
    fillColorAndOpacity=<rgba 0.0331 0.855 1 1>
    widgetStyle {
     enableFillAnimation=false
     backgroundImage {
      bIsDynamicallyLoaded=false
      drawAs=Box
      imageType=FullColor
      mirroring=NoMirror
      resourceName=../../../Engine/Content/Slate/Common/ProgressBar_Background.png
      resourceObject=&M_Ui_Spacial_Seaparator_Background_progress
      tiling=NoTile
      imageSize=<0 0>
      margin=<ltrb 0.2 0.2 0.2 0.2>
      outlineSettings {
       bUseBrushTransparency=false
       roundingType=HalfHeightRadius
       width=0
       color {colorUseRule=UseColor_Specified, specifiedColor=<rgba 0 0 0 0>}
       cornerRadii=<0 0 0 0>
      }
      tintColor {colorUseRule=UseColor_Specified, specifiedColor=<rgba 0 0.2502 1 0.25>}
      uVRegion {bIsValid=false, max=<0 0>, min=<0 0>}
     }
     fillImage {
      bIsDynamicallyLoaded=false
      drawAs=Box
      imageType=FullColor
      mirroring=NoMirror
      resourceName=../../../Engine/Content/Slate/Common/ProgressBar_Fill.png
      resourceObject=&M_Ui_Spacial_Seaparator_Background_progress
      tiling=NoTile
      imageSize=<0 0>
      margin=<ltrb 0.25 0.25 0.25 0.25>
      outlineSettings {
       bUseBrushTransparency=false
       roundingType=HalfHeightRadius
       width=0
       color {colorUseRule=UseColor_Specified, specifiedColor=<rgba 0 0 0 0>}
       cornerRadii=<0 0 0 0>
      }
      tintColor {colorUseRule=UseColor_Specified, specifiedColor=<rgba 0.5583 0.9473 1 1>}
      uVRegion {bIsValid=false, max=<0 0>, min=<0 0>}
     }
     marqueeImage {
      bIsDynamicallyLoaded=false
      drawAs=Image
      imageType=FullColor
      mirroring=NoMirror
      resourceName=../../../Engine/Content/Slate/Common/ProgressBar_Marquee.png
      resourceObject=null
      tiling=Horizontal
      imageSize=<20 12>
      margin=<ltrb 0 0 0 0>
      outlineSettings {
       bUseBrushTransparency=false
       roundingType=HalfHeightRadius
       width=0
       color {colorUseRule=UseColor_Specified, specifiedColor=<rgba 0 0 0 0>}
       cornerRadii=<0 0 0 0>
      }
      tintColor {colorUseRule=UseColor_Specified, specifiedColor=<rgba 1 1 1 1>}
      uVRegion {bIsValid=false, max=<0 0>, min=<0 0>}
     }
    }
   }
   slot {class=OverlaySlot, properties {horizontalAlignment=HAlign_Fill, verticalAlignment=VAlign_Fill}}
  }
  SizeBox_3 {
   widgetClassPath=&SizeBox
   parent=HorizontalBox_1
   widget=SizeBox_3
   properties {bOverride_MinDesiredWidth=true, minDesiredWidth=88, widthOverride=128}
   slot {class=HorizontalBoxSlot, properties {horizontalAlignment=HAlign_Center, verticalAlignment=VAlign_Bottom, padding=<ltrb 8 0 0 0>}}
  }
  RichText_ChallengeCount {
   widgetClassPath=&CommonRichTextBlock
   bInherited=true
   bIsVariable=true
   parent=SizeBox_3
   widget=RichText_ChallengeCount
   properties {
    bOverrideDefaultStyle=true
    text="1<text color=\"0089FF\"> / 5</>"
    defaultTextStyleOverride {
     overflowPolicy=Clip
     transformPolicy=None
     colorAndOpacity {colorUseRule=UseColor_Specified, specifiedColor=<rgba 0.5583 0.9473 1 1>}
     font {
      bForceMonospaced=false
      bMaterialIsStencil=false
      fontMaterial=null
      fontObject=&BurbankSmall
      letterSpacing=0
      monospacedWidth=1
      size=18
      skewAmount=0
      typefaceFontName=Black
      outlineSettings {
       bApplyOutlineToDropShadows=false
       bMiteredCorners=false
       bSeparateFillAlpha=false
       outlineMaterial=null
       outlineSize=0
       outlineColor=<rgba 0 0 0 1>
      }
     }
     highlightColor {colorUseRule=UseColor_Specified, specifiedColor=<rgba 0 0 0 1>}
     highlightShape {
      bIsDynamicallyLoaded=false
      drawAs=Image
      imageType=NoImage
      mirroring=NoMirror
      resourceName=None
      resourceObject=null
      tiling=NoTile
      imageSize=<32 32>
      margin=<ltrb 0 0 0 0>
      outlineSettings {
       bUseBrushTransparency=false
       roundingType=HalfHeightRadius
       width=0
       color {colorUseRule=UseColor_Specified, specifiedColor=<rgba 0 0 0 0>}
       cornerRadii=<0 0 0 0>
      }
      tintColor {colorUseRule=UseColor_Specified, specifiedColor=<rgba 1 1 1 1>}
      uVRegion {bIsValid=false, max=<0 0>, min=<0 0>}
     }
     selectedBackgroundColor {colorUseRule=UseColor_Foreground, specifiedColor=<rgba 1 0 1 1>}
     shadowColorAndOpacity=<rgba 0 0 0 1>
     shadowOffset=<0 0>
     strikeBrush {
      bIsDynamicallyLoaded=false
      drawAs=Image
      imageType=NoImage
      mirroring=NoMirror
      resourceName=None
      resourceObject=null
      tiling=NoTile
      imageSize=<32 32>
      margin=<ltrb 0 0 0 0>
      outlineSettings {
       bUseBrushTransparency=false
       roundingType=HalfHeightRadius
       width=0
       color {colorUseRule=UseColor_Specified, specifiedColor=<rgba 0 0 0 0>}
       cornerRadii=<0 0 0 0>
      }
      tintColor {colorUseRule=UseColor_Specified, specifiedColor=<rgba 1 1 1 1>}
      uVRegion {bIsValid=false, max=<0 0>, min=<0 0>}
     }
     underlineBrush {
      bIsDynamicallyLoaded=false
      drawAs=Image
      imageType=NoImage
      mirroring=NoMirror
      resourceName=None
      resourceObject=null
      tiling=NoTile
      imageSize=<32 32>
      margin=<ltrb 0 0 0 0>
      outlineSettings {
       bUseBrushTransparency=false
       roundingType=HalfHeightRadius
       width=0
       color {colorUseRule=UseColor_Specified, specifiedColor=<rgba 0 0 0 0>}
       cornerRadii=<0 0 0 0>
      }
      tintColor {colorUseRule=UseColor_Specified, specifiedColor=<rgba 1 1 1 1>}
      uVRegion {bIsValid=false, max=<0 0>, min=<0 0>}
     }
    }
    renderTransform {angle=0, scale=<1 1>, shear=<-12 0>, translation=<0 0>}
   }
   slot {class=SizeBoxSlot, properties {horizontalAlignment=HAlign_Left}}
  }
 ]
}
