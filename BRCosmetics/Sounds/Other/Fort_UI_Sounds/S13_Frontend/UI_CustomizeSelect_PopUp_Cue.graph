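# UI sound cue for the customize-select pop-up in the Season 13 frontend, routed to the Fort_UI sound class.
# A fixed root delay holds back a Mixer that layers two waves from other frontend sets.
# The player-survey pop-up plays first at reduced input volume, followed by a Season 12 armory whoosh at full volume after a short, slightly randomized extra delay.
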SoundCue /BRCosmetics/Sounds/Other/Fort_UI_Sounds/S13_Frontend/UI_CustomizeSelect_PopUp_Cue
root=Delay_0
nodes [
 {id=Delay_0, children [Mixer_0], properties {delayMax=0.4, delayMin=0.4}}
 {id=Delay_1, children [WavePlayer_1], properties {delayMax=0.1, delayMin=0.08}}
 {id=Mixer_0, children [Delay_1,WavePlayer_0], properties {inputVolume [1,0.7]}}
 {id=WavePlayer_0, properties {soundWaveAssetPtr=/Game/Sounds/Fort_UI_Sounds/FrontEnd/Fort_UI_PlayerSurvey_PopUp_01}}
 {id=WavePlayer_1, properties {soundWaveAssetPtr=/Game/Sounds/Fort_UI_Sounds/S12_Frontend/UI_S12_Armory_Skin_Options_Whoosh_Back}}
]
properties {maxDistance=2097152, soundClassObject=&Fort_UI, volumeMultiplier=0.7}
