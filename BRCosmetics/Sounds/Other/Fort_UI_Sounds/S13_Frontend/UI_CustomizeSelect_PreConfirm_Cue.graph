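# UI sound cue for the pre-confirm step of a customize selection in the frontend, routed to the Fort_UI sound class.
# A Mixer layers a main button click at full volume with a quieter Battle Pass hover sound.
# The hover layer passes through a short fixed delay and a Modulator that pitches it up, so it trails just behind the click.
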SoundCue /BRCosmetics/Sounds/Other/Fort_UI_Sounds/S13_Frontend/UI_CustomizeSelect_PreConfirm_Cue
root=Mixer_0
nodes [
 {id=Delay_0, children [Modulator_0], properties {delayMax=0.02, delayMin=0.02}}
 {id=Mixer_0, children [WavePlayer_2,Delay_0], properties {inputVolume [1,0.65]}}
 {id=Modulator_0, children [WavePlayer_0], properties {pitchMax=1.2, pitchMin=1.2}}
 {id=WavePlayer_0, properties {soundWaveAssetPtr=/Game/Sounds/Fort_UI_Sounds/S13_Frontend/S13_S13_Battlepass_Main_Hover}}
 {id=WavePlayer_2, properties {soundWaveAssetPtr=/Game/Sounds/Fort_UI_Sounds/FrontEnd/New_FrontEnd/UI_FN_Button_Click_Main_01}}
]
properties {maxDistance=2097152, soundClassObject=&Fort_UI, volumeMultiplier=1}
