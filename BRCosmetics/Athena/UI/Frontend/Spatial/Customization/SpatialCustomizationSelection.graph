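# Spatial customization selection widget for the frontend, where the player browses categories and choices of cosmetic upgrades and finalizes one.
# Builds on native AthenaSpatialCustomizationSelection, with a three-item choice carousel or color picker and previous/next buttons.
# Category state selects the body and footer (Battle Pass purchase, locked, quest progress, finalize, used) and tints the header.
# Animations drive slide-ins, category changes, carousel switching (reversed for previous) and a success fanfare with sound.
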WidgetBlueprint /BRCosmetics/Athena/UI/Frontend/Spatial/Customization/SpatialCustomizationSelection
animations [
 IntroChoice {
  binding_count=3
  length_seconds=0.35
  object=&self:IntroChoice
  sequence {bindings [
    SizeBox_Footer {
     id=#1
     possessed_class=SizeBox
     tracks [{
       type=MovieScene2DTransformTrack
       sections [{
         type=MovieScene2DTransformSection
         channels [
          Translation.X {type=MovieSceneScriptingFloatChannel, keys [[2.0003,2000],[7,0]]}
          Translation.Y {type=MovieSceneScriptingFloatChannel}
          Angle {type=MovieSceneScriptingFloatChannel}
          Scale.X {type=MovieSceneScriptingFloatChannel}
          Scale.Y {type=MovieSceneScriptingFloatChannel}
          Shear.X {type=MovieSceneScriptingFloatChannel}
          Shear.Y {type=MovieSceneScriptingFloatChannel}
         ]
        }]
      }]
    }
    Border_Body {
     id=#2
     possessed_class=Border
     tracks [{
       type=MovieScene2DTransformTrack
       sections [{
         type=MovieScene2DTransformSection
         channels [
          Translation.X {type=MovieSceneScriptingFloatChannel, keys [[1,2000],[5.9997,0]]}
          Translation.Y {type=MovieSceneScriptingFloatChannel}
          Angle {type=MovieSceneScriptingFloatChannel}
          Scale.X {type=MovieSceneScriptingFloatChannel}
          Scale.Y {type=MovieSceneScriptingFloatChannel}
          Shear.X {type=MovieSceneScriptingFloatChannel}
          Shear.Y {type=MovieSceneScriptingFloatChannel}
         ]
        }]
      }]
    }
    SizeBox_Header {
     id=#3
     display_name="Image_1 (2)"
     possessed_class=Overlay
     tracks [{
       type=MovieScene2DTransformTrack
       sections [{
         type=MovieScene2DTransformSection
         channels [
          Translation.X {type=MovieSceneScriptingFloatChannel, keys [[0,2000],[4.9997,0]]}
          Translation.Y {type=MovieSceneScriptingFloatChannel}
          Angle {type=MovieSceneScriptingFloatChannel}
          Scale.X {type=MovieSceneScriptingFloatChannel}
          Scale.Y {type=MovieSceneScriptingFloatChannel}
          Shear.X {type=MovieSceneScriptingFloatChannel}
          Shear.Y {type=MovieSceneScriptingFloatChannel}
         ]
        }]
      }]
    }
   ]}
 }
 ChangeCategory {
  binding_count=7
  length_seconds=0.25
  object=&self:ChangeCategory
  sequence {
   bindings [
    Text_CategoryTitle {
     id=#4
     possessed_class=CommonTextBlock
     tracks [
      {
       type=MovieSceneColorTrack
       display_name=Color and Opacity
       sections [{
         type=MovieSceneColorSection
         channels [
          Color.R {type=MovieSceneScriptingFloatChannel}
          Color.G {type=MovieSceneScriptingFloatChannel}
          Color.B {type=MovieSceneScriptingFloatChannel}
          Color.A {type=MovieSceneScriptingFloatChannel, keys [[0,0],[30.002,1]]}
         ]
        }]
      }
      {
       type=MovieScene2DTransformTrack
       sections [{type=MovieScene2DTransformSection, channels [Translation.Y {type=MovieSceneScriptingFloatChannel, keys [[0,32],[30.002,0]]}]}]
      }
     ]
    }
    Text_CategoryDescription {
     id=#5
     possessed_class=CommonTextBlock
     tracks [{
       type=MovieSceneFloatTrack
       display_name=Render Opacity
       sections [{type=MovieSceneFloatSection, end_frame=30, start_frame=0, channels [None {type=MovieSceneScriptingFloatChannel, keys [[0,0],[30,1]]}]}]
      }]
    }
    Overlay_Unlocked {
     id=#6
     possessed_class=Overlay
     tracks [{
       type=MovieScene2DTransformTrack
       sections [{
         type=MovieScene2DTransformSection
         channels [
          Translation.X {type=MovieSceneScriptingFloatChannel}
          Translation.Y {type=MovieSceneScriptingFloatChannel}
          Angle {type=MovieSceneScriptingFloatChannel}
          Scale.X {type=MovieSceneScriptingFloatChannel, keys [[10,1.9],[30,1]]}
          Scale.Y {type=MovieSceneScriptingFloatChannel, keys [[10,1.9],[30,1]]}
          Shear.X {type=MovieSceneScriptingFloatChannel}
          Shear.Y {type=MovieSceneScriptingFloatChannel}
         ]
        }]
      }]
    }
    Border_Choice_Header {
     id=#7
     possessed_class=Border
     tracks [{
       type=MovieSceneWidgetMaterialTrack
       sections [{
         type=MovieSceneParameterSection
         end_frame=30
         start_frame=0
         channels [HoverAnimControl {type=MovieSceneScriptingFloatChannel, keys [[0,0],[30,1]]}]
        }]
      }]
    }
    Choice_CustomizationChoicePrevious {
     id=#8
     possessed_class=SpatialCustomizationChoice_C
     tracks [{
       type=MovieScene2DTransformTrack
       sections [{
         type=MovieScene2DTransformSection
         channels [
          Translation.X {type=MovieSceneScriptingFloatChannel, keys [[0,50],[30,0]]}
          Translation.Y {type=MovieSceneScriptingFloatChannel}
          Angle {type=MovieSceneScriptingFloatChannel}
          Scale.X {type=MovieSceneScriptingFloatChannel}
          Scale.Y {type=MovieSceneScriptingFloatChannel}
          Shear.X {type=MovieSceneScriptingFloatChannel}
          Shear.Y {type=MovieSceneScriptingFloatChannel}
         ]
        }]
      }]
    }
    Border_Body {
     id=#9
     display_name=VerticalBox_1
     possessed_class=Border
     tracks [
      {
       type=MovieSceneFloatTrack
       display_name=Render Opacity
       sections [{type=MovieSceneFloatSection, channels [None {type=MovieSceneScriptingFloatChannel, keys [[0,0],[10,0],[30,1]]}]}]
      }
      {
       type=MovieScene2DTransformTrack
       sections [{
         type=MovieScene2DTransformSection
         channels [
          Translation.X {type=MovieSceneScriptingFloatChannel}
          Translation.Y {type=MovieSceneScriptingFloatChannel, keys [[0,-16],[30.002,0]]}
          Angle {type=MovieSceneScriptingFloatChannel}
          Scale.X {type=MovieSceneScriptingFloatChannel}
          Scale.Y {type=MovieSceneScriptingFloatChannel}
          Shear.X {type=MovieSceneScriptingFloatChannel}
          Shear.Y {type=MovieSceneScriptingFloatChannel}
         ]
        }]
      }
     ]
    }
    Choice_CustomizationChoiceNext {
     id=#10
     possessed_class=SpatialCustomizationChoice_C
     tracks [{
       type=MovieScene2DTransformTrack
       sections [{
         type=MovieScene2DTransformSection
         channels [
          Translation.X {type=MovieSceneScriptingFloatChannel, keys [[0,-50],[30,0]]}
          Translation.Y {type=MovieSceneScriptingFloatChannel}
          Angle {type=MovieSceneScriptingFloatChannel}
          Scale.X {type=MovieSceneScriptingFloatChannel}
          Scale.Y {type=MovieSceneScriptingFloatChannel}
          Shear.X {type=MovieSceneScriptingFloatChannel}
          Shear.Y {type=MovieSceneScriptingFloatChannel}
         ]
        }]
      }]
    }
   ]
   display_rate {numerator=120}
  }
 }
 Fanfare {
  binding_count=5
  length_seconds=1.496
  object=&self:Fanfare
  sequence {
   bindings [
    Image_PulseRing {
     id=#11
     possessed_class=Image
     tracks [
      {
       type=MovieSceneColorTrack
       display_name=Color and Opacity
       sections [{
         type=MovieSceneColorSection
         channels [
          Color.R {type=MovieSceneScriptingFloatChannel}
          Color.G {type=MovieSceneScriptingFloatChannel}
          Color.B {type=MovieSceneScriptingFloatChannel}
          Color.A {type=MovieSceneScriptingFloatChannel, keys [[9.961,0],[11.961,1],[51.969,1],[56.969,0]]}
         ]
        }]
      }
      {
       type=MovieSceneWidgetMaterialTrack
       sections [{
         type=MovieSceneParameterSection
         end_frame=89
         start_frame=-1
         channels [
          HardRingDensity {type=MovieSceneScriptingFloatChannel, keys [[9.96,44.6938],[37.969,1.9212],[74.878,0]]}
          HardRingEdge {type=MovieSceneScriptingFloatChannel, keys [[9.96,0.1947],[46.969,0.0496]]}
          HardRingRadius {type=MovieSceneScriptingFloatChannel, keys [[9.996,0.0414],[46.969,0.5457]]}
          HardRingSize {type=MovieSceneScriptingFloatChannel, keys [[9.96,-0.0047],[29,0.0222],[46.969,0.1514]]}
          NoiseAmount {type=MovieSceneScriptingFloatChannel, keys [[9.96,0],[46.969,1.2778],[89.758,1.2778]]}
          NoiseRepeat {type=MovieSceneScriptingFloatChannel, keys [[22.306,0.9471],[89.758,1.4696]]}
         ]
        }]
      }
     ]
    }
    Image_GreenCheck {
     id=#12
     possessed_class=Image
     tracks [
      {
       type=MovieSceneColorTrack
       display_name=Color and Opacity
       sections [{
         type=MovieSceneColorSection
         channels [
          Color.R {type=MovieSceneScriptingFloatChannel}
          Color.G {type=MovieSceneScriptingFloatChannel}
          Color.B {type=MovieSceneScriptingFloatChannel}
          Color.A {type=MovieSceneScriptingFloatChannel, keys [[22.961,0],[25.961,0.5],[47.961,0.1]]}
         ]
        }]
      }
      {
       type=MovieScene2DTransformTrack
       sections [{
         type=MovieScene2DTransformSection
         channels [
          Scale.X {type=MovieSceneScriptingFloatChannel, keys [[17.961,0.5],[47.961,1]]}
          Scale.Y {type=MovieSceneScriptingFloatChannel, keys [[17.961,0.5],[47.961,1]]}
         ]
        }]
      }
     ]
    }
    Image_triangles {
     id=#13
     possessed_class=Image
     tracks [
      {
       type=MovieSceneColorTrack
       display_name=Color and Opacity
       sections [{
         type=MovieSceneColorSection
         channels [
          Color.R {type=MovieSceneScriptingFloatChannel}
          Color.G {type=MovieSceneScriptingFloatChannel}
          Color.B {type=MovieSceneScriptingFloatChannel}
          Color.A {type=MovieSceneScriptingFloatChannel, keys [[2,0],[5,1],[44.961,1],[89.758,0]]}
         ]
        }]
      }
      {
       type=MovieSceneWidgetMaterialTrack
       sections [{type=MovieSceneParameterSection, channels [FlareFade {type=MovieSceneScriptingFloatChannel, keys [[2,0],[14,2.0596],[50,0]]}]}]
      }
     ]
    }
    RichText_UnlockUsed {
     id=#14
     possessed_class=CommonRichTextBlock
     tracks [{
       type=MovieScene2DTransformTrack
       sections [{
         type=MovieScene2DTransformSection
         channels [
          Scale.X {type=MovieSceneScriptingFloatChannel, keys [[39.961,0],[54.961,1]]}
          Scale.Y {type=MovieSceneScriptingFloatChannel, keys [[39.961,0],[54.961,1]]}
         ]
        }]
      }]
    }
    RichText_Upgraded {
     id=#15
     possessed_class=CommonRichTextBlock
     tracks [{
       type=MovieScene2DTransformTrack
       sections [{
         type=MovieScene2DTransformSection
         channels [
          Scale.X {type=MovieSceneScriptingFloatChannel, keys [[34.961,0],[41.961,0.9],[49.961,1]]}
          Scale.Y {type=MovieSceneScriptingFloatChannel, keys [[34.961,0],[41.961,0.9],[49.961,1]]}
         ]
        }]
      }]
    }
   ]
   display_rate {numerator=60}
  }
 }
 switch_right {
  binding_count=3
  length_seconds=0.1084
  object=&self:switch_right
  sequence {
   bindings [
    Choice_SelectedChoice {
     id=#16
     possessed_class=SpatialCustomizationChoice_C
     tracks [
      {
       type=MovieScene2DTransformTrack
       sections [{
         type=MovieScene2DTransformSection
         channels [
          Translation.X {type=MovieSceneScriptingFloatChannel, keys [[0,0],[12,-160],[13,0]]}
          Translation.Y {type=MovieSceneScriptingFloatChannel}
          Angle {type=MovieSceneScriptingFloatChannel}
          Scale.X {type=MovieSceneScriptingFloatChannel, keys [[0,1],[12,0.75],[13,1]]}
          Scale.Y {type=MovieSceneScriptingFloatChannel, keys [[0,1],[12,0.75],[13,1]]}
          Shear.X {type=MovieSceneScriptingFloatChannel}
          Shear.Y {type=MovieSceneScriptingFloatChannel}
         ]
        }]
      }
      {
       type=MovieSceneColorTrack
       display_name=Color and Opacity
       sections [{
         type=MovieSceneColorSection
         channels [
          Color.R {type=MovieSceneScriptingFloatChannel}
          Color.G {type=MovieSceneScriptingFloatChannel}
          Color.B {type=MovieSceneScriptingFloatChannel}
          Color.A {type=MovieSceneScriptingFloatChannel, keys [[0,1],[12,0.66],[13,1]]}
         ]
        }]
      }
     ]
    }
    Choice_CustomizationChoiceNext {
     id=#17
     possessed_class=SpatialCustomizationChoice_C
     tracks [
      {
       type=MovieSceneColorTrack
       display_name=Color and Opacity
       sections [{
         type=MovieSceneColorSection
         channels [
          Color.R {type=MovieSceneScriptingFloatChannel, keys [[0,0.9],[12,1],[13,0.9]]}
          Color.G {type=MovieSceneScriptingFloatChannel, keys [[0,0.9],[12,1],[13,0.9]]}
          Color.B {type=MovieSceneScriptingFloatChannel, keys [[0,0.9],[12,1],[13,0.9]]}
          Color.A {type=MovieSceneScriptingFloatChannel, keys [[0,0.66],[12,1],[13,0.66]]}
         ]
        }]
      }
      {
       type=MovieScene2DTransformTrack
       sections [{
         type=MovieScene2DTransformSection
         channels [
          Translation.X {type=MovieSceneScriptingFloatChannel, keys [[0,0],[12,-145.6546],[13,0]]}
          Translation.Y {type=MovieSceneScriptingFloatChannel}
          Angle {type=MovieSceneScriptingFloatChannel}
          Scale.X {type=MovieSceneScriptingFloatChannel, keys [[0,1],[12,1.3],[13,1]]}
          Scale.Y {type=MovieSceneScriptingFloatChannel, keys [[0,1],[12,1.3],[13,1]]}
          Shear.X {type=MovieSceneScriptingFloatChannel}
          Shear.Y {type=MovieSceneScriptingFloatChannel}
         ]
        }]
      }
     ]
    }
    Choice_CustomizationChoicePrevious {
     id=#18
     possessed_class=SpatialCustomizationChoice_C
     tracks [{
       type=MovieScene2DTransformTrack
       sections [{
         type=MovieScene2DTransformSection
         end_frame=13
         channels [
          Translation.X {type=MovieSceneScriptingFloatChannel, keys [[0,0],[12,318],[13,0]]}
          Translation.Y {type=MovieSceneScriptingFloatChannel}
          Angle {type=MovieSceneScriptingFloatChannel}
          Scale.X {type=MovieSceneScriptingFloatChannel}
          Scale.Y {type=MovieSceneScriptingFloatChannel}
          Shear.X {type=MovieSceneScriptingFloatChannel}
          Shear.Y {type=MovieSceneScriptingFloatChannel}
         ]
        }]
      }]
    }
   ]
   display_rate {numerator=120}
  }
 }
 Activated {
  binding_count=4
  length_seconds=0.35
  object=&self:Activated
  sequence {bindings [
    SizeBox_Footer {
     id=#1
     possessed_class=SizeBox
     tracks [{
       type=MovieScene2DTransformTrack
       sections [{
         type=MovieScene2DTransformSection
         channels [
          Translation.X {type=MovieSceneScriptingFloatChannel, keys [[2.0003,2000],[7,0]]}
          Translation.Y {type=MovieSceneScriptingFloatChannel}
          Angle {type=MovieSceneScriptingFloatChannel}
          Scale.X {type=MovieSceneScriptingFloatChannel}
          Scale.Y {type=MovieSceneScriptingFloatChannel}
          Shear.X {type=MovieSceneScriptingFloatChannel}
          Shear.Y {type=MovieSceneScriptingFloatChannel}
         ]
        }]
      }]
    }
    BackingPanel {
     id=#19
     possessed_class=Image
     tracks [
      {
       type=MovieSceneFloatTrack
       display_name=Render Opacity
       sections [{type=MovieSceneFloatSection, channels [None {type=MovieSceneScriptingFloatChannel, keys [[0,0],[6,1]]}]}]
      }
      {
       type=MovieScene2DTransformTrack
       sections [{type=MovieScene2DTransformSection, channels [Translation.X {type=MovieSceneScriptingFloatChannel, keys [[0,512],[6,88]]}]}]
      }
      {
       type=MovieSceneWidgetMaterialTrack
       sections [{
         type=MovieSceneParameterSection
         end_frame=7
         start_frame=0
         channels [
          Bar_Width {type=MovieSceneScriptingFloatChannel, keys [[0,10],[6,1.6]]}
          Bar_Rotation {type=MovieSceneScriptingFloatChannel, keys [[0,0.5],[6,1.175]]}
         ]
        }]
      }
     ]
    }
    Border_Body {
     id=#2
     possessed_class=Border
     tracks [{
       type=MovieScene2DTransformTrack
       sections [{
         type=MovieScene2DTransformSection
         channels [
          Translation.X {type=MovieSceneScriptingFloatChannel, keys [[1,2000],[5.9997,0]]}
          Translation.Y {type=MovieSceneScriptingFloatChannel}
          Angle {type=MovieSceneScriptingFloatChannel}
          Scale.X {type=MovieSceneScriptingFloatChannel}
          Scale.Y {type=MovieSceneScriptingFloatChannel}
          Shear.X {type=MovieSceneScriptingFloatChannel}
          Shear.Y {type=MovieSceneScriptingFloatChannel}
         ]
        }]
      }]
    }
    SizeBox_Header {
     id=#3
     display_name="Image_1 (2)"
     possessed_class=Overlay
     tracks [{
       type=MovieScene2DTransformTrack
       sections [{
         type=MovieScene2DTransformSection
         channels [
          Translation.X {type=MovieSceneScriptingFloatChannel, keys [[0,2000],[4.9997,0]]}
          Translation.Y {type=MovieSceneScriptingFloatChannel}
          Angle {type=MovieSceneScriptingFloatChannel}
          Scale.X {type=MovieSceneScriptingFloatChannel}
          Scale.Y {type=MovieSceneScriptingFloatChannel}
          Shear.X {type=MovieSceneScriptingFloatChannel}
          Shear.Y {type=MovieSceneScriptingFloatChannel}
         ]
        }]
      }]
    }
   ]}
 }
]
graphs [
 EventGraph {nodes [
   "On Button Clicked (Button_NextChoice)"!() -> (OutputDelegate, then, Button:Common Button Legacy Object Reference) {is=title K2Node_ComponentBoundEvent}
   n1 = Get switch_right() -> (switch_right:Widget Animation Object Reference)
   > PlayAnimationForward(InAnimation=@n1)
   "On Button Clicked (Button_PreviousChoice)"!() -> (OutputDelegate, then, Button:Common Button Legacy Object Reference) {is=title K2Node_ComponentBoundEvent}
   > PlayAnimationReverse(InAnimation=@n1)
   OnIntroChoiceCompleted() {is=title K2Node_CustomEvent}
   > PlaySound2D(Sound=Get Popup Sound() -> (Popup Sound:Sound Base Object Reference))
   n2 = Event OnCategoryUpdated!() -> (
    OutputDelegate
    then
    CategoryState:ESpatialCustomizationCategoryState Enum=LockedByBattlePass
    Color:Linear Color Structure
    bUseColorPicker:bool
   ) {is=title K2Node_Event}
   > HandleCategoryUpdated!(
    execute
    self
    State:ESpatialCustomizationCategoryState Enum=@n2.CategoryState
    Color:Linear Color Structure=@n2.Color
    UseColorPicker:bool=@n2.bUseColorPicker
   ) -> (then)
   n3 = Event OnUpdateFinalizedText!() -> (OutputDelegate, then, InFinalizedChoiceText:"Text (by ref)") {is=title K2Node_Event}
   > SetText(
    self=Get RichText_UnlockUsed() -> (RichText_UnlockUsed:Common Rich Text Block Object Reference)
    InText=Format Text!(Format="LOCTEXT(\"With <text color=\\\"FFF\\\">{finalizedChoice}!</>\")", finalizedChoice:"Text (by ref)"=@n3.InFinalizedChoiceText) -> (Result)
   )
   Event On Activated()
   > PlayAnimationForward(InAnimation=Get Activated() -> (Activated:Widget Animation Object Reference))
   Event Construct()
   > BindToAnimationStarted!(
    execute
    self:User Widget Object Reference
    Animation:Widget Animation Object Reference=Get IntroChoice() -> (IntroChoice:Widget Animation Object Reference)
    Delegate:Delegate=Create Event()
   ) -> (then)
   n4 = Event OnFinalizeChoiceFinished!() -> (OutputDelegate, then, bSucceeded:bool) {is=title K2Node_Event}
   > Branch(Condition=@n4.bSucceeded)
   > PlaySound2D(Sound=Get Confirm Sound() -> (Confirm Sound:Sound Base Object Reference))
   > PlayAnimationForward(InAnimation=Get Fanfare() -> (Fanfare:Widget Animation Object Reference))
   > PushSoundMixModifier!(execute, InSoundMixModifier:Sound Mix Object Reference=/BRCosmetics/Sounds/Other/FortSoundClassesAndModes/UI_Upgrade_Suppress_Music_MixMod) -> (then)
   n5 = Event OnChoiceCountUpdated!() -> (OutputDelegate, then, ChoiceNum:int=0, TotalCount:int=0) {is=title K2Node_Event}
   > HandleChoiceCountUpdated!(execute, self, ChoiceNum:int=@n5.ChoiceNum, TotalCount:int=@n5.TotalCount) -> (then)
  ]}
 HandleCategoryUpdated {nodes [
   HandleCategoryUpdated!() -> (then, State:ESpatialCustomizationCategoryState Enum=LockedByBattlePass, Color:Linear Color Structure, UseColorPicker:bool=~) {is=title K2Node_FunctionEntry}
   > PlayAnimationForward(InAnimation=Get ChangeCategory() -> (ChangeCategory:Widget Animation Object Reference))
   > n1 = Sequence()
   n2 = Select(
    Option 0:Widget Object Reference=Get Overlay_Choices() -> (Overlay_Choices:Vertical Box Object Reference)
    Option 1:Widget Object Reference=Get ColorPicker() -> (ColorPicker:Fort Color Picker Object Reference)
    Index=Get UseColorPicker() -> (UseColorPicker:bool)
   ) -> (ReturnValue:Widget Object Reference)
   > SetActiveWidget(
    self:Widget Switcher Object Reference=Get Switcher_Selection() -> (Switcher_Selection:Common Activatable Widget Switcher Object Reference)
    Widget=Select!(
     LockedByBattlePass:Widget Object Reference=@n2
     LockedByChallenge:Widget Object Reference=@n2
     LockedByChallengeCompletion:Widget Object Reference=@n2
     UnlockAvailable:Widget Object Reference=@n2
     UnlockUsed:Widget Object Reference=Get Overlay_Unlocked() -> (Overlay_Unlocked:Overlay Object Reference)
     Index:ESpatialCustomizationCategoryState Enum=Get State() -> (State:ESpatialCustomizationCategoryState Enum=LockedByBattlePass)
    ) -> (ReturnValue)
   )
   > SetActiveWidget(
    self=Get Switcher_Finalize() -> (Switcher_Finalize:Widget Switcher Object Reference)
    Widget=Select(
     LockedByBattlePass=Get Button_PurchaseBattlepass() -> (Button_PurchaseBattlepass:Common Button Legacy Object Reference)
     LockedByChallenge=Get Border_Locked() -> (Border_Locked:Border Object Reference)
     LockedByChallengeCompletion=Get QuestProgress_ChallengeProgression() -> (QuestProgress_ChallengeProgression:Athena Quest Progression Object Reference)
     UnlockAvailable=Get Button_FinalizeChoice() -> (Button_FinalizeChoice:Common Button Legacy Object Reference)
     UnlockUsed=Get Border_UnlockUsed() -> (Border_UnlockUsed:Border Object Reference)
     Index=Get State() -> (State=LockedByBattlePass)
    )
   )
   n3 = GetDynamicMaterial(execute=@n1.then_1, self:Border Object Reference=Get Border_Choice_Header() -> (Border_Choice_Header:Border Object Reference))
   n4 = BreakColor(InColor=Get Color() -> (Color:Linear Color Structure))
   n5 = Make Linear Color!(R:float=@n4, G:float=@n4.G, B:float=@n4.B, A:float=1) -> (LinearColor:Linear Color Structure) {is=title K2Node_MakeStruct}
   > SetVectorParameterValue(self=@n3, ParameterName=Highlight, Value=@n5)
   > SetColorAndOpacity(
    self=Get Text_CategoryDescription() -> (Text_CategoryDescription:Common Text Object Reference)
    InColorAndOpacity=Make Slate Color(SpecifiedColor=@n5)
   )
   n6 = Get State() -> (State=LockedByBattlePass)
   n7 = OR Boolean(
    A="Equal (Enum)"(A:ESpatialCustomizationCategoryState Enum=@n6, B:ESpatialCustomizationCategoryState Enum=UnlockAvailable)
    B="Equal (Enum)"(A=@n6, B=UnlockUsed)
   )
   SetColorAndOpacity(
    self=Get Text_Preview() -> (Text_Preview:Common Text Object Reference)
    InColorAndOpacity=Make Slate Color(SpecifiedColor=Select!(Option 0:Linear Color Structure=(R=0.135,G=0.1987,B=0.9,A=1), Option 1:Linear Color Structure=(R=0,G=0.1651,B=1,A=1), Index:bool=@n7) -> (ReturnValue:Linear Color Structure))
   )
  ]}
 HandleChoiceCountUpdated {nodes [
   n1 = HandleChoiceCountUpdated!() -> (then, ChoiceNum:int, TotalCount:int) {is=title K2Node_FunctionEntry}
   > SetText(
    self=Get RichText_ChoiceCount() -> (RichText_ChoiceCount:Common Rich Text Block Object Reference)
    InText=Format Text!(Format="LOCTEXT(\"<text color=\\\"89E1FF\\\" size=\\\"40\\\">{current}</> / {total}\")", current:int=@n1, total:int=@n1.TotalCount) -> (Result)
   )
  ]}
]
variables [
 Confirm Sound {type=SoundBase, default=/BRCosmetics/Sounds/Other/Fort_UI_Sounds/S13_Frontend/UI_CustomizeSelect_Confirm_Cue}
 Popup Sound {type=SoundBase, default=/BRCosmetics/Sounds/Other/Fort_UI_Sounds/S13_Frontend/UI_CustomizeSelect_PopUp_Cue}
]
widgets {
 info {
  inheritedWidgetCount=16
  parentClass=/Script/FortniteUI.AthenaSpatialCustomizationSelection
  rootWidgetClass=&Overlay
  widgetCount=57
 }
 widgets [
  Overlay_1 {widgetClassPath=&Overlay, widget=Overlay_1}
  SafeZone_1 {
   widgetClassPath=&SafeZone
   bIsVariable=true
   parent=Overlay_1
   widget=SafeZone_1
   properties {padBottom=false, padLeft=false, padTop=false}
   slot {class=OverlaySlot, properties {horizontalAlignment=HAlign_Right, verticalAlignment=VAlign_Fill}}
  }
  BackingPanel {
   widgetClassPath=&Image
   bIsVariable=true
   parent=SafeZone_1
   widget=BackingPanel
   properties {
    renderOpacity=0.5
    visibility=SelfHitTestInvisible
    brush {
     bIsDynamicallyLoaded=false
     drawAs=Image
     imageType=NoImage
     mirroring=NoMirror
     resourceName=None
     resourceObject=/BRCosmetics/Athena/UI/Frontend/Spatial/Art/M_UI_SpatialCusotomization_BackingPanel_Safezone
     tiling=NoTile
     imageSize=<1024 1024>
     margin=<ltrb 0 0 0 0>
     outlineSettings {
      bUseBrushTransparency=false
      roundingType=HalfHeightRadius
      width=0
      color {colorUseRule=UseColor_Specified, specifiedColor=<rgba 0 0 0 0>}
      cornerRadii=<0 0 0 0>
     }
     tintColor {colorUseRule=UseColor_Specified, specifiedColor=<rgba 1 1 1 1>}
     uVRegion {bIsValid=false, max=<0 0>, min=<0 0>}
    }
    renderTransform {angle=0, scale=<1.25 1>, shear=<0 0>, translation=<88 0>}
   }
   slot {class=SafeZoneSlot, properties {hAlign=HAlign_Right}}
  }
  Switcher_CustomizationState {
   widgetClassPath=&WidgetSwitcher
   bIsVariable=true
   parent=Overlay_1
   widget=Switcher_CustomizationState
   slot {class=OverlaySlot, properties {horizontalAlignment=HAlign_Fill, verticalAlignment=VAlign_Fill}}
  }
  SafeZone_0 {
   widgetClassPath=&SafeZone
   bIsVariable=true
   parent=Switcher_CustomizationState
   widget=SafeZone_0
   properties {padLeft=false, padTop=false}
   slot {class=WidgetSwitcherSlot, properties {horizontalAlignment=HAlign_Right, padding=<ltrb 0 128 128 128>}}
  }
  SizeBox_Choice {
   widgetClassPath=&SizeBox
   bIsVariable=true
   parent=SafeZone_0
   widget=SizeBox_Choice
   properties {bOverride_WidthOverride=true, minDesiredWidth=500, widthOverride=640}
   slot {class=SafeZoneSlot}
  }
  VBox_Choice {widgetClassPath=&VerticalBox, bIsVariable=true, parent=SizeBox_Choice, widget=VBox_Choice, slot {class=SizeBoxSlot}}
  SizeBox_Header {
   widgetClassPath=&SizeBox
   parent=VBox_Choice
   widget=SizeBox_Header
   properties {bOverride_HeightOverride=true, heightOverride=128}
   slot {class=VerticalBoxSlot, properties {padding=<ltrb 0 0 0 16>}}
  }
  Overlay_3 {widgetClassPath=&Overlay, parent=SizeBox_Header, widget=Overlay_3, slot {class=SizeBoxSlot}}
  Border_Choice_Header {
   widgetClassPath=&Border
   bIsVariable=true
   parent=Overlay_3
   widget=Border_Choice_Header
   properties {
    verticalAlignment=VAlign_Center
    background {
     bIsDynamicallyLoaded=false
     drawAs=NoDrawType
     imageType=NoImage
     mirroring=NoMirror
     resourceName=None
     resourceObject=null
     tiling=NoTile
     imageSize=<128 128>
     margin=<ltrb 0.5 0.25 0.25 0.25>
     outlineSettings {
      bUseBrushTransparency=false
      roundingType=HalfHeightRadius
      width=0
      color {colorUseRule=UseColor_Specified, specifiedColor=<rgba 0 0 0 0>}
      cornerRadii=<0 0 0 0>
     }
     tintColor {colorUseRule=UseColor_Specified, specifiedColor=<rgba 1 1 1 1>}
     uVRegion {bIsValid=false, max=<0 0>, min=<0 0>}
    }
    padding=<ltrb 88 0 88 0>
   }
   slot {class=OverlaySlot, properties {horizontalAlignment=HAlign_Fill, verticalAlignment=VAlign_Fill}}
  }
  VerticalBox_0 {
   widgetClassPath=&VerticalBox
   parent=Border_Choice_Header
   widget=VerticalBox_0
   slot {class=BorderSlot, properties {verticalAlignment=VAlign_Center, padding=<ltrb 88 0 88 0>}}
  }
  ScaleBox_0 {
   widgetClassPath=&ScaleBox
   parent=VerticalBox_0
   widget=ScaleBox_0
   properties {stretch=ScaleToFitX, stretchDirection=DownOnly}
   slot {class=VerticalBoxSlot}
  }
  Text_CategoryTitle {
   widgetClassPath=&CommonTextBlock
   bInherited=true
   parent=ScaleBox_0
   widget=Text_CategoryTitle
   properties {
    justification=Center
    lineHeightPercentage=0.95
    scrollStyle=&TextScrollStyle-Base_C
    text=Title
    textTransformPolicy=ToUpper
    font {
     bForceMonospaced=false
     bMaterialIsStencil=false
     fontMaterial=null
     fontObject=&Burbank
     letterSpacing=0
     monospacedWidth=1
     size=64
     skewAmount=0
     typefaceFontName=BurbankRegular-Black
     outlineSettings {
      bApplyOutlineToDropShadows=false
      bMiteredCorners=false
      bSeparateFillAlpha=true
      outlineMaterial=null
      outlineSize=0
      outlineColor=<rgba 0 0 0 1>
     }
    }
    margin=<ltrb 16 0 16 0>
    shadowOffset=<0 0>
   }
   slot {class=ScaleBoxSlot}
  }
  Text_CategoryDescription {
   widgetClassPath=&CommonTextBlock
   bInherited=true
   bIsVariable=true
   parent=VerticalBox_0
   widget=Text_CategoryDescription
   properties {
    justification=Center
    lineHeightPercentage=0.9
    scrollStyle=&TextScrollStyle-Base_C
    text="CHOICE LOCKED!"
    textTransformPolicy=ToUpper
    font {
     bForceMonospaced=false
     bMaterialIsStencil=false
     fontMaterial=null
     fontObject=&BurbankSmall
     letterSpacing=0
     monospacedWidth=1
     size=20
     skewAmount=0
     typefaceFontName=Black
     outlineSettings {
      bApplyOutlineToDropShadows=false
      bMiteredCorners=false
      bSeparateFillAlpha=false
      outlineMaterial=null
      outlineSize=0
      outlineColor=<rgba 0 0 0 1>
     }
    }
    shadowOffset=<0 0>
   }
   slot {class=VerticalBoxSlot, properties {horizontalAlignment=HAlign_Center}}
  }
  SizeBox_LeftButton {
   widgetClassPath=&SizeBox
   parent=Overlay_3
   widget=SizeBox_LeftButton
   properties {bOverride_HeightOverride=true, bOverride_WidthOverride=true, heightOverride=64, widthOverride=64}
   slot {class=OverlaySlot, properties {verticalAlignment=VAlign_Center}}
  }
  Button_PreviousCategory {
   widgetClassPath=/FortniteCoreUI/Frontend/Showdown/CircleArrowButton
   bInherited=true
   bIsVariable=true
   parent=SizeBox_LeftButton
   widget=Button_PreviousCategory
   properties {
    clickMethod=DownAndUp
    edgeSoftness=0.075
    flip=true
    onButtonBaseClicked=(&self:WidgetTree.Button_PreviousCategory.HandleOnButtonClicked)
    onButtonBaseDoubleClicked=(&self:WidgetTree.Button_PreviousCategory.HandleOnButtonDoubleClicked)
    onButtonBaseHovered=(&self:WidgetTree.Button_PreviousCategory.HandleOnButtonHovered)
    onButtonBaseUnhovered=(&self:WidgetTree.Button_PreviousCategory.HandleOnButtonUnhovered)
    onSelectedChangedBase=(&self:WidgetTree.Button_PreviousCategory.HandleOnSelectedChanged)
    pressMethod=DownAndUp
    useShadow=false
    visibility=Visible
    arrowColor=<rgba 0 0.3333 1 1>
    hoveredSlateSoundOverride {resourceObject=/Game/Sounds/Fort_UI_Sounds/S13_Frontend/UI_Battlepass_Main_Hover_Cue}
    pressedSlateSoundOverride {resourceObject=/BRCosmetics/Sounds/Other/Fort_UI_Sounds/S13_Frontend/UI_Battlepass_Main_Click_Cue}
    triggeringInputAction {dataTable=&InputActionDataTable, rowName=GenericLeftTrigger}
   }
   slot {class=SizeBoxSlot}
  }
  SizeBox_RightButton {
   widgetClassPath=&SizeBox
   parent=Overlay_3
   widget=SizeBox_RightButton
   properties {bOverride_HeightOverride=true, bOverride_WidthOverride=true, heightOverride=64, widthOverride=64}
   slot {class=OverlaySlot, properties {horizontalAlignment=HAlign_Right, verticalAlignment=VAlign_Center}}
  }
  Button_NextCategory {
   widgetClassPath=&CircleArrowButton
   bInherited=true
   bIsVariable=true
   parent=SizeBox_RightButton
   widget=Button_NextCategory
   properties {
    clickMethod=DownAndUp
    edgeSoftness=0.075
    onButtonBaseClicked=(&self:WidgetTree.Button_NextCategory.HandleOnButtonClicked)
    onButtonBaseDoubleClicked=(&self:WidgetTree.Button_NextCategory.HandleOnButtonDoubleClicked)
    onButtonBaseHovered=(&self:WidgetTree.Button_NextCategory.HandleOnButtonHovered)
    onButtonBaseUnhovered=(&self:WidgetTree.Button_NextCategory.HandleOnButtonUnhovered)
    onSelectedChangedBase=(&self:WidgetTree.Button_NextCategory.HandleOnSelectedChanged)
    pressMethod=DownAndUp
    useShadow=false
    arrowColor=<rgba 0 0.3333 1 1>
    hoveredSlateSoundOverride {resourceObject=&UI_Battlepass_Main_Hover_Cue}
    pressedSlateSoundOverride {resourceObject=&UI_Battlepass_Main_Click_Cue}
    triggeringInputAction {dataTable=&InputActionDataTable, rowName=GenericRightTrigger}
   }
   slot {class=SizeBoxSlot}
  }
  Border_Body {
   widgetClassPath=&Border
   bIsVariable=true
   parent=VBox_Choice
   widget=Border_Body
   properties {
    visibility=SelfHitTestInvisible
    background {
     bIsDynamicallyLoaded=false
     drawAs=NoDrawType
     imageType=NoImage
     mirroring=NoMirror
     resourceName=None
     resourceObject=null
     tiling=NoTile
     imageSize=<128 128>
     margin=<ltrb 0.5 0.25 0.25 0.25>
     outlineSettings {
      bUseBrushTransparency=false
      roundingType=HalfHeightRadius
      width=0
      color {colorUseRule=UseColor_Specified, specifiedColor=<rgba 0 0 0 0>}
      cornerRadii=<0 0 0 0>
     }
     tintColor {colorUseRule=UseColor_Specified, specifiedColor=<rgba 1 1 1 1>}
     uVRegion {bIsValid=false, max=<0 0>, min=<0 0>}
    }
    padding=<ltrb 0 0 0 0>
   }
   slot {class=VerticalBoxSlot, properties {size {sizeRule=Fill, value=1}}}
  }
  VerticalBox_1 {
   widgetClassPath=&VerticalBox
   parent=Border_Body
   widget=VerticalBox_1
   slot {class=BorderSlot, properties {padding=<ltrb 0 0 0 0>}}
  }
  Switcher_Selection {
   widgetClassPath=&CommonActivatableWidgetSwitcher
   bIsVariable=true
   parent=VerticalBox_1
   widget=Switcher_Selection
   properties {transitionDuration=0}
   slot {class=VerticalBoxSlot, properties {padding=<ltrb 32 0 32 0>, size {sizeRule=Fill, value=1}}}
  }
  Overlay_Choices {
   widgetClassPath=&VerticalBox
   bIsVariable=true
   parent=Switcher_Selection
   widget=Overlay_Choices
   slot {class=WidgetSwitcherSlot, properties {verticalAlignment=VAlign_Center}}
  }
  VerticalBox_3 {
   widgetClassPath=&VerticalBox
   parent=Overlay_Choices
   widget=VerticalBox_3
   slot {class=VerticalBoxSlot, properties {horizontalAlignment=HAlign_Center}}
  }
  Text_Preview {
   widgetClassPath=&CommonTextBlock
   bIsVariable=true
   parent=VerticalBox_3
   widget=Text_Preview
   properties {
    justification=Center
    lineHeightPercentage=0.9
    text=Preview
    textTransformPolicy=ToUpper
    colorAndOpacity {colorUseRule=UseColor_Specified, specifiedColor=<rgba 0 0.1651 1 1>}
    font {
     bForceMonospaced=false
     bMaterialIsStencil=false
     fontMaterial=/BRCosmetics/Athena/UI/Season/Art/M_UI_PreviewUmbrella_Fade_Text
     fontObject=&Burbank
     letterSpacing=0
     monospacedWidth=1
     size=40
     skewAmount=0
     typefaceFontName=BurbankRegular-Black
     outlineSettings {
      bApplyOutlineToDropShadows=false
      bMiteredCorners=false
      bSeparateFillAlpha=false
      outlineMaterial=null
      outlineSize=0
      outlineColor=<rgba 0 0 0 1>
     }
    }
    renderTransform {angle=0, scale=<1 1>, shear=<-12 0>, translation=<0 4>}
    shadowOffset=<0 0>
   }
   slot {class=VerticalBoxSlot, properties {horizontalAlignment=HAlign_Center, verticalAlignment=VAlign_Top}}
  }
  Text_SelectedChoiceName {
   widgetClassPath=&CommonTextBlock
   bInherited=true
   bIsVariable=true
   parent=VerticalBox_3
   widget=Text_SelectedChoiceName
   properties {
    justification=Center
    scrollStyle=&TextScrollStyle-Base_C
    text="{Upgrade Name}"
    textTransformPolicy=ToUpper
    colorAndOpacity {colorUseRule=UseColor_Specified, specifiedColor=<rgba 0.5 0.8333 1 1>}
    font {
     bForceMonospaced=false
     bMaterialIsStencil=false
     fontMaterial=null
     fontObject=&BurbankSmall
     letterSpacing=0
     monospacedWidth=1
     size=24
     skewAmount=0
     typefaceFontName=Black
     outlineSettings {
      bApplyOutlineToDropShadows=false
      bMiteredCorners=false
      bSeparateFillAlpha=false
      outlineMaterial=null
      outlineSize=0
      outlineColor=<rgba 0 0 0 1>
     }
    }
    shadowOffset=<0 0>
   }
   slot {class=VerticalBoxSlot, properties {horizontalAlignment=HAlign_Center, padding=<ltrb 0 -20 0 0>}}
  }
  Overlay_0 {widgetClassPath=&Overlay, parent=Overlay_Choices, widget=Overlay_0, slot {class=VerticalBoxSlot}}
  Choice_CustomizationChoicePrevious {
   widgetClassPath=&/SpatialCustomizationChoice
   bInherited=true
   bIsVariable=true
   parent=Overlay_0
   widget=Choice_CustomizationChoicePrevious
   properties {overrideSize=192, showStatusIcon=false, colorAndOpacity=<rgba 0.9 0.9 0.9 0.66>, size=<128 128>}
   slot {class=OverlaySlot, properties {verticalAlignment=VAlign_Center}}
  }
  Choice_CustomizationChoiceNext {
   widgetClassPath=&SpatialCustomizationChoice
   bInherited=true
   bIsVariable=true
   parent=Overlay_0
   widget=Choice_CustomizationChoiceNext
   properties {overrideSize=192, showStatusIcon=false, colorAndOpacity=<rgba 0.9 0.9 0.9 0.66>, size=<128 128>}
   slot {class=OverlaySlot, properties {horizontalAlignment=HAlign_Right, verticalAlignment=VAlign_Center}}
  }
  Choice_SelectedChoice {
   widgetClassPath=&SpatialCustomizationChoice
   bInherited=true
   bIsVariable=true
   parent=Overlay_0
   widget=Choice_SelectedChoice
   properties {fullOpacity=true, overrideSize=256, showStatusIcon=false, size=<256 256>}
   slot {class=OverlaySlot, properties {horizontalAlignment=HAlign_Center, padding=<ltrb 128 0 128 0>}}
  }
  HorizontalBox_135 {
   widgetClassPath=&HorizontalBox
   parent=Overlay_Choices
   widget=HorizontalBox_135
   slot {class=VerticalBoxSlot, properties {horizontalAlignment=HAlign_Center, verticalAlignment=VAlign_Top, padding=<ltrb 0 16 0 0>}}
  }
  Button_PreviousChoice {
   widgetClassPath=&/SpatialCustomizationChoiceNavigationButton
   bInherited=true
   bIsVariable=true
   parent=HorizontalBox_135
   widget=Button_PreviousChoice
   properties {
    onButtonBaseClicked=(&self:WidgetTree.Button_PreviousChoice.HandleOnButtonClicked)
    onButtonBaseDoubleClicked=(&self:WidgetTree.Button_PreviousChoice.HandleOnButtonDoubleClicked)
    onButtonBaseHovered=(&self:WidgetTree.Button_PreviousChoice.HandleOnButtonHovered)
    onButtonBaseUnhovered=(&self:WidgetTree.Button_PreviousChoice.HandleOnButtonUnhovered)
    onSelectedChangedBase=(&self:WidgetTree.Button_PreviousChoice.HandleOnSelectedChanged)
    hoveredSlateSoundOverride {resourceObject=&UI_Battlepass_Main_Hover_Cue}
    pressedSlateSoundOverride {resourceObject=&UI_Battlepass_Main_Click_Cue}
    triggeringInputAction {dataTable=&InputActionDataTable, rowName=GenericLeftShoulder}
   }
   slot {class=HorizontalBoxSlot}
  }
  SizeBox_0 {
   widgetClassPath=&SizeBox
   parent=HorizontalBox_135
   widget=SizeBox_0
   properties {bOverride_MinDesiredWidth=true, minDesiredWidth=100, renderTransform {angle=0, scale=<1 1>, shear=<-12 0>, translation=<0 0>}}
   slot {class=HorizontalBoxSlot, properties {verticalAlignment=VAlign_Center, padding=<ltrb 4 0 4 0>}}
  }
  RichText_ChoiceCount {
   widgetClassPath=&CommonRichTextBlock
   bIsVariable=true
   parent=SizeBox_0
   widget=RichText_ChoiceCount
   properties {
    bOverrideDefaultStyle=true
    justification=Center
    text="<text color=\"89E1FF\" size=\"24\">{current}</> / {total}"
    defaultTextStyleOverride {
     overflowPolicy=Clip
     transformPolicy=None
     colorAndOpacity {colorUseRule=UseColor_Specified, specifiedColor=<rgba 0 0.125 0.75 1>}
     font {
      bForceMonospaced=false
      bMaterialIsStencil=false
      fontMaterial=null
      fontObject=&BurbankSmall
      letterSpacing=0
      monospacedWidth=1
      size=22
      skewAmount=0
      typefaceFontName=Black
      outlineSettings {
       bApplyOutlineToDropShadows=false
       bMiteredCorners=false
       bSeparateFillAlpha=false
       outlineMaterial=null
       outlineSize=0
       outlineColor=<rgba 0 0 0 1>
      }
     }
     highlightColor {colorUseRule=UseColor_Specified, specifiedColor=<rgba 0 0 0 1>}
     highlightShape {
      bIsDynamicallyLoaded=false
      drawAs=Image
      imageType=NoImage
      mirroring=NoMirror
      resourceName=None
      resourceObject=null
      tiling=NoTile
      imageSize=<32 32>
      margin=<ltrb 0 0 0 0>
      outlineSettings {
       bUseBrushTransparency=false
       roundingType=HalfHeightRadius
       width=0
       color {colorUseRule=UseColor_Specified, specifiedColor=<rgba 0 0 0 0>}
       cornerRadii=<0 0 0 0>
      }
      tintColor {colorUseRule=UseColor_Specified, specifiedColor=<rgba 1 1 1 1>}
      uVRegion {bIsValid=false, max=<0 0>, min=<0 0>}
     }
     selectedBackgroundColor {colorUseRule=UseColor_Foreground, specifiedColor=<rgba 1 0 1 1>}
     shadowColorAndOpacity=<rgba 0 0 0 1>
     shadowOffset=<0 0>
     strikeBrush {
      bIsDynamicallyLoaded=false
      drawAs=Image
      imageType=NoImage
      mirroring=NoMirror
      resourceName=None
      resourceObject=null
      tiling=NoTile
      imageSize=<32 32>
      margin=<ltrb 0 0 0 0>
      outlineSettings {
       bUseBrushTransparency=false
       roundingType=HalfHeightRadius
       width=0
       color {colorUseRule=UseColor_Specified, specifiedColor=<rgba 0 0 0 0>}
       cornerRadii=<0 0 0 0>
      }
      tintColor {colorUseRule=UseColor_Specified, specifiedColor=<rgba 1 1 1 1>}
      uVRegion {bIsValid=false, max=<0 0>, min=<0 0>}
     }
     underlineBrush {
      bIsDynamicallyLoaded=false
      drawAs=Image
      imageType=NoImage
      mirroring=NoMirror
      resourceName=None
      resourceObject=null
      tiling=NoTile
      imageSize=<32 32>
      margin=<ltrb 0 0 0 0>
      outlineSettings {
       bUseBrushTransparency=false
       roundingType=HalfHeightRadius
       width=0
       color {colorUseRule=UseColor_Specified, specifiedColor=<rgba 0 0 0 0>}
       cornerRadii=<0 0 0 0>
      }
      tintColor {colorUseRule=UseColor_Specified, specifiedColor=<rgba 1 1 1 1>}
      uVRegion {bIsValid=false, max=<0 0>, min=<0 0>}
     }
    }
   }
   slot {class=SizeBoxSlot, properties {horizontalAlignment=HAlign_Center}}
  }
  Button_NextChoice {
   widgetClassPath=&SpatialCustomizationChoiceNavigationButton
   bInherited=true
   bIsVariable=true
   parent=HorizontalBox_135
   widget=Button_NextChoice
   properties {
    bRight=true
    onButtonBaseClicked=(&self:WidgetTree.Button_NextChoice.HandleOnButtonClicked)
    onButtonBaseDoubleClicked=(&self:WidgetTree.Button_NextChoice.HandleOnButtonDoubleClicked)
    onButtonBaseHovered=(&self:WidgetTree.Button_NextChoice.HandleOnButtonHovered)
    onButtonBaseUnhovered=(&self:WidgetTree.Button_NextChoice.HandleOnButtonUnhovered)
    onSelectedChangedBase=(&self:WidgetTree.Button_NextChoice.HandleOnSelectedChanged)
    hoveredSlateSoundOverride {resourceObject=&UI_Battlepass_Main_Hover_Cue}
    pressedSlateSoundOverride {resourceObject=&UI_Battlepass_Main_Click_Cue}
    triggeringInputAction {dataTable=&InputActionDataTable, rowName=GenericRightShoulder}
   }
   slot {class=HorizontalBoxSlot}
  }
  Image_checkbox {
   widgetClassPath=&Image
   bIsVariable=true
   parent=Switcher_Selection
   widget=Image_checkbox
   properties {brush {
     bIsDynamicallyLoaded=false
     drawAs=Image
     imageType=NoImage
     mirroring=NoMirror
     resourceName=None
     resourceObject=/FortUILibrary/Foundation/Art_ToolKit/AngledBoxes/M_UI_AngledCheckBox_Inst
     tiling=NoTile
     imageSize=<320 320>
     margin=<ltrb 0 0 0 0>
     outlineSettings {
      bUseBrushTransparency=false
      roundingType=HalfHeightRadius
      width=0
      color {colorUseRule=UseColor_Specified, specifiedColor=<rgba 0 0 0 0>}
      cornerRadii=<0 0 0 0>
     }
     tintColor {colorUseRule=UseColor_Specified, specifiedColor=<rgba 1 1 1 0.248>}
     uVRegion {bIsValid=false, max=<0 0>, min=<0 0>}
    }}
   slot {class=WidgetSwitcherSlot, properties {horizontalAlignment=HAlign_Center, verticalAlignment=VAlign_Center, padding=<ltrb 0 0 32 0>}}
  }
  ColorPicker {
   widgetClassPath=&/ColorPicker
   bInherited=true
   bIsVariable=true
   parent=Switcher_Selection
   widget=ColorPicker
   properties {renderTransform {angle=0, scale=<1 1>, shear=<-4 0>, translation=<-24 0>}}
   slot {class=WidgetSwitcherSlot, properties {padding=<ltrb 0 16 0 0>}}
  }
  Overlay_Unlocked {
   widgetClassPath=&Overlay
   bIsVariable=true
   parent=Switcher_Selection
   widget=Overlay_Unlocked
   slot {class=WidgetSwitcherSlot, properties {horizontalAlignment=HAlign_Center, verticalAlignment=VAlign_Center}}
  }
  Image_PulseRing {
   widgetClassPath=&Image
   bIsVariable=true
   parent=Overlay_Unlocked
   widget=Image_PulseRing
   properties {
    brush {
     bIsDynamicallyLoaded=false
     drawAs=Image
     imageType=NoImage
     mirroring=NoMirror
     resourceName=None
     resourceObject=/BRCosmetics/Athena/Items/Accolades/Materials/M_UI_PulseRing_Anime_Green
     tiling=NoTile
     imageSize=<32 32>
     margin=<ltrb 0 0 0 0>
     outlineSettings {
      bUseBrushTransparency=false
      roundingType=HalfHeightRadius
      width=0
      color {colorUseRule=UseColor_Specified, specifiedColor=<rgba 0 0 0 0>}
      cornerRadii=<0 0 0 0>
     }
     tintColor {colorUseRule=UseColor_Specified, specifiedColor=<rgba 1 1 1 1>}
     uVRegion {bIsValid=false, max=<0 0>, min=<0 0>}
    }
    colorAndOpacity=<rgba 1 1 1 0>
    renderTransform {angle=0, scale=<2.5 2.5>, shear=<0 0>, translation=<0 0>}
   }
   slot {class=OverlaySlot, properties {horizontalAlignment=HAlign_Fill, verticalAlignment=VAlign_Fill}}
  }
  Choice_UnlockUsed {
   widgetClassPath=&SpatialCustomizationChoice
   bInherited=true
   bIsVariable=true
   parent=Overlay_Unlocked
   widget=Choice_UnlockUsed
   properties {overrideSize=256, showStatusIcon=false, size=<256 256>}
   slot {class=OverlaySlot}
  }
  Image_triangles {
   widgetClassPath=&Image
   bIsVariable=true
   parent=Overlay_Unlocked
   widget=Image_triangles
   properties {
    brush {
     bIsDynamicallyLoaded=false
     drawAs=Image
     imageType=NoImage
     mirroring=NoMirror
     resourceName=None
     resourceObject=/BRCosmetics/UI/Foundation/Onboarding/Materials/M_UI_UpgradeSquareParticle_Inst_Inst1
     tiling=NoTile
     imageSize=<232.0923 230>
     margin=<ltrb 0 0 0 0>
     outlineSettings {
      bUseBrushTransparency=false
      roundingType=HalfHeightRadius
      width=0
      color {colorUseRule=UseColor_Specified, specifiedColor=<rgba 0 0 0 0>}
      cornerRadii=<0 0 0 0>
     }
     tintColor {colorUseRule=UseColor_Specified, specifiedColor=<rgba 1 1 1 1>}
     uVRegion {bIsValid=false, max=<0 0>, min=<0 0>}
    }
    colorAndOpacity=<rgba 1 1 1 0>
    renderTransform {angle=-90, scale=<1 1>, shear=<0 0>, translation=<0 0>}
   }
   slot {class=OverlaySlot, properties {horizontalAlignment=HAlign_Center, verticalAlignment=VAlign_Center}}
  }
  Image_318 {
   widgetClassPath=&Image
   bIsVariable=true
   parent=Overlay_Unlocked
   widget=Image_318
   properties {brush {
     bIsDynamicallyLoaded=false
     drawAs=Image
     imageType=NoImage
     mirroring=NoMirror
     resourceName=None
     resourceObject=/FortUILibrary/Foundation/Art_ToolKit/AngledBoxes/M_UI_AngledCheckBox
     tiling=NoTile
     imageSize=<50 50>
     margin=<ltrb 0 0 0 0>
     outlineSettings {
      bUseBrushTransparency=false
      roundingType=HalfHeightRadius
      width=0
      color {colorUseRule=UseColor_Specified, specifiedColor=<rgba 0 0 0 0>}
      cornerRadii=<0 0 0 0>
     }
     tintColor {colorUseRule=UseColor_Specified, specifiedColor=<rgba 1 1 1 1>}
     uVRegion {bIsValid=false, max=<0 0>, min=<0 0>}
    }}
   slot {class=OverlaySlot, properties {horizontalAlignment=HAlign_Right, padding=<ltrb 18 18 18 18>}}
  }
  SizeBox_Footer {
   widgetClassPath=&SizeBox
   parent=VBox_Choice
   widget=SizeBox_Footer
   properties {bOverride_MinDesiredHeight=true, heightOverride=128, minDesiredHeight=168}
   slot {class=VerticalBoxSlot, properties {padding=<ltrb 0 24 0 0>}}
  }
  Switcher_Finalize {
   widgetClassPath=&WidgetSwitcher
   bIsVariable=true
   parent=SizeBox_Footer
   widget=Switcher_Finalize
   properties {activeWidgetIndex=2, renderTransform {angle=0, scale=<1 1>, shear=<0 0>, translation=<-16 0>}}
   slot {class=SizeBoxSlot, properties {verticalAlignment=VAlign_Top}}
  }
  Button_PurchaseBattlepass {
   widgetClassPath=/FortniteCoreUI/Frontend/Spatial/Customization/Spatial_Specialbutton
   bInherited=true
   bIsVariable=true
   parent=Switcher_Finalize
   widget=Button_PurchaseBattlepass
   properties {
    button text="Purchase\r\n<text color=\"ffffffff\">Battle Pass!</>"
    justifiction=Center
    onButtonBaseClicked=(&self:WidgetTree.Button_PurchaseBattlepass.HandleOnButtonClicked)
    onButtonBaseDoubleClicked=(&self:WidgetTree.Button_PurchaseBattlepass.HandleOnButtonDoubleClicked)
    onButtonBaseHovered=(&self:WidgetTree.Button_PurchaseBattlepass.HandleOnButtonHovered)
    onButtonBaseUnhovered=(&self:WidgetTree.Button_PurchaseBattlepass.HandleOnButtonUnhovered)
    onSelectedChangedBase=(&self:WidgetTree.Button_PurchaseBattlepass.HandleOnSelectedChanged)
    text Color {colorUseRule=UseColor_Specified, specifiedColor=<rgba 0 0.3333 1 1>}
    triggeringInputAction {dataTable=&InputActionDataTable, rowName=GenericAccept}
   }
   slot {class=WidgetSwitcherSlot, properties {verticalAlignment=VAlign_Top}}
  }
  Border_Locked {
   widgetClassPath=&Border
   bIsVariable=true
   parent=Switcher_Finalize
   widget=Border_Locked
   properties {
    verticalAlignment=VAlign_Center
    background {
     bIsDynamicallyLoaded=false
     drawAs=NoDrawType
     imageType=NoImage
     mirroring=NoMirror
     resourceName=None
     resourceObject=null
     tiling=NoTile
     imageSize=<128 128>
     margin=<ltrb 0.5 0.25 0.25 0.25>
     outlineSettings {
      bUseBrushTransparency=false
      roundingType=HalfHeightRadius
      width=0
      color {colorUseRule=UseColor_Specified, specifiedColor=<rgba 0 0 0 0>}
      cornerRadii=<0 0 0 0>
     }
     tintColor {colorUseRule=UseColor_Specified, specifiedColor=<rgba 1 1 1 1>}
     uVRegion {bIsValid=false, max=<0 0>, min=<0 0>}
    }
    padding=<ltrb 64 8 42 8>
   }
   slot {class=WidgetSwitcherSlot}
  }
  Overlay_7 {
   widgetClassPath=&Overlay
   parent=Border_Locked
   widget=Overlay_7
   slot {class=BorderSlot, properties {verticalAlignment=VAlign_Center, padding=<ltrb 64 8 42 8>}}
  }
  SizeBox_2 {
   widgetClassPath=&SizeBox
   parent=Overlay_7
   widget=SizeBox_2
   properties {bOverride_HeightOverride=true, bOverride_WidthOverride=true, heightOverride=128, widthOverride=128}
   slot {class=OverlaySlot, properties {horizontalAlignment=HAlign_Center, verticalAlignment=VAlign_Center}}
  }
  Overlay_111 {
   widgetClassPath=&Overlay
   parent=SizeBox_2
   widget=Overlay_111
   properties {renderTransform {angle=0, scale=<1.5 1.5>, shear=<-4 0>, translation=<0 0>}}
   slot {class=SizeBoxSlot}
  }
  RichText_LockedReason {
   widgetClassPath=&CommonRichTextBlock
   bInherited=true
   bIsVariable=true
   parent=Overlay_7
   widget=RichText_LockedReason
   properties {
    autoWrapText=true
    bOverrideDefaultStyle=true
    justification=Center
    lineHeightPercentage=0.9
    text="AVAILABLE WITH <text color=\"FFFFFFFF\">WEEK 6</> CHALLENGES FOR <text color=\"FFFFFFFF\">BATTLE PASS</> OWNERS!"
    textTransformPolicy=ToUpper
    defaultTextStyleOverride {
     overflowPolicy=Clip
     transformPolicy=None
     colorAndOpacity {colorUseRule=UseColor_Specified, specifiedColor=<rgba 0.0423 0.1746 0.855 1>}
     font {
      bForceMonospaced=false
      bMaterialIsStencil=false
      fontMaterial=null
      fontObject=&BurbankSmall
      letterSpacing=0
      monospacedWidth=1
      size=28
      skewAmount=0
      typefaceFontName=Black
      outlineSettings {
       bApplyOutlineToDropShadows=false
       bMiteredCorners=false
       bSeparateFillAlpha=false
       outlineMaterial=null
       outlineSize=0
       outlineColor=<rgba 0 0 0 1>
      }
     }
     highlightColor {colorUseRule=UseColor_Specified, specifiedColor=<rgba 0 0 0 1>}
     highlightShape {
      bIsDynamicallyLoaded=false
      drawAs=Image
      imageType=NoImage
      mirroring=NoMirror
      resourceName=None
      resourceObject=null
      tiling=NoTile
      imageSize=<32 32>
      margin=<ltrb 0 0 0 0>
      outlineSettings {
       bUseBrushTransparency=false
       roundingType=HalfHeightRadius
       width=0
       color {colorUseRule=UseColor_Specified, specifiedColor=<rgba 0 0 0 0>}
       cornerRadii=<0 0 0 0>
      }
      tintColor {colorUseRule=UseColor_Specified, specifiedColor=<rgba 1 1 1 1>}
      uVRegion {bIsValid=false, max=<0 0>, min=<0 0>}
     }
     selectedBackgroundColor {colorUseRule=UseColor_Foreground, specifiedColor=<rgba 1 0 1 1>}
     shadowColorAndOpacity=<rgba 0 0 0 1>
     shadowOffset=<0 0>
     strikeBrush {
      bIsDynamicallyLoaded=false
      drawAs=Image
      imageType=NoImage
      mirroring=NoMirror
      resourceName=None
      resourceObject=null
      tiling=NoTile
      imageSize=<32 32>
      margin=<ltrb 0 0 0 0>
      outlineSettings {
       bUseBrushTransparency=false
       roundingType=HalfHeightRadius
       width=0
       color {colorUseRule=UseColor_Specified, specifiedColor=<rgba 0 0 0 0>}
       cornerRadii=<0 0 0 0>
      }
      tintColor {colorUseRule=UseColor_Specified, specifiedColor=<rgba 1 1 1 1>}
      uVRegion {bIsValid=false, max=<0 0>, min=<0 0>}
     }
     underlineBrush {
      bIsDynamicallyLoaded=false
      drawAs=Image
      imageType=NoImage
      mirroring=NoMirror
      resourceName=None
      resourceObject=null
      tiling=NoTile
      imageSize=<32 32>
      margin=<ltrb 0 0 0 0>
      outlineSettings {
       bUseBrushTransparency=false
       roundingType=HalfHeightRadius
       width=0
       color {colorUseRule=UseColor_Specified, specifiedColor=<rgba 0 0 0 0>}
       cornerRadii=<0 0 0 0>
      }
      tintColor {colorUseRule=UseColor_Specified, specifiedColor=<rgba 1 1 1 1>}
      uVRegion {bIsValid=false, max=<0 0>, min=<0 0>}
     }
    }
    margin=<ltrb 0 0 0 4>
   }
   slot {class=OverlaySlot, properties {horizontalAlignment=HAlign_Fill, verticalAlignment=VAlign_Center}}
  }
  QuestProgress_ChallengeProgression {
   widgetClassPath=&/SpatialQuestProgress
   bInherited=true
   bIsVariable=true
   parent=Switcher_Finalize
   widget=QuestProgress_ChallengeProgression
   slot {class=WidgetSwitcherSlot}
  }
  Button_FinalizeChoice {
   widgetClassPath=&Spatial_Specialbutton
   bInherited=true
   bIsVariable=true
   parent=Switcher_Finalize
   widget=Button_FinalizeChoice
   properties {
    button text="finalize\r\n<text color=\"ffffffff\" size=\"38\">Upgrade!</>"
    justifiction=Center
    onButtonBaseClicked=(&self:WidgetTree.Button_FinalizeChoice.HandleOnButtonClicked)
    onButtonBaseDoubleClicked=(&self:WidgetTree.Button_FinalizeChoice.HandleOnButtonDoubleClicked)
    onButtonBaseHovered=(&self:WidgetTree.Button_FinalizeChoice.HandleOnButtonHovered)
    onButtonBaseUnhovered=(&self:WidgetTree.Button_FinalizeChoice.HandleOnButtonUnhovered)
    onSelectedChangedBase=(&self:WidgetTree.Button_FinalizeChoice.HandleOnSelectedChanged)
    right Image=null
    pressedSlateSoundOverride {resourceObject=/BRCosmetics/Sounds/Other/Fort_UI_Sounds/S13_Frontend/UI_CustomizeSelect_PreConfirm_Cue}
    text Color {colorUseRule=UseColor_Specified, specifiedColor=<rgba 0 0.3333 1 1>}
    triggeringInputAction {dataTable=&InputActionDataTable, rowName=GenericAccept}
   }
   slot {class=WidgetSwitcherSlot, properties {padding=<ltrb 20 0 10 0>}}
  }
  Border_UnlockUsed {
   widgetClassPath=&Border
   bIsVariable=true
   parent=Switcher_Finalize
   widget=Border_UnlockUsed
   properties {
    background {
     bIsDynamicallyLoaded=false
     drawAs=NoDrawType
     imageType=NoImage
     mirroring=NoMirror
     resourceName=None
     resourceObject=null
     tiling=NoTile
     imageSize=<128 128>
     margin=<ltrb 0.5 0.25 0.25 0.25>
     outlineSettings {
      bUseBrushTransparency=false
      roundingType=HalfHeightRadius
      width=0
      color {colorUseRule=UseColor_Specified, specifiedColor=<rgba 0 0 0 0>}
      cornerRadii=<0 0 0 0>
     }
     tintColor {colorUseRule=UseColor_Specified, specifiedColor=<rgba 0 0.147 0.063 0.5>}
     uVRegion {bIsValid=false, max=<0 0>, min=<0 0>}
    }
    padding=<ltrb 39 6 22 6>
   }
   slot {class=WidgetSwitcherSlot}
  }
  Overlay_2 {
   widgetClassPath=&Overlay
   parent=Border_UnlockUsed
   widget=Overlay_2
   slot {class=BorderSlot, properties {padding=<ltrb 39 6 22 6>}}
  }
  Image_GreenCheck {
   widgetClassPath=&Image
   bIsVariable=true
   parent=Overlay_2
   widget=Image_GreenCheck
   properties {
    brush {
     bIsDynamicallyLoaded=false
     drawAs=Image
     imageType=NoImage
     mirroring=NoMirror
     resourceName=None
     resourceObject=/FortUILibrary/Foundation/Art_ToolKit/AngledBoxes/M_UI_AngledCheckBox_Inst1
     tiling=NoTile
     imageSize=<150 150>
     margin=<ltrb 0 0 0 0>
     outlineSettings {
      bUseBrushTransparency=false
      roundingType=HalfHeightRadius
      width=0
      color {colorUseRule=UseColor_Specified, specifiedColor=<rgba 0 0 0 0>}
      cornerRadii=<0 0 0 0>
     }
     tintColor {colorUseRule=UseColor_Specified, specifiedColor=<rgba 1 1 1 1>}
     uVRegion {bIsValid=false, max=<0 0>, min=<0 0>}
    }
    colorAndOpacity=<rgba 1 1 1 0.125>
   }
   slot {class=OverlaySlot, properties {horizontalAlignment=HAlign_Center, verticalAlignment=VAlign_Center}}
  }
  VerticalBox_2 {
   widgetClassPath=&VerticalBox
   parent=Overlay_2
   widget=VerticalBox_2
   slot {class=OverlaySlot, properties {horizontalAlignment=HAlign_Fill, verticalAlignment=VAlign_Center, padding=<ltrb 20 4 20 4>}}
  }
  RichText_Upgraded {
   widgetClassPath=&CommonRichTextBlock
   bIsVariable=true
   parent=VerticalBox_2
   widget=RichText_Upgraded
   properties {
    autoWrapText=true
    bOverrideDefaultStyle=true
    justification=Center
    lineHeightPercentage=0.9
    text="Umbrella Upgraded!"
    textTransformPolicy=ToUpper
    defaultTextStyleOverride {
     overflowPolicy=Clip
     transformPolicy=None
     colorAndOpacity {colorUseRule=UseColor_Specified, specifiedColor=<rgba 0.0343 1 0.227 1>}
     font {
      bForceMonospaced=false
      bMaterialIsStencil=false
      fontMaterial=null
      fontObject=&Burbank
      letterSpacing=0
      monospacedWidth=1
      size=36
      skewAmount=0
      typefaceFontName=BurbankRegular-Black
      outlineSettings {
       bApplyOutlineToDropShadows=false
       bMiteredCorners=false
       bSeparateFillAlpha=false
       outlineMaterial=null
       outlineSize=0
       outlineColor=<rgba 0 0 0 1>
      }
     }
     highlightColor {colorUseRule=UseColor_Specified, specifiedColor=<rgba 0 0 0 1>}
     highlightShape {
      bIsDynamicallyLoaded=false
      drawAs=Image
      imageType=NoImage
      mirroring=NoMirror
      resourceName=None
      resourceObject=null
      tiling=NoTile
      imageSize=<32 32>
      margin=<ltrb 0 0 0 0>
      outlineSettings {
       bUseBrushTransparency=false
       roundingType=HalfHeightRadius
       width=0
       color {colorUseRule=UseColor_Specified, specifiedColor=<rgba 0 0 0 0>}
       cornerRadii=<0 0 0 0>
      }
      tintColor {colorUseRule=UseColor_Specified, specifiedColor=<rgba 1 1 1 1>}
      uVRegion {bIsValid=false, max=<0 0>, min=<0 0>}
     }
     selectedBackgroundColor {colorUseRule=UseColor_Foreground, specifiedColor=<rgba 1 0 1 1>}
     shadowColorAndOpacity=<rgba 0 0 0 1>
     shadowOffset=<0 0>
     strikeBrush {
      bIsDynamicallyLoaded=false
      drawAs=Image
      imageType=NoImage
      mirroring=NoMirror
      resourceName=None
      resourceObject=null
      tiling=NoTile
      imageSize=<32 32>
      margin=<ltrb 0 0 0 0>
      outlineSettings {
       bUseBrushTransparency=false
       roundingType=HalfHeightRadius
       width=0
       color {colorUseRule=UseColor_Specified, specifiedColor=<rgba 0 0 0 0>}
       cornerRadii=<0 0 0 0>
      }
      tintColor {colorUseRule=UseColor_Specified, specifiedColor=<rgba 1 1 1 1>}
      uVRegion {bIsValid=false, max=<0 0>, min=<0 0>}
     }
     underlineBrush {
      bIsDynamicallyLoaded=false
      drawAs=Image
      imageType=NoImage
      mirroring=NoMirror
      resourceName=None
      resourceObject=null
      tiling=NoTile
      imageSize=<32 32>
      margin=<ltrb 0 0 0 0>
      outlineSettings {
       bUseBrushTransparency=false
       roundingType=HalfHeightRadius
       width=0
       color {colorUseRule=UseColor_Specified, specifiedColor=<rgba 0 0 0 0>}
       cornerRadii=<0 0 0 0>
      }
      tintColor {colorUseRule=UseColor_Specified, specifiedColor=<rgba 1 1 1 1>}
      uVRegion {bIsValid=false, max=<0 0>, min=<0 0>}
     }
    }
    renderTransform {angle=0, scale=<1 1>, shear=<-12 0>, translation=<0 0>}
   }
   slot {class=VerticalBoxSlot}
  }
  RichText_UnlockUsed {
   widgetClassPath=&CommonRichTextBlock
   bIsVariable=true
   parent=VerticalBox_2
   widget=RichText_UnlockUsed
   properties {
    autoWrapText=true
    bOverrideDefaultStyle=true
    justification=Center
    text="With <text color=\"FFF\">choice!</>"
    textTransformPolicy=ToUpper
    visibility=Collapsed
    defaultTextStyleOverride {
     overflowPolicy=Clip
     transformPolicy=None
     colorAndOpacity {colorUseRule=UseColor_Specified, specifiedColor=<rgba 0.0343 1 0.227 1>}
     font {
      bForceMonospaced=false
      bMaterialIsStencil=false
      fontMaterial=null
      fontObject=&BurbankSmall
      letterSpacing=0
      monospacedWidth=1
      size=18
      skewAmount=0
      typefaceFontName=Black
      outlineSettings {
       bApplyOutlineToDropShadows=false
       bMiteredCorners=false
       bSeparateFillAlpha=false
       outlineMaterial=null
       outlineSize=0
       outlineColor=<rgba 0 0 0 1>
      }
     }
     highlightColor {colorUseRule=UseColor_Specified, specifiedColor=<rgba 0 0 0 1>}
     highlightShape {
      bIsDynamicallyLoaded=false
      drawAs=Image
      imageType=NoImage
      mirroring=NoMirror
      resourceName=None
      resourceObject=null
      tiling=NoTile
      imageSize=<32 32>
      margin=<ltrb 0 0 0 0>
      outlineSettings {
       bUseBrushTransparency=false
       roundingType=HalfHeightRadius
       width=0
       color {colorUseRule=UseColor_Specified, specifiedColor=<rgba 0 0 0 0>}
       cornerRadii=<0 0 0 0>
      }
      tintColor {colorUseRule=UseColor_Specified, specifiedColor=<rgba 1 1 1 1>}
      uVRegion {bIsValid=false, max=<0 0>, min=<0 0>}
     }
     selectedBackgroundColor {colorUseRule=UseColor_Foreground, specifiedColor=<rgba 1 0 1 1>}
     shadowColorAndOpacity=<rgba 0 0 0 1>
     shadowOffset=<0 0>
     strikeBrush {
      bIsDynamicallyLoaded=false
      drawAs=Image
      imageType=NoImage
      mirroring=NoMirror
      resourceName=None
      resourceObject=null
      tiling=NoTile
      imageSize=<32 32>
      margin=<ltrb 0 0 0 0>
      outlineSettings {
       bUseBrushTransparency=false
       roundingType=HalfHeightRadius
       width=0
       color {colorUseRule=UseColor_Specified, specifiedColor=<rgba 0 0 0 0>}
       cornerRadii=<0 0 0 0>
      }
      tintColor {colorUseRule=UseColor_Specified, specifiedColor=<rgba 1 1 1 1>}
      uVRegion {bIsValid=false, max=<0 0>, min=<0 0>}
     }
     underlineBrush {
      bIsDynamicallyLoaded=false
      drawAs=Image
      imageType=NoImage
      mirroring=NoMirror
      resourceName=None
      resourceObject=null
      tiling=NoTile
      imageSize=<32 32>
      margin=<ltrb 0 0 0 0>
      outlineSettings {
       bUseBrushTransparency=false
       roundingType=HalfHeightRadius
       width=0
       color {colorUseRule=UseColor_Specified, specifiedColor=<rgba 0 0 0 0>}
       cornerRadii=<0 0 0 0>
      }
      tintColor {colorUseRule=UseColor_Specified, specifiedColor=<rgba 1 1 1 1>}
      uVRegion {bIsValid=false, max=<0 0>, min=<0 0>}
     }
    }
   }
   slot {class=VerticalBoxSlot}
  }
 ]
}
