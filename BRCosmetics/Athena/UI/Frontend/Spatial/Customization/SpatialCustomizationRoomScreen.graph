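# Frontend widget screen for the Build-A-Brella customizable umbrella glider room, built on AthenaSpatialCustomizationRoom.
# Room state changes switch between category, selection or placeholder views and cache the active one in WidgetToRestore.
# The More Info button, which logs an analytic, and the first-time event show an About modal with buttons collapsed; closing restores the cached view.
# Hardware visibility borders swap the Back button for a close button on touch input.
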WidgetBlueprint /BRCosmetics/Athena/UI/Frontend/Spatial/Customization/SpatialCustomizationRoomScreen
graphs [
 EventGraph {nodes [
   "On Button Clicked (Button_About)"!() -> (OutputDelegate, then, Button:Common Button Legacy Object Reference) {is=title K2Node_ComponentBoundEvent}
   > n1 = FireSpatialScreenAnalytic!(execute, self:Athena Spatial Screen Object Reference, Interaction:String=MoreInfo) -> (then)
   SetActiveWidget(
    execute=[@n1.then, @n2.then]
    self:Widget Switcher Object Reference=Get Switcher_CustomizationState() -> (Switcher_CustomizationState:Common Activatable Widget Switcher Object Reference)
    Widget=Get BPS12_AboutModal() -> (BPS12_AboutModal:BPS12 About Modal Object Reference)
   )
   > SetVisibility(
    self=[
     Get Overlay_TouchBack() -> (Overlay_TouchBack:Overlay Object Reference)
     Get SafeZone_buttons() -> (SafeZone_buttons:Safe Zone Object Reference)
    ]
    InVisibility=Collapsed
   )
   "BP on About Modal Closed (BPS12_AboutModal)"() {is=title K2Node_ComponentBoundEvent}
   > SetActiveWidget(self=Get Switcher_CustomizationState(), Widget=Get WidgetToRestore() -> (WidgetToRestore:Widget Object Reference))
   > SetVisibility(self=[Get SafeZone_buttons(), Get Overlay_TouchBack()], InVisibility=SelfHitTestInvisible)
   n2 = Event OnShowFirstTimeInfo() {is=title K2Node_Event}
   n3 = Event OnRoomStateChanged!() -> (OutputDelegate, then, CurrentRoomState:ESpatialCustomizationRoomState Enum=None, PreviousRoomState:ESpatialCustomizationRoomState Enum=None) {is=title K2Node_Event}
   > HandeRoomStateChanged!(
    execute
    self
    CurrentRoomState:ESpatialCustomizationRoomState Enum=@n3.CurrentRoomState
    PreviousRoomState:ESpatialCustomizationRoomState Enum=@n3.PreviousRoomState
   ) -> (then)
  ]}
 HandeRoomStateChanged {nodes [
   HandeRoomStateChanged!() -> (
    then
    CurrentRoomState:ESpatialCustomizationRoomState Enum=CategorySelection
    PreviousRoomState:ESpatialCustomizationRoomState Enum=CategorySelection
   ) {is=title K2Node_FunctionEntry}
   n1 = Select!(
    None:Widget Object Reference=Get Overlay_Dummy() -> (Overlay_Dummy:Overlay Object Reference)
    CategorySelection:Widget Object Reference=Get Categories() -> (Categories:Athena Spatial Customization Categories Object Reference)
    CustomizationSelection:Widget Object Reference=Get Selection() -> (Selection:Athena Spatial Customization Selection Object Reference)
    Index:ESpatialCustomizationRoomState Enum=Get CurrentRoomState() -> (CurrentRoomState:ESpatialCustomizationRoomState Enum=None)
   ) -> (ReturnValue:Widget Object Reference)
   > SetActiveWidget(self=Get Switcher_CustomizationState(), Widget=@n1)
   > Set WidgetToRestore(WidgetToRestore:Widget Object Reference=@n1) -> (Output_Get:Widget Object Reference)
   > SetVisibility(self=Get SafeZone_buttons(), InVisibility=SelfHitTestInvisible)
  ]}
]
variables [WidgetToRestore {type=Widget, default=null}]
widgets {
 info {inheritedWidgetCount=5, parentClass=/Script/FortniteUI.AthenaSpatialCustomizationRoom, rootWidgetClass=&Overlay, widgetCount=15}
 widgets [
  Overlay_Customization {widgetClassPath=&Overlay, bIsVariable=true, widget=Overlay_Customization}
  Switcher_CustomizationState {
   widgetClassPath=&CommonActivatableWidgetSwitcher
   bIsVariable=true
   parent=Overlay_Customization
   widget=Switcher_CustomizationState
   properties {transitionDuration=0}
   slot {class=OverlaySlot, properties {horizontalAlignment=HAlign_Fill, verticalAlignment=VAlign_Fill}}
  }
  Overlay_Dummy {
   widgetClassPath=&Overlay
   bIsVariable=true
   parent=Switcher_CustomizationState
   widget=Overlay_Dummy
   slot {class=WidgetSwitcherSlot}
  }
  Categories {
   widgetClassPath=&/SpatialCustomizationCategories
   bInherited=true
   bIsVariable=true
   parent=Switcher_CustomizationState
   widget=Categories
   slot {class=WidgetSwitcherSlot}
  }
  Selection {
   widgetClassPath=&/SpatialCustomizationSelection
   bInherited=true
   bIsVariable=true
   parent=Switcher_CustomizationState
   widget=Selection
   slot {class=WidgetSwitcherSlot}
  }
  BPS12_AboutModal {
   widgetClassPath=/BRCosmetics/Athena/UI/Frontend/Spatial/BPS12_AboutModal
   bIsVariable=true
   parent=Switcher_CustomizationState
   widget=BPS12_AboutModal
   properties {
    header=Build-A-Brella
    subHeader=Available for Battle Pass owners
    title="Customizable glider!"
    divider_Color=<rgba 0.25 0.875 1 1>
    entries [
     "<text color=\"FFFFFFFF\" fontface=\"black\" case=\"upper\">Level up</> to unlock more <text color=\"FFFFFFFF\" fontface=\"black\" case=\"upper\">Umbrella Parts</>."
     "<text color=\"FFFFFFFF\" fontface=\"black\" case=\"upper\">Complete challenges</> to unlock new <text color=\"FFFFFFFF\" fontface=\"black\" case=\"upper\">colors</>."
     "<text color=\"FFFFFFFF\" fontface=\"black\" case=\"upper\">Permanently finalize</> your choices to craft your own <text color=\"FFFFFFFF\" fontface=\"black\" case=\"upper\">unique umbrella</>."
    ]
    subheader_Color=<rgba 0 0.6667 1 1>
   }
   slot {class=WidgetSwitcherSlot}
  }
  SafeZone_buttons {
   widgetClassPath=&SafeZone
   bIsVariable=true
   parent=Overlay_Customization
   widget=SafeZone_buttons
   slot {class=OverlaySlot, properties {horizontalAlignment=HAlign_Right, verticalAlignment=VAlign_Bottom, padding=<ltrb 0 0 32 32>}}
  }
  SizeBox_buttons {
   widgetClassPath=&SizeBox
   parent=SafeZone_buttons
   widget=SizeBox_buttons
   properties {bOverride_HeightOverride=true, heightOverride=60}
   slot {class=SafeZoneSlot}
  }
  HorizontalBox_buttonbottom {widgetClassPath=&HorizontalBox, parent=SizeBox_buttons, widget=HorizontalBox_buttonbottom, slot {class=SizeBoxSlot}}
  Button_About {
   widgetClassPath=/FortUILibrary/Foundation/Buttons/SoloButton
   bInherited=true
   bIsVariable=true
   parent=HorizontalBox_buttonbottom
   widget=Button_About
   properties {
    button Description=MORE INFO
    clickMethod=DownAndUp
    fontSize=18
    fontSizeOveride=true
    minWidth=192
    onButtonBaseClicked=(&self:WidgetTree.Button_About.HandleOnButtonClicked)
    onButtonBaseDoubleClicked=(&self:WidgetTree.Button_About.HandleOnButtonDoubleClicked)
    onButtonBaseHovered=(&self:WidgetTree.Button_About.HandleOnButtonHovered)
    onButtonBaseUnhovered=(&self:WidgetTree.Button_About.HandleOnButtonUnhovered)
    onSelectedChangedBase=(&self:WidgetTree.Button_About.HandleOnSelectedChanged)
    pressMethod=DownAndUp
    hoveredSlateSoundOverride {resourceObject=/Game/Sounds/Fort_UI_Sounds/FrontEnd/New_FrontEnd/Fortnite_UI_MainMenu_Tab_Hover_01_Cue}
    padding Overide=<ltrb 24 8 24 8>
    pressedSlateSoundOverride {resourceObject=/Game/Sounds/Fort_UI_Sounds/FrontEnd/New_FrontEnd/Fortnite_UI_MainMenu_Generic_01_Cue}
    triggeringInputAction {dataTable=&InputActionDataTable, rowName=GenericFaceButtonTop}
   }
   slot {class=HorizontalBoxSlot, properties {horizontalAlignment=HAlign_Right, verticalAlignment=VAlign_Bottom, padding=<ltrb 0 0 24 0>}}
  }
  HardwareVisibilityBorder_Back {
   widgetClassPath=&CommonHardwareVisibilityBorder
   parent=HorizontalBox_buttonbottom
   widget=HardwareVisibilityBorder_Back
   properties {visibilityQuery {
     autoDescription=" NONE( Input.Touch )"
     tokenStreamVersion=0
     userDescription=""
     queryTokenStream [0,1,3,1,0]
     tagDictionary [{tagName=Input.Touch}]
    }}
   slot {class=HorizontalBoxSlot, properties {horizontalAlignment=HAlign_Right, verticalAlignment=VAlign_Bottom}}
  }
  Button_Back {
   widgetClassPath=&SoloButton
   bInherited=true
   bIsVariable=true
   parent=HardwareVisibilityBorder_Back
   widget=Button_Back
   properties {
    button Description=Back
    clickMethod=DownAndUp
    fontSize=18
    fontSizeOveride=true
    minWidth=192
    onButtonBaseClicked=(&self:WidgetTree.Button_Back.HandleOnButtonClicked)
    onButtonBaseDoubleClicked=(&self:WidgetTree.Button_Back.HandleOnButtonDoubleClicked)
    onButtonBaseHovered=(&self:WidgetTree.Button_Back.HandleOnButtonHovered)
    onButtonBaseUnhovered=(&self:WidgetTree.Button_Back.HandleOnButtonUnhovered)
    onSelectedChangedBase=(&self:WidgetTree.Button_Back.HandleOnSelectedChanged)
    pressMethod=DownAndUp
    hoveredSlateSoundOverride {resourceObject=&Fortnite_UI_MainMenu_Tab_Hover_01_Cue}
    padding Overide=<ltrb 24 8 24 8>
    pressedSlateSoundOverride {resourceObject=&Fortnite_UI_MainMenu_Generic_01_Cue}
    triggeringInputAction {dataTable=&InputActionDataTable, rowName=Back}
   }
   slot {class=BorderSlot, properties {padding=<ltrb 0 0 0 0>}}
  }
  Overlay_TouchBack {
   widgetClassPath=&Overlay
   bIsVariable=true
   parent=Overlay_Customization
   widget=Overlay_TouchBack
   slot {class=OverlaySlot, properties {horizontalAlignment=HAlign_Right, padding=<ltrb 0 32 32 0>}}
  }
  HardwareVisibilityBorder_TouchBack {
   widgetClassPath=&CommonHardwareVisibilityBorder
   parent=Overlay_TouchBack
   widget=HardwareVisibilityBorder_TouchBack
   properties {visibilityQuery {
     autoDescription=" ANY( Input.Touch )"
     tokenStreamVersion=0
     userDescription=""
     queryTokenStream [0,1,1,1,0]
     tagDictionary [{tagName=Input.Touch}]
    }}
   slot {class=OverlaySlot}
  }
  Button_TouchBack {
   widgetClassPath=/FortUILibrary/Foundation/Buttons/CloseButton
   bInherited=true
   bIsVariable=true
   parent=HardwareVisibilityBorder_TouchBack
   widget=Button_TouchBack
   properties {
    onButtonBaseClicked=(&self:WidgetTree.Button_TouchBack.HandleOnButtonClicked)
    onButtonBaseDoubleClicked=(&self:WidgetTree.Button_TouchBack.HandleOnButtonDoubleClicked)
    onButtonBaseHovered=(&self:WidgetTree.Button_TouchBack.HandleOnButtonHovered)
    onButtonBaseUnhovered=(&self:WidgetTree.Button_TouchBack.HandleOnButtonUnhovered)
    onSelectedChangedBase=(&self:WidgetTree.Button_TouchBack.HandleOnSelectedChanged)
   }
   slot {class=BorderSlot, properties {padding=<ltrb 0 0 0 0>}}
  }
 ]
}
